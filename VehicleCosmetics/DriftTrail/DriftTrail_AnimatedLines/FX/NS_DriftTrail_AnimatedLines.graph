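# Niagara system for the AnimatedLines drift trail vehicle cosmetic, drawing short-lived, glowing line effects at the vehicle's skeletal mesh.
# Looping burst emitters partition particles across bones or sockets, mirroring mesh scale, and sometimes spin, per partition through array lookups.
# Paint variant overrides choose color, while curves drive dynamic material parameters and brightness, switching on painted state.
# Several variant emitters are disabled; others compute game-mode flags from User.GameModeID.
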NiagaraSystem /VehicleCosmetics/DriftTrail/DriftTrail_AnimatedLines/FX/NS_DriftTrail_AnimatedLines
dependencies {
 usedDataInterfaces [
  &NiagaraDataInterfaceArrayColor
  &NiagaraDataInterfaceArrayFloat3
  &NiagaraDataInterfaceCurve
  &NiagaraDataInterfaceVectorCurve
  &NiagaraDataInterfaceMeshRendererInfo
  &NiagaraDataInterfaceArrayFloat
  &NiagaraDataInterfaceSkeletalMesh
 ]
 usedDynamicInputs []
 usedModules [
  &self:SystemSpawnScript.SystemScriptSource.SystemScriptGraph.NiagaraNodeAssignment_0.SetVariables_#1
  &SystemState
 ]
}
emitters {
 NE_VC_DriftTrail_CURVE001 {
  emitterUpdateScript [
   &EmitterState(Life Cycle Mode=::NewEnumerator1, Loop Duration=0.05, Recalculate Duration Each Loop=true, UseLoopDelay=false)
   &SpawnBurst_Instantaneous(Spawn Count=25, Use Spawn Probability=false, UseLoopCountLimit=false)
   &SpawnRate(
    SpawnRate=&MakeCustomFloatFromBool(Bool=$User.bIsFrontendPreview, True Float=0, False Float=50)
    Use Spawn Probability=false
   ) {enabled=false}
   /SetVariables(Emitter.IS BR=HlslExpression() {hlslExpression="User.GameModeID == 1 ? true : false"}) {isSetParameters=true}
  ]
  particleSpawnScript [
   &VC_PaintedVariantsOverride(
    Black=<rgba 0.01 0.01 0.01 1>
    BurntSienna=<rgba 0.25 0.0625 0 1>
    Cobalt=<rgba 0 0.125 1.5 1>
    DefaultColor=&SelectLinearColorFromArray(
     Array Sampling Mode=::NewEnumerator0
     Direct Array Index=?
     Color Selection Array=DataInterface() {propertyValues {ColorData [<rgba 0 1.6667 4 1>, <rgba 1.5 0 0.2 1>]}}
     Random Seed=0
    )
    ForestGreen=<rgba 0 0.7 0 1>
    Grey=<rgba 0.1 0.1 0.1 1>
    Lime=<rgba 0.5208 1.25 0 1>
    Orange=<rgba 1.25 0.0417 0 1>
    Pink=<rgba 0.8 0.0958 0.5439 1>
    Saffron=<rgba 0.7 0.7 0 1>
    SkyBlue=<rgba 0 0.6654 1 1>
    TitaniumWhite=<rgba 1 1 1 1>
   )
   &InitializeParticle(
    Color=$Particles.Color
    Lifetime Max=0.3
    Lifetime Min=0.3
    Lifetime Mode=::NewEnumerator1
    Mesh Scale Max=<5 6 6>
    Mesh Scale Min=<2 3 3>
    Mesh Scale Mode=::NewEnumerator2
    Sprite Rotation Mode=::NewEnumerator2
    Sprite Size Mode=::NewEnumerator3
    UsePositionOffset=false
    Write Material Random=true
   )
   &PartitionParticles()
   &SkeletalMeshLocation(
    Bone / Socket Index=&Add_Integer(A=$Particles.PartitionParticles.Partition, B=4)
    Bone / Socket Sampling Mode=::NewEnumerator5
    Interpolate Mesh Sampling=false
    MaskSpawnGroup=false
    Mesh Sampling Type=::NewEnumerator6
    Sampled Position Offset=&AddVectorToPosition(Position=<0 0 0>, Vector=$User.offset)
    Skeletal Mesh=$System.SkeletalMesh
   )
   &AddVelocity(Velocity=<1000 0 0>, Velocity Speed Scale=-1) {enabled=false}
   &InitialMeshOrientation(
    Mesh Axis To Orient=&RandomRangeVector(Minimum=<0 -1 0>, Maximum=<0 1 0>)
    Mesh Orientation Mode=::NewEnumerator2
    Orientation Coordinate Space=::Local
    Orientation Vector=&RandomRangeVector(Minimum=<0 -1 -1>, Maximum=<0 1 -1>, Randomness Mode=::NewEnumerator1)
    Rotation=<90 70 0>
    Rotation Coordinate Space=::NewEnumerator2
    Use Rotation Vector=true
   )
  ]
  particleUpdateScript [
   &ParticleState()
   &ScaleMeshSize(Scale Factor=&SelectVectorFromArray(
     Direct Array Index=$Particles.PartitionParticles.Partition
     Vector Selection Array=DataInterface() {propertyValues {FloatData [<1 0.75 0.75>, <1 -0.75 0.75>]}}
    ))
   &MeshRotationForce(Rotation=<-5 0 0>)
   &InheritSourceMovement(
    Applied Position Scale=&Multiply_VectorByFloat(Vector=<1 1 1>, Float=1)
    ApplyForce=false
    ApplyPosition=true
    ApplyPositionFalloff=false
   ) {enabled=false}
   &SolveForcesAndVelocity(Clamp Velocity=false, Limit Acceleration=false)
   &DynamicMaterialParameters(
    Index 0 Param 1=&FloatFromCurve(FloatCurve=DataInterface() {propertyValues {Curve {keys [<at 0.2 0 RCIM_Cubic>, {arriveTangent=0.644, tangentMode=RCTM_User, time=1, value=1}]}}})
    Index 0 Param 2=&FloatFromCurve(
     FloatCurve=DataInterface() {propertyValues {Curve {keys [{arriveTangent=-0.0161, interpMode=RCIM_Cubic, leaveTangent=-0.0161, tangentMode=RCTM_User, time=0.1, value=1}, <at 1 0 RCIM_Cubic>]}}}
     Scale Curve=&RandomRangeFloat(Minimum=1, Maximum=1)
    )
    Index 1 Param 1=&MakeCustomFloatFromBool(Bool=$User.bIsPainted, True Float=0, False Float=1)
    Module.Index0Param0Write=true
    Module.Index0Param1Write=true
    Module.Index0Param2Write=true
    Module.Index0Param3Write=true
    Module.Index1Param0Write=true
    Module.Index1Param1Write=true
    Module.Index1Param2Write=true
    Module.Index1Param3Write=true
    Write Parameter Index 1=true
   ) {enabled=false}
   &DynamicMaterialParameters(
    Index 0 Param 1=&FloatFromCurve(
     FloatCurve=DataInterface() {propertyValues {Curve {keys [
         <at 0.4 0 RCIM_Cubic>
         {arriveTangent=0.002, interpMode=RCIM_Cubic, leaveTangent=0.7645, tangentMode=RCTM_Break, time=0.5, value=0.5}
         {arriveTangent=0.644, tangentMode=RCTM_User, time=1, value=1}
        ]}}}
     Scale Curve=0.8
    )
    Index 0 Param 2=&FloatFromCurve(
     FloatCurve=DataInterface() {propertyValues {Curve {keys [
         <at 0 1 RCIM_Cubic>
         {arriveTangent=-1.3855, interpMode=RCIM_Cubic, leaveTangent=-1.3855, tangentMode=RCTM_User, time=0.5, value=0.3}
         <at 1 0 RCIM_Cubic>
        ]}}}
     Scale Curve=&RandomRangeFloat(Minimum=0.85, Maximum=2.85)
    )
    Index 1 Param 1=&MakeCustomFloatFromBool(Bool=$User.bIsPainted, True Float=0, False Float=1)
    Index 1 Param 2=&MakeCustomFloatFromBool(Bool=$User.bIsPainted, True Float=&MakeFloatFromInt(Int=$User.Lines_Glow), False Float=100)
    Module.Index0Param0Write=true
    Module.Index0Param1Write=true
    Module.Index0Param2Write=true
    Module.Index0Param3Write=true
    Module.Index1Param0Write=true
    Module.Index1Param1Write=true
    Module.Index1Param2Write=true
    Module.Index1Param3Write=true
    Write Parameter Index 1=true
   ) {module=DynamicMaterialParameters001}
   &ScaleColor(
    Scale RGB=&VectorFromFloat(Value=&MakeCustomFloatFromBool(
      Bool=$User.bIsPainted
      True Float=&FloatFromCurve(
       FloatCurve=DataInterface() {propertyValues {Curve {keys [{interpMode=RCIM_Cubic, leaveTangent=-1.2605, leaveTangentWeight=0.6485, tangentMode=RCTM_Break, time=0, value=1.25}, <at 1 0.1 RCIM_Cubic>]}}}
       Scale Curve=5
      )
      False Float=&FloatFromCurve(
       FloatCurve=DataInterface() {propertyValues {Curve {keys [
           {arriveTangent=-1.9978, interpMode=RCIM_Cubic, leaveTangent=-1.9978, tangentMode=RCTM_User, time=0, value=1}
           {
            arriveTangentWeight=0.311
            interpMode=RCIM_Cubic
            leaveTangent=-2.5749
            leaveTangentWeight=0.0777
            tangentMode=RCTM_Break
            time=0.8
            value=0.4
           }
           {arriveTangentWeight=0.2828, interpMode=RCIM_Cubic, leaveTangentWeight=0.0471, tangentMode=RCTM_User, time=1, value=0.2}
          ]}}}
       Scale Curve=12
      )
     ))
    ScaleA=true
    ScaleRGB=true
   )
  ]
  rendererClasses [&NiagaraMeshRendererProperties, &NiagaraSpriteRendererProperties]
  renderers [
   {rendererClass=&NiagaraMeshRendererProperties, rendererIndex=0}
   {rendererClass=&NiagaraSpriteRendererProperties, rendererIndex=1}
  ]
 }
 NE_VC_DriftTrail_CURVE002 {
  bEnabled=false
  emitterUpdateScript [
   &EmitterState(Life Cycle Mode=::NewEnumerator1, Loop Duration=0.1, Recalculate Duration Each Loop=true, UseLoopDelay=false)
   &SpawnBurst_Instantaneous(
    Spawn Count=&MakeCustomIntFromBool(Bool=$User.bIsFrontendPreview, True INT=0, False INT=30)
    Use Spawn Probability=false
    UseLoopCountLimit=false
   )
   &SpawnRate(SpawnRate=50, Use Spawn Probability=false) {enabled=false}
  ]
  particleSpawnScript [
   &VC_PaintedVariantsOverride(
    Black=<rgba 0.01 0.01 0.01 1>
    BurntSienna=<rgba 1 0.2116 0 1>
    Cobalt=<rgba 0 0.125 1.5 1>
    DefaultColor=&SelectLinearColorFromArray(
     Array Sampling Mode=::NewEnumerator0
     Direct Array Index=?
     Color Selection Array=DataInterface() {propertyValues {ColorData [<rgba 0 0.42 1.5 1>, <rgba 0.5 0 0.4333 1>]}}
     Random Seed=0
    )
    ForestGreen=<rgba 0 0.7 0 1>
    Grey=<rgba 0.1 0.1 0.1 1>
    Lime=<rgba 0.5208 1.25 0 1>
    Orange=<rgba 1.25 0.0417 0 1>
    Pink=<rgba 0.8 0.0958 0.5439 1>
    Purple=<rgba 0.1 0 0.6 10>
    Saffron=<rgba 0.7 0.7 0 1>
    SkyBlue=<rgba 0 0.6654 1 1>
    TitaniumWhite=<rgba 1 1 1 1>
   )
   &InitializeParticle(
    Color=$Particles.Color
    Lifetime Max=0.3
    Lifetime Min=0.3
    Lifetime Mode=::NewEnumerator1
    Mesh Scale Max=<5 6 6>
    Mesh Scale Min=<2 3 3>
    Mesh Scale Mode=::NewEnumerator2
    Sprite Rotation Mode=::NewEnumerator2
    Sprite Size Mode=::NewEnumerator3
    UsePositionOffset=false
    Write Material Random=true
   )
   &PartitionParticles()
   &SkeletalMeshLocation(
    Bone / Socket Index=&Add_Integer(A=$Particles.PartitionParticles.Partition, B=4)
    Bone / Socket Sampling Mode=::NewEnumerator5
    Interpolate Mesh Sampling=false
    MaskSpawnGroup=false
    Mesh Sampling Type=::NewEnumerator6
    Sampled Position Offset=&AddVectorToPosition(Position=<0 0 0>, Vector=$User.offset)
    Skeletal Mesh=$System.SkeletalMesh
   )
   &AddVelocity(Velocity=<1000 0 0>, Velocity Speed Scale=-1) {enabled=false}
   &InitialMeshOrientation(
    Mesh Axis To Orient=&RandomRangeVector(Minimum=<0 -1 -1>, Maximum=<0 1 1>)
    Mesh Orientation Mode=::NewEnumerator2
    Orientation Vector=&RandomRangeVector(Minimum=<0 1 1>, Maximum=<0 -1 -1>, Randomness Mode=::NewEnumerator1)
    Rotation=<90 65 0>
    Rotation Coordinate Space=::NewEnumerator2
    Use Rotation Vector=true
   )
  ]
  particleUpdateScript [
   &ParticleState()
   &ScaleMeshSize(Scale Factor=&SelectVectorFromArray(
     Direct Array Index=$Particles.PartitionParticles.Partition
     Vector Selection Array=DataInterface() {propertyValues {FloatData [<1 -0.75 0.75>, <1 0.75 0.75>]}}
    ))
   &MeshRotationForce(Rotation=&SelectVectorFromArray(
     Direct Array Index=$Particles.PartitionParticles.Partition
     Vector Selection Array=DataInterface() {propertyValues {FloatData [<-5 0 0>, <5 0 0>]}}
    ))
   &SolveForcesAndVelocity(Clamp Velocity=false, Limit Acceleration=false)
   &DynamicMaterialParameters(
    Index 0 Param 1=&FloatFromCurve(FloatCurve=DataInterface() {propertyValues {Curve {keys [
         <at 0.3 0 RCIM_Cubic>
         {arriveTangent=0.002, interpMode=RCIM_Cubic, leaveTangent=0.7645, tangentMode=RCTM_Break, time=0.6, value=0.7}
         {arriveTangent=0.644, tangentMode=RCTM_User, time=1, value=1}
        ]}}})
    Index 0 Param 2=&FloatFromCurve(
     FloatCurve=DataInterface() {propertyValues {Curve {keys [
         <at 0 1 RCIM_Cubic>
         {arriveTangent=0.1281, interpMode=RCIM_Cubic, leaveTangent=0.1281, tangentMode=RCTM_User, time=0.5, value=0.2}
         <at 1 1 RCIM_Cubic>
        ]}}}
     Scale Curve=&RandomRangeFloat(Minimum=1, Maximum=1.75)
    )
    Index 1 Param 1=&MakeCustomFloatFromBool(Bool=$User.bIsPainted, True Float=0, False Float=1)
    Module.Index0Param0Write=true
    Module.Index0Param1Write=true
    Module.Index0Param2Write=true
    Module.Index0Param3Write=true
    Module.Index1Param0Write=true
    Module.Index1Param1Write=true
    Module.Index1Param2Write=true
    Module.Index1Param3Write=true
    Write Parameter Index 1=true
   )
   &ScaleColor(
    Scale RGB=&VectorFromFloat(Value=&MakeCustomFloatFromBool(
      Bool=$User.bIsPainted
      True Float=&FloatFromCurve(
       FloatCurve=DataInterface() {propertyValues {Curve {keys [{interpMode=RCIM_Cubic, leaveTangent=-1.6688, tangentMode=RCTM_Break, time=0, value=1.25}, <at 0.7 0.01 RCIM_Cubic>]}}}
       Scale Curve=4
      )
      False Float=&FloatFromCurve(
       FloatCurve=DataInterface() {propertyValues {Curve {keys [{arriveTangent=-2.4129, interpMode=RCIM_Cubic, leaveTangent=-2.4129, tangentMode=RCTM_User, time=0.3, value=1}, <at 1 0 RCIM_Cubic>]}}}
       Scale Curve=&RandomRangeFloat(Minimum=3, Maximum=6)
      )
     ))
    ScaleA=true
    ScaleRGB=true
   )
  ]
  rendererClasses [&NiagaraMeshRendererProperties, &NiagaraSpriteRendererProperties]
  renderers [
   {rendererClass=&NiagaraMeshRendererProperties, rendererIndex=0}
   {rendererClass=&NiagaraSpriteRendererProperties, rendererIndex=1}
  ]
 }
 NE_VC_DriftTrail_CURVE003 {
  emitterUpdateScript [
   &EmitterState(Life Cycle Mode=::NewEnumerator1, Loop Duration=0.05, Recalculate Duration Each Loop=true, UseLoopDelay=false)
   &SpawnBurst_Instantaneous(Spawn Count=12, Use Spawn Probability=false, UseLoopCountLimit=false)
   &SpawnRate(
    SpawnRate=&MakeCustomFloatFromBool(Bool=$User.bIsFrontendPreview, True Float=120, False Float=0)
    Use Spawn Probability=false
   ) {enabled=false}
   /SetVariables(Emitter.IS RR=HlslExpression() {hlslExpression="User.GameModeID == 2 ? true : false"}) {isSetParameters=true}
  ]
  particleSpawnScript [
   &VC_PaintedVariantsOverride(
    Black=<rgba 0.01 0.01 0.01 1>
    BurntSienna=<rgba 0.25 0.0625 0 1>
    Cobalt=<rgba 0 0.125 1.5 1>
    DefaultColor=&SelectLinearColorFromArray(
     Array Sampling Mode=::NewEnumerator0
     Direct Array Index=?
     Color Selection Array=DataInterface() {propertyValues {ColorData [<rgba 0 1.6667 4 1>, <rgba 1.5 0 0.2 1>]}}
     Random Seed=0
    )
    ForestGreen=<rgba 0 0.7 0 1>
    Grey=<rgba 0.1 0.1 0.1 1>
    Lime=<rgba 0.5208 1.25 0 1>
    Orange=<rgba 1.25 0.0417 0 1>
    Pink=<rgba 0.8 0.0958 0.5439 1>
    Saffron=<rgba 0.7 0.7 0 1>
    SkyBlue=<rgba 0 0.6654 1 1>
    TitaniumWhite=<rgba 1 1 1 1>
   )
   &InitializeParticle(
    Color=$Particles.Color
    Lifetime Max=0.4
    Lifetime Min=0.3
    Lifetime Mode=::NewEnumerator1
    Mesh Scale Max=&RandomRangeVector(Minimum=<5 3 3>, Maximum=<5 5 5>)
    Mesh Scale Min=&RandomRangeVector(Minimum=<2 5 5>, Maximum=<2 5 7>)
    Mesh Scale Mode=::NewEnumerator2
    Sprite Rotation Mode=::NewEnumerator2
    Sprite Size Mode=::NewEnumerator3
    UsePositionOffset=false
    Write Material Random=true
   )
   &PartitionParticles()
   &SkeletalMeshLocation(
    Bone / Socket Index=&Add_Integer(A=$Particles.PartitionParticles.Partition, B=4)
    Bone / Socket Sampling Mode=::NewEnumerator5
    Interpolate Mesh Sampling=false
    MaskSpawnGroup=false
    Mesh Sampling Type=::NewEnumerator6
    Sampled Position Offset=&AddVectorToPosition(Position=<0 0 0>, Vector=$User.offset)
    Skeletal Mesh=$System.SkeletalMesh
   )
   &AddVelocity(Velocity=<1000 0 0>, Velocity Speed Scale=-1) {enabled=false}
   &InitialMeshOrientation(
    Mesh Axis To Orient=&RandomRangeVector(Minimum=<0 -1 -1>, Maximum=<0 1 1>)
    Mesh Orientation Mode=::NewEnumerator2
    Orientation Vector=&RandomRangeVector(Minimum=<0 1 1>, Maximum=<0 -1 -1>, Randomness Mode=::NewEnumerator1)
    Rotation=<90 65 0>
    Rotation Coordinate Space=::NewEnumerator2
    Use Rotation Vector=false
   )
  ]
  particleUpdateScript [
   &ParticleState()
   &ScaleMeshSize(Scale Factor=&SelectVectorFromArray(
     Direct Array Index=$Particles.PartitionParticles.Partition
     Vector Selection Array=DataInterface() {propertyValues {FloatData [<1 -0.75 0.75>, <1 0.75 0.75>]}}
    ))
   &MeshRotationForce(Rotation=&SelectVectorFromArray(
     Direct Array Index=$Particles.PartitionParticles.Partition
     Vector Selection Array=DataInterface() {propertyValues {FloatData [<-10 0 0>, <10 0 0>]}}
    )) {enabled=false}
   &SolveForcesAndVelocity(Clamp Velocity=false, Limit Acceleration=false)
   &DynamicMaterialParameters(
    Index 0 Param 1=&FloatFromCurve(
     FloatCurve=DataInterface() {propertyValues {Curve {keys [
         <at 0.4 0 RCIM_Cubic>
         {arriveTangent=0.002, interpMode=RCIM_Cubic, leaveTangent=0.7645, tangentMode=RCTM_Break, time=0.5, value=0.55}
         {arriveTangent=0.644, tangentMode=RCTM_User, time=1, value=0.95}
        ]}}}
     Scale Curve=0.9
    )
    Index 0 Param 2=&FloatFromCurve(
     FloatCurve=DataInterface() {propertyValues {Curve {keys [
         <at 0 1 RCIM_Cubic>
         {arriveTangent=-1.3855, interpMode=RCIM_Cubic, leaveTangent=-1.3855, tangentMode=RCTM_User, time=0.5, value=0.4}
         <at 1 0 RCIM_Cubic>
        ]}}}
     Scale Curve=&RandomRangeFloat(Minimum=0.9, Maximum=0.9)
    )
    Index 1 Param 1=&MakeCustomFloatFromBool(Bool=$User.bIsPainted, True Float=0, False Float=1)
    Index 1 Param 2=&MakeCustomFloatFromBool(Bool=$User.bIsPainted, True Float=&MakeFloatFromInt(Int=$User.Lines_Glow), False Float=100)
    Module.Index0Param0Write=true
    Module.Index0Param1Write=true
    Module.Index0Param2Write=true
    Module.Index0Param3Write=true
    Module.Index1Param0Write=true
    Module.Index1Param1Write=true
    Module.Index1Param2Write=true
    Module.Index1Param3Write=true
    Write Parameter Index 1=true
   )
   &ScaleColor(
    Scale RGB=&VectorFromFloat(Value=&MakeCustomFloatFromBool(
      Bool=$User.bIsPainted
      True Float=&FloatFromCurve(
       FloatCurve=DataInterface() {propertyValues {Curve {keys [{interpMode=RCIM_Cubic, leaveTangent=-1.6688, tangentMode=RCTM_Break, time=0, value=1.25}, <at 0.9 0.15 RCIM_Cubic>]}}}
       Scale Curve=4
      )
      False Float=&FloatFromCurve(
       FloatCurve=DataInterface() {propertyValues {Curve {keys [
           {arriveTangent=-1.9978, interpMode=RCIM_Cubic, leaveTangent=-1.9978, tangentMode=RCTM_User, time=0, value=1}
           {
            arriveTangentWeight=0.311
            interpMode=RCIM_Cubic
            leaveTangent=-2.5749
            leaveTangentWeight=0.0777
            tangentMode=RCTM_Break
            time=0.8
            value=0.4
           }
           {arriveTangentWeight=0.2828, interpMode=RCIM_Cubic, leaveTangentWeight=0.0471, tangentMode=RCTM_User, time=1, value=0.2}
          ]}}}
       Scale Curve=&RandomRangeFloat(Minimum=5, Maximum=5)
      )
     ))
    ScaleA=true
    ScaleRGB=true
   )
  ]
  rendererClasses [&NiagaraMeshRendererProperties, &NiagaraSpriteRendererProperties]
  renderers [
   {rendererClass=&NiagaraMeshRendererProperties, rendererIndex=0}
   {rendererClass=&NiagaraSpriteRendererProperties, rendererIndex=1}
  ]
 }
 NE_VC_DriftTrail_CURVE004 {
  emitterUpdateScript [
   &EmitterState(Life Cycle Mode=::NewEnumerator1, Loop Duration=0.05, Recalculate Duration Each Loop=true, UseLoopDelay=false)
   &SpawnBurst_Instantaneous(Spawn Count=6, Use Spawn Probability=false, UseLoopCountLimit=false)
   &SpawnRate(
    SpawnRate=&MakeCustomFloatFromBool(Bool=$User.bIsFrontendPreview, True Float=120, False Float=0)
    Use Spawn Probability=false
   ) {enabled=false}
   /SetVariables(Emitter.IS FRONTEND=HlslExpression() {hlslExpression="User.GameModeID == 0 ? true : false"}) {isSetParameters=true}
  ]
  particleSpawnScript [
   &VC_PaintedVariantsOverride(
    Black=<rgba 0.01 0.01 0.01 1>
    BurntSienna=<rgba 0.25 0.0625 0 1>
    Cobalt=<rgba 0 0.125 1.5 1>
    DefaultColor=&SelectLinearColorFromArray(
     Array Sampling Mode=::NewEnumerator0
     Direct Array Index=?
     Color Selection Array=DataInterface() {propertyValues {ColorData [<rgba 0 1.6667 4 1>, <rgba 1.5 0 0.2 1>]}}
     Random Seed=0
    )
    ForestGreen=<rgba 0 0.7 0 1>
    Grey=<rgba 0.07 0.07 0.07 1>
    Lime=<rgba 0.5208 1.25 0 1>
    Orange=<rgba 1.25 0.0833 0 1>
    Pink=<rgba 0.8 0.0958 0.5439 1>
    Saffron=<rgba 0.7 0.7 0 1>
    SkyBlue=<rgba 0 0.6654 1 1>
    TitaniumWhite=<rgba 1 1 1 1>
   )
   &InitializeParticle(
    Color=$Particles.Color
    Lifetime Max=0.3
    Lifetime Min=0.3
    Lifetime Mode=::NewEnumerator1
    Mesh Scale Max=&RandomRangeVector(Minimum=<4 3 3>, Maximum=<4 5 5>)
    Mesh Scale Min=&RandomRangeVector(Minimum=<2 5 5>, Maximum=<2 5 7>)
    Mesh Scale Mode=::NewEnumerator2
    Sprite Rotation Mode=::NewEnumerator2
    Sprite Size Mode=::NewEnumerator3
    UsePositionOffset=false
    Write Material Random=true
   )
   &PartitionParticles()
   &SkeletalMeshLocation(
    Bone / Socket Index=&Add_Integer(A=$Particles.PartitionParticles.Partition, B=4)
    Bone / Socket Sampling Mode=::NewEnumerator5
    Interpolate Mesh Sampling=false
    MaskSpawnGroup=false
    Mesh Sampling Type=::NewEnumerator6
    Sampled Position Offset=&AddVectorToPosition(Position=<0 0 0>, Vector=$User.offset)
    Skeletal Mesh=$System.SkeletalMesh
   )
   &AddVelocity(Velocity=<1000 0 0>, Velocity Speed Scale=-1) {enabled=false}
   &InitialMeshOrientation(
    Mesh Axis To Orient=&RandomRangeVector(Minimum=<0 -1 1>, Maximum=<0 1 -1>)
    Mesh Orientation Mode=::NewEnumerator2
    Orientation Vector=&RandomRangeVector(Minimum=<0 1 1>, Maximum=<0 -1 1>, Randomness Mode=::NewEnumerator1)
    Rotation=<90 65 0>
    Rotation Coordinate Space=::NewEnumerator2
    Use Rotation Vector=false
   )
  ]
  particleUpdateScript [
   &ParticleState()
   &ScaleMeshSize(Scale Factor=&SelectVectorFromArray(
     Direct Array Index=$Particles.PartitionParticles.Partition
     Vector Selection Array=DataInterface() {propertyValues {FloatData [<1 -0.75 0.75>, <1 -0.75 -0.75>]}}
    ))
   &MeshRotationForce(Rotation=&SelectVectorFromArray(
     Direct Array Index=$Particles.PartitionParticles.Partition
     Vector Selection Array=DataInterface() {propertyValues {FloatData [<-10 0 0>, <10 0 0>]}}
    )) {enabled=false}
   &SolveForcesAndVelocity(Clamp Velocity=false, Limit Acceleration=false)
   &DynamicMaterialParameters(
    Index 0 Param 1=&FloatFromCurve(
     FloatCurve=DataInterface() {propertyValues {Curve {keys [
         <at 0.4 0 RCIM_Cubic>
         {arriveTangent=0.002, interpMode=RCIM_Cubic, leaveTangent=0.7645, tangentMode=RCTM_Break, time=0.5, value=0.5}
         {arriveTangent=0.644, tangentMode=RCTM_User, time=1, value=1}
        ]}}}
     Scale Curve=0.9
    )
    Index 0 Param 2=&FloatFromCurve(
     FloatCurve=DataInterface() {propertyValues {Curve {keys [
         <at 0 1 RCIM_Cubic>
         {arriveTangent=-1.3855, interpMode=RCIM_Cubic, leaveTangent=-1.3855, tangentMode=RCTM_User, time=0.5, value=0.3}
         <at 1 0 RCIM_Cubic>
        ]}}}
     Scale Curve=&RandomRangeFloat(Minimum=1, Maximum=1)
    )
    Index 1 Param 1=&MakeCustomFloatFromBool(Bool=$User.bIsPainted, True Float=2, False Float=1)
    Module.Index0Param0Write=true
    Module.Index0Param1Write=true
    Module.Index0Param2Write=true
    Module.Index0Param3Write=true
    Module.Index1Param0Write=true
    Module.Index1Param1Write=true
    Module.Index1Param2Write=true
    Module.Index1Param3Write=true
    Write Parameter Index 1=true
   ) {enabled=false}
   &ScaleColor(
    Scale RGB=&VectorFromFloat(Value=&MakeCustomFloatFromBool(
      Bool=$User.bIsPainted
      True Float=&FloatFromCurve(
       FloatCurve=DataInterface() {propertyValues {Curve {keys [{interpMode=RCIM_Cubic, leaveTangent=-1.6688, tangentMode=RCTM_Break, time=0, value=1.25}, <at 1 0.15 RCIM_Cubic>]}}}
       Scale Curve=4
      )
      False Float=&FloatFromCurve(
       FloatCurve=DataInterface() {propertyValues {Curve {keys [
           {arriveTangent=-1.9978, interpMode=RCIM_Cubic, leaveTangent=-1.9978, tangentMode=RCTM_User, time=0, value=1}
           {
            arriveTangentWeight=0.311
            interpMode=RCIM_Cubic
            leaveTangent=-2.5749
            leaveTangentWeight=0.0777
            tangentMode=RCTM_Break
            time=0.8
            value=0.4
           }
           {arriveTangentWeight=0.2828, interpMode=RCIM_Cubic, leaveTangentWeight=0.0471, tangentMode=RCTM_User, time=1, value=0.2}
          ]}}}
       Scale Curve=&RandomRangeFloat(Minimum=3, Maximum=4)
      )
     ))
    ScaleA=true
    ScaleRGB=true
   )
   &DynamicMaterialParameters(
    Index 0 Param 1=&FloatFromCurve(
     FloatCurve=DataInterface() {propertyValues {Curve {keys [
         <at 0.4 0 RCIM_Cubic>
         {arriveTangent=0.002, interpMode=RCIM_Cubic, leaveTangent=0.7645, tangentMode=RCTM_Break, time=0.5, value=0.5}
         {arriveTangent=0.644, tangentMode=RCTM_User, time=1, value=1}
        ]}}}
     Scale Curve=0.9
    )
    Index 0 Param 2=&FloatFromCurve(
     FloatCurve=DataInterface() {propertyValues {Curve {keys [
         <at 0 1 RCIM_Cubic>
         {arriveTangent=-1.3855, interpMode=RCIM_Cubic, leaveTangent=-1.3855, tangentMode=RCTM_User, time=0.5, value=0.3}
         <at 1 0 RCIM_Cubic>
        ]}}}
     Scale Curve=&RandomRangeFloat(Minimum=1, Maximum=1)
    )
    Index 1 Param 1=&MakeCustomFloatFromBool(Bool=$User.bIsPainted, True Float=0, False Float=1)
    Index 1 Param 2=&MakeCustomFloatFromBool(Bool=$User.bIsPainted, True Float=&MakeFloatFromInt(Int=$User.Lines_Glow), False Float=100)
    Module.Index0Param0Write=true
    Module.Index0Param1Write=true
    Module.Index0Param2Write=true
    Module.Index0Param3Write=true
    Module.Index1Param0Write=true
    Module.Index1Param1Write=true
    Module.Index1Param2Write=true
    Module.Index1Param3Write=true
    Write Parameter Index 1=true
   ) {module=DynamicMaterialParameters001}
  ]
  rendererClasses [&NiagaraMeshRendererProperties, &NiagaraSpriteRendererProperties]
  renderers [
   {rendererClass=&NiagaraMeshRendererProperties, rendererIndex=0}
   {rendererClass=&NiagaraSpriteRendererProperties, rendererIndex=1}
  ]
 }
 NE_VC_DriftTrail_CURVE_02 {
  bEnabled=false
  emitterUpdateScript [
   &EmitterState(Life Cycle Mode=::NewEnumerator1, Loop Duration=0.12, Recalculate Duration Each Loop=true, UseLoopDelay=false)
   &SpawnBurst_Instantaneous(
    Spawn Count=&MakeCustomIntFromBool(Bool=$User.bIsFrontendPreview, True INT=4, False INT=0)
    Use Spawn Probability=false
    UseLoopCountLimit=false
   )
   &SpawnRate(SpawnRate=20, Use Spawn Probability=false) {enabled=false}
  ]
  particleSpawnScript [
   &VC_PaintedVariantsOverride(
    Black=<rgba 0.01 0.01 0.01 1>
    BurntSienna=<rgba 0.8 0.1693 0 1>
    Cobalt=<rgba 0 0.0567 1.7 1>
    DefaultColor=&SelectLinearColorFromArray(
     Array Sampling Mode=::NewEnumerator0
     Direct Array Index=?
     Normalized Interpolation=false
     Color Selection Array=DataInterface() {propertyValues {ColorData [<rgba 0 0.42 1.5 1>, <rgba 0.5 0 0.4333 1>]}}
     Random Seed=0
    )
    ForestGreen=<rgba 0 0.5 0 1>
    Grey=<rgba 0.099 0.099 0.099 1>
    Lime=<rgba 0.25 0.6 0 1>
    Orange=<rgba 1.2 0.08 0 1>
    Pink=<rgba 0.7 0.0839 0.4759 1>
    Purple=<rgba 0.1 0 0.6 10>
    Saffron=<rgba 0.7 0.7 0 1>
    SkyBlue=<rgba 0 0.5323 0.8 1>
   )
   &InitializeParticle(
    Color=$Particles.Color
    Lifetime=0.3
    Mesh Scale=<4 5 5>
    Mesh Scale Mode=::NewEnumerator4
    Sprite Rotation Mode=::NewEnumerator2
    Sprite Size Mode=::NewEnumerator3
    UsePositionOffset=false
    Write Material Random=true
   )
   &PartitionParticles()
   &SkeletalMeshLocation(
    Bone / Socket Index=&Add_Integer(A=$Particles.PartitionParticles.Partition, B=4)
    Bone / Socket Sampling Mode=::NewEnumerator5
    Interpolate Mesh Sampling=false
    MaskSpawnGroup=false
    Mesh Sampling Type=::NewEnumerator6
    Sampled Position Offset=&AddVectorToPosition(Position=<10 0 0>, Vector=$User.offset)
    Skeletal Mesh=$System.SkeletalMesh
   )
   &AddVelocity(Velocity=&RandomRangeVector(Minimum=<1000 0 0>, Maximum=<1000 0 0>), Velocity Speed Scale=-1)
   &InitialMeshOrientation(
    Mesh Axis To Orient=&RandomRangeVector(Minimum=<0 -1 -1>, Maximum=<0 1 1>)
    Mesh Orientation Mode=::NewEnumerator2
    Orientation Vector=&RandomRangeVector(Minimum=<0 1 1>, Maximum=<0 -1 -1>)
    Rotation Coordinate Space=::NewEnumerator0
    Use Rotation Vector=true
   )
  ]
  particleUpdateScript [
   &ParticleState()
   &ScaleMeshSize(Scale Factor=&VectorFromCurve(
     VectorCurve=DataInterface() {propertyValues {
       XCurve {keys [<at 0.1 1 RCIM_Cubic>,<at 1 0.2 RCIM_Cubic>]}
       YCurve {keys [<at 0.1 1 RCIM_Cubic>,<at 1 0.2 RCIM_Cubic>]}
       ZCurve {keys [<at 0.1 1 RCIM_Cubic>,<at 1 0.2 RCIM_Cubic>]}
      }}
     Scale Curve=&SelectVectorFromArray(
      Direct Array Index=$Particles.PartitionParticles.Partition
      Vector Selection Array=DataInterface() {propertyValues {FloatData [<1 -0.75 1>, <1 0.75 1>]}}
     )
    ))
   &MeshRotationForce(
    Coordinate Space:ENiagaraMeshTransforms=::NewEnumerator2
    Rotation=&SelectVectorFromArray(
     Direct Array Index=$Particles.PartitionParticles.Partition
     Vector Selection Array=DataInterface() {propertyValues {FloatData [<100 0 0>, <-100 0 0>]}}
    )
   )
   &Drag(
    Drag=&FloatFromCurve(
     FloatCurve=DataInterface() {propertyValues {Curve {keys [{interpMode=RCIM_Cubic, tangentMode=RCTM_User, time=0, value=0}, {interpMode=RCIM_Cubic, tangentMode=RCTM_User, time=1, value=1}]}}}
     Scale Curve=2
    )
    Use Linear Drag=true
    Use Rotational Drag=true
   ) {enabled=false}
   &SolveForcesAndVelocity(Clamp Velocity=false, Limit Acceleration=false)
   &DynamicMaterialParameters(
    Index 0 Param 1=&FloatFromCurve(FloatCurve=DataInterface() {propertyValues {Curve {keys [
         <at 0.1 0 RCIM_Cubic>
         {arriveTangent=0.002, interpMode=RCIM_Cubic, leaveTangent=0.7645, tangentMode=RCTM_Break, time=0.6, value=0.2}
         {arriveTangent=0.644, tangentMode=RCTM_User, time=1, value=5}
        ]}}})
    Index 0 Param 2=&FloatFromCurve(
     FloatCurve=DataInterface() {propertyValues {Curve {keys [
         <at 0 1 RCIM_Cubic>
         {arriveTangent=0.1281, interpMode=RCIM_Cubic, leaveTangent=0.1281, tangentMode=RCTM_User, time=0.5, value=0.4}
         <at 1 0 RCIM_Cubic>
        ]}}}
     Scale Curve=0.5
    )
    Index 1 Param 1=&MakeCustomFloatFromBool(Bool=$User.bIsPainted, True Float=0, False Float=1)
    Module.Index0Param0Write=true
    Module.Index0Param1Write=true
    Module.Index0Param2Write=true
    Module.Index0Param3Write=true
    Module.Index1Param0Write=true
    Module.Index1Param1Write=true
    Module.Index1Param2Write=true
    Module.Index1Param3Write=true
    Write Parameter Index 1=true
   )
   &ScaleColor(
    Scale RGB=&VectorFromFloat(Value=&MakeCustomFloatFromBool(
      Bool=$User.bIsPainted
      True Float=&FloatFromCurve(
       FloatCurve=DataInterface() {propertyValues {Curve {keys [{interpMode=RCIM_Cubic, leaveTangent=-2.9826, tangentMode=RCTM_Break, time=0, value=1}, <at 0.95 0.15 RCIM_Cubic>]}}}
       Scale Curve=3.5
      )
      False Float=&FloatFromCurve(
       FloatCurve=DataInterface() {propertyValues {Curve {keys [{arriveTangent=-5.3196, interpMode=RCIM_Cubic, leaveTangent=-5.3196, tangentMode=RCTM_User, time=0.3, value=2}, <at 1 0.4 RCIM_Cubic>]}}}
       Scale Curve=&RandomRangeFloat(Minimum=3, Maximum=2)
      )
     ))
    ScaleA=true
    ScaleRGB=true
   )
  ]
  rendererClasses [&NiagaraMeshRendererProperties, &NiagaraSpriteRendererProperties]
  renderers [
   {rendererClass=&NiagaraMeshRendererProperties, rendererIndex=0}
   {rendererClass=&NiagaraSpriteRendererProperties, rendererIndex=1}
  ]
 }
 NE_VC_DriftTrail_CURVE_02001 {
  emitterUpdateScript [
   &EmitterState(Life Cycle Mode=::NewEnumerator1, Loop Duration=0.1, UseLoopDelay=false)
   &SpawnBurst_Instantaneous(Spawn Count=5, Use Spawn Probability=false, UseLoopCountLimit=false)
   &SpawnRate(
    SpawnRate=&MakeCustomFloatFromBool(Bool=$User.bIsFrontendPreview, True Float=0, False Float=50)
    Use Spawn Probability=false
   )
   /SetVariables(Emitter.IS BR=HlslExpression() {hlslExpression="User.GameModeID == 1 ? true : false"}) {isSetParameters=true}
  ]
  particleSpawnScript [
   &VC_PaintedVariantsOverride(
    Black=<rgba 0.01 0.01 0.01 1>
    BurntSienna=<rgba 0.25 0.0625 0 1>
    Cobalt=<rgba 0 0.0567 1.7 1>
    DefaultColor=&SelectLinearColorFromArray(
     Array Sampling Mode=::NewEnumerator0
     Direct Array Index=?
     Normalized Interpolation=false
     Color Selection Array=DataInterface() {propertyValues {ColorData [<rgba 0 1.3333 4 1>, <rgba 1.2 0 0.3 1>]}}
     Random Seed=0
    )
    ForestGreen=<rgba 0 0.5 0 1>
    Grey=<rgba 0.099 0.099 0.099 1>
    Lime=<rgba 0.25 0.6 0 1>
    Orange=<rgba 1.2 0.08 0 1>
    Pink=<rgba 0.7 0.0839 0.4759 1>
    Saffron=<rgba 0.7 0.7 0 1>
    SkyBlue=<rgba 0 0.5323 0.8 1>
   )
   &InitializeParticle(
    Color=$Particles.Color
    Lifetime=0.35
    Mesh Scale Max=<5 5 5>
    Mesh Scale Min=<5 8 8>
    Mesh Scale Mode=::NewEnumerator2
    Sprite Rotation Mode=::NewEnumerator2
    Sprite Size Mode=::NewEnumerator3
    UsePositionOffset=false
    Write Material Random=true
   )
   &PartitionParticles()
   &SkeletalMeshLocation(
    Bone / Socket Index=&Add_Integer(A=$Particles.PartitionParticles.Partition, B=4)
    Bone / Socket Sampling Mode=::NewEnumerator5
    Interpolate Mesh Sampling=false
    MaskSpawnGroup=false
    Mesh Sampling Type=::NewEnumerator6
    Sampled Position Offset=&AddVectorToPosition(Position=<0 0 10>, Vector=$User.offset)
    Skeletal Mesh=$System.SkeletalMesh
   )
   &AddVelocity(
    Velocity=&SelectVectorFromArray(
     Array Sampling Mode=::NewEnumerator0
     Direct Array Index=?
     Vector Selection Array=DataInterface() {propertyValues {FloatData [<1250 0 0>, <1250 1 1>]}}
     Random Seed=0
    )
    Velocity Speed Scale=-1
   )
   &InitialMeshOrientation(
    Mesh Axis To Orient=&RandomRangeVector(Minimum=<0 -1 -1>, Maximum=<0 1 1>)
    Mesh Orientation Mode=::NewEnumerator2
    Orientation Vector=&RandomRangeVector(Minimum=<0 1 1>, Maximum=<0 -1 -1>)
    Rotation Coordinate Space=::NewEnumerator0
    Use Rotation Vector=false
   )
  ]
  particleUpdateScript [
   &ParticleState()
   &ScaleMeshSize(Scale Factor=&VectorFromCurve(
     VectorCurve=DataInterface() {propertyValues {
       XCurve {keys [{arriveTangent=-1.7724, interpMode=RCIM_Cubic, leaveTangent=-1.7724, tangentMode=RCTM_User, time=0, value=1}, <at 1 0 RCIM_Cubic>]}
       YCurve {keys [{arriveTangent=-1.7724, interpMode=RCIM_Cubic, leaveTangent=-1.7724, tangentMode=RCTM_User, time=0, value=1}, <at 1 0 RCIM_Cubic>]}
       ZCurve {keys [{arriveTangent=-1.7724, interpMode=RCIM_Cubic, leaveTangent=-1.7724, tangentMode=RCTM_User, time=0, value=1}, <at 1 0 RCIM_Cubic>]}
      }}
     Scale Curve=&SelectVectorFromArray(
      Direct Array Index=$Particles.PartitionParticles.Partition
      Vector Selection Array=DataInterface() {propertyValues {FloatData [<1 0.75 0.75>, <1 -0.75 0.75>]}}
     )
    ))
   &MeshRotationForce(
    Coordinate Space:ENiagaraMeshTransforms=::NewEnumerator2
    Rotation=&SelectVectorFromArray(
     Direct Array Index=$Particles.PartitionParticles.Partition
     Vector Selection Array=DataInterface() {propertyValues {FloatData [<250 0 0>, <-250 0 0>]}}
    )
   )
   &InheritSourceMovement(
    Applied Position Scale=&Multiply_VectorByFloat(Vector=<1 1 1>, Float=1)
    ApplyForce=false
    ApplyPosition=true
    ApplyPositionFalloff=false
   )
   &SolveForcesAndVelocity(Clamp Velocity=false, Limit Acceleration=false)
   &DynamicMaterialParameters(
    Index 0 Param 1=&FloatFromCurve(
     FloatCurve=DataInterface() {propertyValues {Curve {keys [
         <at 0.2 0 RCIM_Cubic>
         {arriveTangent=0.002, interpMode=RCIM_Cubic, leaveTangent=0.7645, tangentMode=RCTM_Break, time=0.6, value=0.2}
         {arriveTangent=0.644, tangentMode=RCTM_User, time=1, value=5}
        ]}}}
     Scale Curve=1.75
    )
    Index 0 Param 2=&FloatFromCurve(
     FloatCurve=DataInterface() {propertyValues {Curve {keys [
         <at 0 1 RCIM_Cubic>
         {arriveTangent=0.1281, interpMode=RCIM_Cubic, leaveTangent=0.1281, tangentMode=RCTM_User, time=0.5, value=0.25}
         <at 1 0 RCIM_Cubic>
        ]}}}
     Scale Curve=0.5
    )
    Index 1 Param 1=&MakeCustomFloatFromBool(Bool=$User.bIsPainted, True Float=0, False Float=1)
    Index 1 Param 2=&MakeCustomFloatFromBool(Bool=$User.bIsPainted, True Float=&MakeFloatFromInt(Int=$User.Lines_Glow), False Float=100)
    Module.Index0Param0Write=true
    Module.Index0Param1Write=true
    Module.Index0Param2Write=true
    Module.Index0Param3Write=true
    Module.Index1Param0Write=true
    Module.Index1Param1Write=true
    Module.Index1Param2Write=true
    Module.Index1Param3Write=true
    Write Parameter Index 1=true
   )
   &ScaleColor(
    Scale RGB=&VectorFromFloat(Value=&MakeCustomFloatFromBool(
      Bool=$User.bIsPainted
      True Float=&FloatFromCurve(
       FloatCurve=DataInterface() {propertyValues {Curve {keys [{interpMode=RCIM_Cubic, leaveTangent=-3.8527, tangentMode=RCTM_Break, time=0, value=1}, <at 0.8 0.1 RCIM_Cubic>]}}}
       Scale Curve=3.5
      )
      False Float=&FloatFromCurve(
       FloatCurve=DataInterface() {propertyValues {Curve {keys [{arriveTangent=-8.0319, interpMode=RCIM_Cubic, leaveTangent=-8.0319, tangentMode=RCTM_User, time=0.3, value=2}, <at 1 0.4 RCIM_Cubic>]}}}
       Scale Curve=&RandomRangeFloat(Minimum=3, Maximum=3, Evaluation Type=::NewEnumerator1)
      )
     ))
    ScaleA=true
    ScaleRGB=true
   )
  ]
  rendererClasses [&NiagaraMeshRendererProperties, &NiagaraSpriteRendererProperties]
  renderers [
   {rendererClass=&NiagaraMeshRendererProperties, rendererIndex=0}
   {rendererClass=&NiagaraSpriteRendererProperties, rendererIndex=1}
  ]
 }
 NE_VC_DriftTrail_CURVE_02002 {
  bEnabled=false
  emitterUpdateScript [
   &EmitterState(Life Cycle Mode=::NewEnumerator1, Loop Duration=0.1, Recalculate Duration Each Loop=true, UseLoopDelay=false)
   &SpawnBurst_Instantaneous(
    Spawn Count=&MakeCustomIntFromBool(Bool=$User.bIsFrontendPreview, True INT=0, False INT=0)
    Use Spawn Probability=false
    UseLoopCountLimit=false
   )
   &SpawnRate(
    SpawnRate=&MakeCustomFloatFromBool(Bool=$User.bIsFrontendPreview, True Float=0, False Float=50)
    Use Spawn Probability=false
   )
  ]
  particleSpawnScript [
   &VC_PaintedVariantsOverride(
    Black=<rgba 0.01 0.01 0.01 1>
    BurntSienna=<rgba 0.8 0.1693 0 1>
    Cobalt=<rgba 0 0.0567 1.7 1>
    DefaultColor=&SelectLinearColorFromArray(
     Array Sampling Mode=::NewEnumerator0
     Direct Array Index=?
     Normalized Interpolation=false
     Color Selection Array=DataInterface() {propertyValues {ColorData [<rgba 0 0.42 1.5 1>, <rgba 0.5 0 0.4333 1>]}}
     Random Seed=0
    )
    ForestGreen=<rgba 0 0.5 0 1>
    Grey=<rgba 0.099 0.099 0.099 1>
    Lime=<rgba 0.25 0.6 0 1>
    Orange=<rgba 1.2 0.08 0 1>
    Pink=<rgba 0.7 0.0839 0.4759 1>
    Purple=<rgba 0.1 0 0.6 10>
    Saffron=<rgba 0.7 0.7 0 1>
    SkyBlue=<rgba 0 0.5323 0.8 1>
   )
   &InitializeParticle(
    Color=$Particles.Color
    Lifetime=0.3
    Mesh Scale Max=<5 5 5>
    Mesh Scale Min=<5 5 5>
    Mesh Scale Mode=::NewEnumerator2
    Sprite Rotation Mode=::NewEnumerator2
    Sprite Size Mode=::NewEnumerator3
    UsePositionOffset=false
    Write Material Random=true
   )
   &PartitionParticles()
   &SkeletalMeshLocation(
    Bone / Socket Index=&Add_Integer(A=$Particles.PartitionParticles.Partition, B=4)
    Bone / Socket Sampling Mode=::NewEnumerator5
    Interpolate Mesh Sampling=false
    MaskSpawnGroup=false
    Mesh Sampling Type=::NewEnumerator6
    Sampled Position Offset=&AddVectorToPosition(Position=<-30 0 0>, Vector=$User.offset)
    Skeletal Mesh=$System.SkeletalMesh
   )
   &AddVelocity(Velocity=&RandomRangeVector(Minimum=<1300 0 0>, Maximum=<1300 0 0>), Velocity Speed Scale=-1)
   &InitialMeshOrientation(
    Mesh Axis To Orient=&RandomRangeVector(Minimum=<0 -1 -1>, Maximum=<0 1 1>)
    Mesh Orientation Mode=::NewEnumerator2
    Orientation Vector=&RandomRangeVector(Minimum=<0 1 1>, Maximum=<0 -1 -1>)
    Rotation Coordinate Space=::NewEnumerator2
    Use Rotation Vector=false
   )
  ]
  particleUpdateScript [
   &ParticleState()
   &ScaleMeshSize(Scale Factor=&VectorFromCurve(
     VectorCurve=DataInterface() {propertyValues {
       XCurve {keys [
         <at 0.2 1 RCIM_Cubic>
         {arriveTangentWeight=0.3735, interpMode=RCIM_Cubic, leaveTangentWeight=0.0534, tangentMode=RCTM_User, time=0.9, value=0.2}
         <at 1 0 RCIM_Cubic>
        ]}
       YCurve {keys [
         <at 0.2 1 RCIM_Cubic>
         {arriveTangentWeight=0.3065, interpMode=RCIM_Cubic, leaveTangentWeight=0.0438, tangentMode=RCTM_User, time=0.9, value=0.2}
         <at 1 0 RCIM_Cubic>
        ]}
       ZCurve {keys [
         <at 0.2 1 RCIM_Cubic>
         {arriveTangentWeight=0.3735, interpMode=RCIM_Cubic, leaveTangentWeight=0.0534, tangentMode=RCTM_User, time=0.9, value=0.2}
         <at 1 0 RCIM_Cubic>
        ]}
      }}
     Scale Curve=<1 0.75 0.75>
    ))
   &MeshRotationForce(
    Coordinate Space:ENiagaraMeshTransforms=::NewEnumerator2
    Rotation=&SelectVectorFromArray(
     Direct Array Index=$Particles.PartitionParticles.Partition
     Vector Selection Array=DataInterface() {propertyValues {FloatData [<500 0 0>, <-500 1 1>]}}
    )
   )
   &Drag(Use Linear Drag=true, Use Rotational Drag=true)
   &SolveForcesAndVelocity(Clamp Velocity=false, Limit Acceleration=false)
   &DynamicMaterialParameters(
    Index 0 Param 1=&FloatFromCurve(
     FloatCurve=DataInterface() {propertyValues {Curve {keys [
         <at 0.2 0 RCIM_Cubic>
         {arriveTangent=0.002, interpMode=RCIM_Cubic, leaveTangent=0.7645, tangentMode=RCTM_Break, time=0.6, value=0.2}
         {arriveTangent=0.644, tangentMode=RCTM_User, time=1, value=5}
        ]}}}
     Scale Curve=1.75
    )
    Index 0 Param 2=&FloatFromCurve(
     FloatCurve=DataInterface() {propertyValues {Curve {keys [
         <at 0 1 RCIM_Cubic>
         {arriveTangent=0.1281, interpMode=RCIM_Cubic, leaveTangent=0.1281, tangentMode=RCTM_User, time=0.5, value=0.4}
         <at 1 0 RCIM_Cubic>
        ]}}}
     Scale Curve=0.5
    )
    Index 1 Param 1=&MakeCustomFloatFromBool(Bool=$User.bIsPainted, True Float=0, False Float=1)
    Module.Index0Param0Write=true
    Module.Index0Param1Write=true
    Module.Index0Param2Write=true
    Module.Index0Param3Write=true
    Module.Index1Param0Write=true
    Module.Index1Param1Write=true
    Module.Index1Param2Write=true
    Module.Index1Param3Write=true
    Write Parameter Index 1=true
   )
   &ScaleColor(
    Scale RGB=&VectorFromFloat(Value=&MakeCustomFloatFromBool(
      Bool=$User.bIsPainted
      True Float=&FloatFromCurve(
       FloatCurve=DataInterface() {propertyValues {Curve {keys [{interpMode=RCIM_Cubic, leaveTangent=-3.8527, tangentMode=RCTM_Break, time=0, value=1}, <at 0.8 0.1 RCIM_Cubic>]}}}
       Scale Curve=3.5
      )
      False Float=&FloatFromCurve(
       FloatCurve=DataInterface() {propertyValues {Curve {keys [{arriveTangent=-8.0319, interpMode=RCIM_Cubic, leaveTangent=-8.0319, tangentMode=RCTM_User, time=0.3, value=2}, <at 1 0.4 RCIM_Cubic>]}}}
       Scale Curve=&RandomRangeFloat(Minimum=3, Maximum=3)
      )
     ))
    ScaleA=true
    ScaleRGB=true
   )
  ]
  rendererClasses [&NiagaraMeshRendererProperties, &NiagaraSpriteRendererProperties]
  renderers [
   {rendererClass=&NiagaraMeshRendererProperties, rendererIndex=0}
   {rendererClass=&NiagaraSpriteRendererProperties, rendererIndex=1}
  ]
 }
 NE_VC_DriftTrail_CURVE_02003 {
  bEnabled=false
  emitterUpdateScript [
   &EmitterState(Life Cycle Mode=::NewEnumerator1, Loop Duration=0.12, Recalculate Duration Each Loop=true, UseLoopDelay=false)
   &SpawnBurst_Instantaneous(
    Spawn Count=&MakeCustomIntFromBool(Bool=$User.bIsFrontendPreview, True INT=4, False INT=0)
    Use Spawn Probability=false
    UseLoopCountLimit=false
   )
   &SpawnRate(SpawnRate=20, Use Spawn Probability=false) {enabled=false}
  ]
  particleSpawnScript [
   &VC_PaintedVariantsOverride(
    Black=<rgba 0.01 0.01 0.01 1>
    BurntSienna=<rgba 0.8 0.1693 0 1>
    Cobalt=<rgba 0 0.0567 1.7 1>
    DefaultColor=&SelectLinearColorFromArray(
     Array Sampling Mode=::NewEnumerator0
     Direct Array Index=?
     Normalized Interpolation=false
     Color Selection Array=DataInterface() {propertyValues {ColorData [<rgba 0 0.42 1.5 1>, <rgba 0.5 0 0.4333 1>]}}
     Random Seed=0
    )
    ForestGreen=<rgba 0 0.5 0 1>
    Grey=<rgba 0.099 0.099 0.099 1>
    Lime=<rgba 0.25 0.6 0 1>
    Orange=<rgba 1.2 0.08 0 1>
    Pink=<rgba 0.7 0.0839 0.4759 1>
    Purple=<rgba 0.1 0 0.6 10>
    Saffron=<rgba 0.7 0.7 0 1>
    SkyBlue=<rgba 0 0.5323 0.8 1>
   )
   &InitializeParticle(
    Color=$Particles.Color
    Lifetime=0.3
    Mesh Scale=<5 5 5>
    Mesh Scale Mode=::NewEnumerator4
    Sprite Rotation Mode=::NewEnumerator2
    Sprite Size Mode=::NewEnumerator3
    UsePositionOffset=false
    Write Material Random=true
   )
   &PartitionParticles()
   &SkeletalMeshLocation(
    Bone / Socket Index=&Add_Integer(A=$Particles.PartitionParticles.Partition, B=4)
    Bone / Socket Sampling Mode=::NewEnumerator5
    Interpolate Mesh Sampling=false
    MaskSpawnGroup=false
    Mesh Sampling Type=::NewEnumerator6
    Sampled Position Offset=&AddVectorToPosition(Position=<0 0 0>, Vector=$User.offset)
    Skeletal Mesh=$System.SkeletalMesh
   )
   &AddVelocity(Velocity=&RandomRangeVector(Minimum=<1000 0 0>, Maximum=<1000 0 0>), Velocity Speed Scale=-1)
   &InitialMeshOrientation(
    Mesh Axis To Orient=&RandomRangeVector(Minimum=<0 -1 -1>, Maximum=<0 1 1>)
    Mesh Orientation Mode=::NewEnumerator2
    Orientation Vector=&RandomRangeVector(Minimum=<0 1 1>, Maximum=<0 -1 -1>)
    Rotation Coordinate Space=::NewEnumerator0
    Use Rotation Vector=false
   )
  ]
  particleUpdateScript [
   &ParticleState()
   &ScaleMeshSize(Scale Factor=&VectorFromCurve(
     VectorCurve=DataInterface() {propertyValues {
       XCurve {keys [<at 0 1 RCIM_Cubic>, {arriveTangent=-0.4, interpMode=RCIM_Cubic, leaveTangent=-0.4, time=0.8, value=0.65}, <at 1 0.6 RCIM_Cubic>]}
       YCurve {keys [
         <at 0 0.7 RCIM_Cubic>
         {arriveTangentWeight=0.3065, interpMode=RCIM_Cubic, leaveTangentWeight=0.0438, tangentMode=RCTM_User, time=0.8, value=0.7}
         <at 1 0.4 RCIM_Cubic>
        ]}
       ZCurve {keys [
         <at 0 0.7 RCIM_Cubic>
         {arriveTangentWeight=0.3735, interpMode=RCIM_Cubic, leaveTangentWeight=0.0534, tangentMode=RCTM_User, time=0.8, value=0.7}
         <at 1 0.4 RCIM_Cubic>
        ]}
      }}
     Scale Curve=&SelectVectorFromArray(
      Direct Array Index=$Particles.PartitionParticles.Partition
      Vector Selection Array=DataInterface() {propertyValues {FloatData [<0.75 -0.75 1>, <0.75 0.75 1>]}}
     )
    ))
   &MeshRotationForce(
    Coordinate Space:ENiagaraMeshTransforms=::NewEnumerator2
    Rotation=&SelectVectorFromArray(
     Direct Array Index=$Particles.PartitionParticles.Partition
     Vector Selection Array=DataInterface() {propertyValues {FloatData [<50 0 0>, <-50 0 0>]}}
    )
   ) {enabled=false}
   &Drag(
    Drag=&FloatFromCurve(
     FloatCurve=DataInterface() {propertyValues {Curve {keys [{interpMode=RCIM_Cubic, tangentMode=RCTM_User, time=0.2, value=0}, {interpMode=RCIM_Cubic, tangentMode=RCTM_User, time=0.5, value=2}]}}}
     Scale Curve=6
    )
    Use Linear Drag=true
    Use Rotational Drag=true
   )
   &SolveForcesAndVelocity(Clamp Velocity=false, Limit Acceleration=false)
   &DynamicMaterialParameters(
    Index 0 Param 1=&FloatFromCurve(FloatCurve=DataInterface() {propertyValues {Curve {keys [
         <at 0.2 0 RCIM_Cubic>
         {arriveTangent=0.002, interpMode=RCIM_Cubic, leaveTangent=0.7645, tangentMode=RCTM_Break, time=0.6, value=0.1}
         {arriveTangent=1.75, interpMode=RCIM_Cubic, leaveTangent=1.75, time=0.8, value=0.3}
         {arriveTangent=0.644, tangentMode=RCTM_User, time=1, value=0.8}
        ]}}})
    Index 0 Param 2=&FloatFromCurve(FloatCurve=DataInterface() {propertyValues {Curve {keys [<at 0 1 RCIM_Cubic>,<at 1 0 RCIM_Cubic>]}}}, Scale Curve=0.5)
    Index 1 Param 1=&MakeCustomFloatFromBool(Bool=$User.bIsPainted, True Float=0, False Float=1)
    Module.Index0Param0Write=true
    Module.Index0Param1Write=true
    Module.Index0Param2Write=true
    Module.Index0Param3Write=true
    Module.Index1Param0Write=true
    Module.Index1Param1Write=true
    Module.Index1Param2Write=true
    Module.Index1Param3Write=true
    Write Parameter Index 1=true
   )
   &ScaleColor(
    Scale RGB=&VectorFromFloat(Value=&MakeCustomFloatFromBool(
      Bool=$User.bIsPainted
      True Float=&FloatFromCurve(
       FloatCurve=DataInterface() {propertyValues {Curve {keys [{interpMode=RCIM_Cubic, leaveTangent=-2.9826, tangentMode=RCTM_Break, time=0, value=1}, <at 0.95 0.15 RCIM_Cubic>]}}}
       Scale Curve=3.5
      )
      False Float=&FloatFromCurve(
       FloatCurve=DataInterface() {propertyValues {Curve {keys [{arriveTangent=-6.2924, interpMode=RCIM_Cubic, leaveTangent=-6.2924, tangentMode=RCTM_User, time=0.3, value=2}, <at 1 0.4 RCIM_Cubic>]}}}
       Scale Curve=&RandomRangeFloat(Minimum=4, Maximum=4)
      )
     ))
    ScaleA=true
    ScaleRGB=true
   )
  ]
  rendererClasses [&NiagaraMeshRendererProperties, &NiagaraSpriteRendererProperties]
  renderers [
   {rendererClass=&NiagaraMeshRendererProperties, rendererIndex=0}
   {rendererClass=&NiagaraSpriteRendererProperties, rendererIndex=1}
  ]
 }
 NE_VC_DriftTrail_CURVE_02004 {
  emitterUpdateScript [
   &EmitterState(Life Cycle Mode=::NewEnumerator1, Loop Duration=0.1, Recalculate Duration Each Loop=true, UseLoopDelay=false)
   &SpawnBurst_Instantaneous(
    Spawn Count=&MakeCustomIntFromBool(Bool=$User.bIsFrontendPreview, True INT=0, False INT=0)
    Use Spawn Probability=false
    UseLoopCountLimit=false
   ) {enabled=false}
   &SpawnRate(SpawnRate=50, Use Spawn Probability=false)
   /SetVariables(Emitter.IS RR=HlslExpression() {hlslExpression="User.GameModeID == 2 ? true : false"}) {isSetParameters=true}
  ]
  particleSpawnScript [
   &VC_PaintedVariantsOverride(
    Black=<rgba 0.01 0.01 0.01 1>
    BurntSienna=<rgba 0.25 0.0625 0 1>
    Cobalt=<rgba 0 0.0567 1.7 1>
    DefaultColor=&SelectLinearColorFromArray(
     Array Sampling Mode=::NewEnumerator0
     Direct Array Index=?
     Normalized Interpolation=false
     Color Selection Array=DataInterface() {propertyValues {ColorData [<rgba 0 1.3333 4 1>, <rgba 1.2 0 0.3 1>]}}
     Random Seed=0
    )
    ForestGreen=<rgba 0 0.5 0 1>
    Grey=<rgba 0.099 0.099 0.099 1>
    Lime=<rgba 0.25 0.6 0 1>
    Orange=<rgba 1.2 0.08 0 1>
    Pink=<rgba 0.7 0.0839 0.4759 1>
    Saffron=<rgba 0.7 0.7 0 1>
    SkyBlue=<rgba 0 0.5323 0.8 1>
   )
   &InitializeParticle(
    Color=$Particles.Color
    Lifetime=0.3
    Mesh Scale Max=<5 5 5>
    Mesh Scale Min=<5 5 5>
    Mesh Scale Mode=::NewEnumerator2
    Sprite Rotation Mode=::NewEnumerator2
    Sprite Size Mode=::NewEnumerator3
    UsePositionOffset=false
    Write Material Random=true
   )
   &PartitionParticles()
   &SkeletalMeshLocation(
    Bone / Socket Index=&Add_Integer(A=$Particles.PartitionParticles.Partition, B=4)
    Bone / Socket Sampling Mode=::NewEnumerator5
    Interpolate Mesh Sampling=false
    MaskSpawnGroup=false
    Mesh Sampling Type=::NewEnumerator6
    Sampled Position Offset=&AddVectorToPosition(Position=<-30 0 0>, Vector=$User.offset)
    Skeletal Mesh=$System.SkeletalMesh
   )
   &AddVelocity(Velocity=&RandomRangeVector(Minimum=<1300 0 0>, Maximum=<1300 0 0>), Velocity Speed Scale=-1)
   &InitialMeshOrientation(
    Mesh Axis To Orient=&RandomRangeVector(Minimum=<0 -1 -1>, Maximum=<0 1 1>)
    Mesh Orientation Mode=::NewEnumerator2
    Orientation Vector=&RandomRangeVector(Minimum=<0 1 1>, Maximum=<0 -1 -1>)
    Rotation Coordinate Space=::NewEnumerator2
    Use Rotation Vector=false
   )
  ]
  particleUpdateScript [
   &ParticleState()
   &ScaleMeshSize(Scale Factor=&VectorFromCurve(
     VectorCurve=DataInterface() {propertyValues {
       XCurve {keys [
         <at 0.2 1 RCIM_Cubic>
         {arriveTangentWeight=0.3735, interpMode=RCIM_Cubic, leaveTangentWeight=0.0534, tangentMode=RCTM_User, time=0.9, value=0.2}
         <at 1 0 RCIM_Cubic>
        ]}
       YCurve {keys [
         <at 0.2 1 RCIM_Cubic>
         {arriveTangentWeight=0.3065, interpMode=RCIM_Cubic, leaveTangentWeight=0.0438, tangentMode=RCTM_User, time=0.9, value=0.2}
         <at 1 0 RCIM_Cubic>
        ]}
       ZCurve {keys [
         <at 0.2 1 RCIM_Cubic>
         {arriveTangentWeight=0.3735, interpMode=RCIM_Cubic, leaveTangentWeight=0.0534, tangentMode=RCTM_User, time=0.9, value=0.2}
         <at 1 0 RCIM_Cubic>
        ]}
      }}
     Scale Curve=<1 0.75 0.75>
    ))
   &MeshRotationForce(
    Coordinate Space:ENiagaraMeshTransforms=::NewEnumerator2
    Rotation=&SelectVectorFromArray(
     Direct Array Index=$Particles.PartitionParticles.Partition
     Vector Selection Array=DataInterface() {propertyValues {FloatData [<500 0 0>, <-500 1 1>]}}
    )
   )
   &Drag(Use Linear Drag=true, Use Rotational Drag=true)
   &SolveForcesAndVelocity(Clamp Velocity=false, Limit Acceleration=false)
   &DynamicMaterialParameters(
    Index 0 Param 1=&FloatFromCurve(
     FloatCurve=DataInterface() {propertyValues {Curve {keys [
         <at 0.2 0 RCIM_Cubic>
         {arriveTangent=0.002, interpMode=RCIM_Cubic, leaveTangent=0.7645, tangentMode=RCTM_Break, time=0.6, value=0.2}
         {arriveTangent=0.644, tangentMode=RCTM_User, time=1, value=5}
        ]}}}
     Scale Curve=1.75
    )
    Index 0 Param 2=&FloatFromCurve(
     FloatCurve=DataInterface() {propertyValues {Curve {keys [
         <at 0 1 RCIM_Cubic>
         {arriveTangent=0.1281, interpMode=RCIM_Cubic, leaveTangent=0.1281, tangentMode=RCTM_User, time=0.5, value=0.4}
         <at 1 0 RCIM_Cubic>
        ]}}}
     Scale Curve=0.5
    )
    Index 1 Param 1=&MakeCustomFloatFromBool(Bool=$User.bIsPainted, True Float=0, False Float=1)
    Index 1 Param 2=&MakeCustomFloatFromBool(Bool=$User.bIsPainted, True Float=&MakeFloatFromInt(Int=$User.Lines_Glow), False Float=100)
    Module.Index0Param0Write=true
    Module.Index0Param1Write=true
    Module.Index0Param2Write=true
    Module.Index0Param3Write=true
    Module.Index1Param0Write=true
    Module.Index1Param1Write=true
    Module.Index1Param2Write=true
    Module.Index1Param3Write=true
    Write Parameter Index 1=true
   )
   &ScaleColor(
    Scale RGB=&VectorFromFloat(Value=&MakeCustomFloatFromBool(
      Bool=$User.bIsPainted
      True Float=&FloatFromCurve(
       FloatCurve=DataInterface() {propertyValues {Curve {keys [{interpMode=RCIM_Cubic, leaveTangent=-3.8527, tangentMode=RCTM_Break, time=0, value=1}, <at 0.8 0.1 RCIM_Cubic>]}}}
       Scale Curve=3.5
      )
      False Float=&FloatFromCurve(
       FloatCurve=DataInterface() {propertyValues {Curve {keys [{arriveTangent=-8.0319, interpMode=RCIM_Cubic, leaveTangent=-8.0319, tangentMode=RCTM_User, time=0.3, value=2}, <at 1 0.4 RCIM_Cubic>]}}}
       Scale Curve=&RandomRangeFloat(Minimum=3, Maximum=3)
      )
     ))
    ScaleA=true
    ScaleRGB=true
   )
  ]
  rendererClasses [&NiagaraMeshRendererProperties, &NiagaraSpriteRendererProperties]
  renderers [
   {rendererClass=&NiagaraMeshRendererProperties, rendererIndex=0}
   {rendererClass=&NiagaraSpriteRendererProperties, rendererIndex=1}
  ]
 }
 NE_VC_DriftTrail_CURVE_02005 {
  emitterUpdateScript [
   &EmitterState(Life Cycle Mode=::NewEnumerator1, Loop Duration=0.09, Recalculate Duration Each Loop=true, UseLoopDelay=false)
   &SpawnBurst_Instantaneous(Spawn Count=4, Use Spawn Probability=false, UseLoopCountLimit=false)
   &SpawnRate(SpawnRate=20, Use Spawn Probability=false) {enabled=false}
   /SetVariables(Emitter.IS FRONTEND=HlslExpression() {hlslExpression="User.GameModeID == 0 ? true : false"}) {isSetParameters=true}
  ]
  particleSpawnScript [
   &VC_PaintedVariantsOverride(
    Black=<rgba 0.01 0.01 0.01 1>
    BurntSienna=<rgba 0.25 0.0625 0 1>
    Cobalt=<rgba 0 0.0567 1.7 1>
    DefaultColor=&SelectLinearColorFromArray(
     Array Sampling Mode=::NewEnumerator0
     Direct Array Index=?
     Normalized Interpolation=false
     Color Selection Array=DataInterface() {propertyValues {ColorData [<rgba 0 1.3333 4 1>, <rgba 1.2 0 0.3 1>]}}
     Random Seed=0
    )
    ForestGreen=<rgba 0 0.5 0 1>
    Grey=<rgba 0.075 0.075 0.075 1>
    Lime=<rgba 0.25 0.6 0 1>
    Orange=<rgba 1.2 0.08 0 1>
    Pink=<rgba 0.7 0.0839 0.4759 1>
    Saffron=<rgba 0.7 0.7 0 1>
    SkyBlue=<rgba 0 0.5323 0.8 1>
   )
   &InitializeParticle(
    Color=$Particles.Color
    Lifetime=0.2
    Mesh Scale=<4 5 5>
    Mesh Scale Mode=::NewEnumerator4
    Sprite Rotation Mode=::NewEnumerator2
    Sprite Size Mode=::NewEnumerator3
    UsePositionOffset=false
    Write Material Random=true
   )
   &PartitionParticles()
   &SkeletalMeshLocation(
    Bone / Socket Index=&Add_Integer(A=$Particles.PartitionParticles.Partition, B=4)
    Bone / Socket Sampling Mode=::NewEnumerator5
    Interpolate Mesh Sampling=false
    MaskSpawnGroup=false
    Mesh Sampling Type=::NewEnumerator6
    Sampled Position Offset=&AddVectorToPosition(Position=<10 0 0>, Vector=$User.offset)
    Skeletal Mesh=$System.SkeletalMesh
   )
   &AddVelocity(Velocity=&RandomRangeVector(Minimum=<1500 0 0>, Maximum=<1500 0 0>), Velocity Speed Scale=-1)
   &InitialMeshOrientation(
    Mesh Axis To Orient=&RandomRangeVector(Minimum=<0 -1 -1>, Maximum=<0 1 1>)
    Mesh Orientation Mode=::NewEnumerator2
    Orientation Vector=&RandomRangeVector(Minimum=<0 1 1>, Maximum=<0 -1 1>)
    Rotation Coordinate Space=::NewEnumerator0
    Use Rotation Vector=true
   )
  ]
  particleUpdateScript [
   &ParticleState()
   &ScaleMeshSize(Scale Factor=&VectorFromCurve(
     VectorCurve=DataInterface() {propertyValues {
       XCurve {keys [
         <at 0.1 1 RCIM_Cubic>
         {arriveTangent=-1.2479, arriveTangentWeight=0.3356, interpMode=RCIM_Cubic, leaveTangent=-1.2479, tangentMode=RCTM_User, time=1, value=0}
        ]}
       YCurve {keys [
         <at 0.1 1 RCIM_Cubic>
         {arriveTangent=-1.2479, arriveTangentWeight=0.3356, interpMode=RCIM_Cubic, leaveTangent=-1.2479, tangentMode=RCTM_User, time=1, value=0}
        ]}
       ZCurve {keys [
         <at 0.1 1 RCIM_Cubic>
         {arriveTangent=-1.2479, arriveTangentWeight=0.3356, interpMode=RCIM_Cubic, leaveTangent=-1.2479, tangentMode=RCTM_User, time=1, value=0}
        ]}
      }}
     Scale Curve=&SelectVectorFromArray(
      Direct Array Index=$Particles.PartitionParticles.Partition
      Vector Selection Array=DataInterface() {propertyValues {FloatData [<1 -0.75 1>, <1 0.75 1>]}}
     )
    ))
   &MeshRotationForce(
    Coordinate Space:ENiagaraMeshTransforms=::NewEnumerator2
    Rotation=&SelectVectorFromArray(
     Direct Array Index=$Particles.PartitionParticles.Partition
     Vector Selection Array=DataInterface() {propertyValues {FloatData [<100 0 0>, <-100 0 0>]}}
    )
   ) {enabled=false}
   &SolveForcesAndVelocity(Clamp Velocity=false, Limit Acceleration=false)
   &DynamicMaterialParameters(
    Index 0 Param 1=&FloatFromCurve(FloatCurve=DataInterface() {propertyValues {Curve {keys [
         <at 0.1 0 RCIM_Cubic>
         {arriveTangent=0.002, interpMode=RCIM_Cubic, leaveTangent=0.7645, tangentMode=RCTM_Break, time=0.6, value=0.2}
         {arriveTangent=0.644, tangentMode=RCTM_User, time=1, value=1}
        ]}}})
    Index 0 Param 2=&FloatFromCurve(
     FloatCurve=DataInterface() {propertyValues {Curve {keys [
         <at 0 1 RCIM_Cubic>
         {arriveTangent=-1.4256, interpMode=RCIM_Cubic, leaveTangent=-1.4256, tangentMode=RCTM_User, time=0.5, value=0.4}
         <at 1 0 RCIM_Cubic>
        ]}}}
     Scale Curve=0.5
    )
    Index 0 Param 3=&MakeCustomFloatFromBool(Bool=$User.bIsPainted, True Float=1, False Float=5000)
    Index 1 Param 1=&MakeCustomFloatFromBool(Bool=$User.bIsPainted, True Float=0, False Float=1)
    Index 1 Param 2=&MakeCustomFloatFromBool(Bool=$User.bIsPainted, True Float=&MakeFloatFromInt(Int=$User.Lines_Glow), False Float=100)
    Module.Index0Param0Write=true
    Module.Index0Param1Write=true
    Module.Index0Param2Write=true
    Module.Index0Param3Write=true
    Module.Index1Param0Write=true
    Module.Index1Param1Write=true
    Module.Index1Param2Write=true
    Module.Index1Param3Write=true
    Write Parameter Index 1=true
   )
   &ScaleColor(
    Scale RGB=&VectorFromFloat(Value=&MakeCustomFloatFromBool(
      Bool=$User.bIsPainted
      True Float=&RandomRangeFloat(
       Minimum=&FloatFromCurve(
        FloatCurve=DataInterface() {propertyValues {Curve {keys [{interpMode=RCIM_Cubic, leaveTangent=-2.9826, tangentMode=RCTM_Break, time=0, value=1}, <at 0.95 0.15 RCIM_Cubic>]}}}
        Scale Curve=3
       )
       Maximum=&FloatFromCurve(FloatCurve=DataInterface() {propertyValues {Curve {keys [<at 0 1>,<at 1 0.5>]}}}, Scale Curve=3)
      )
      False Float=&FloatFromCurve(
       FloatCurve=DataInterface() {propertyValues {Curve {keys [{arriveTangent=-5.3196, interpMode=RCIM_Cubic, leaveTangent=-5.3196, tangentMode=RCTM_User, time=0.3, value=2}, <at 1 1 RCIM_Cubic>]}}}
       Scale Curve=&RandomRangeFloat(Minimum=1, Maximum=1)
      )
     ))
    ScaleA=true
    ScaleRGB=true
   )
  ]
  rendererClasses [&NiagaraMeshRendererProperties, &NiagaraSpriteRendererProperties]
  renderers [
   {rendererClass=&NiagaraMeshRendererProperties, rendererIndex=0}
   {rendererClass=&NiagaraSpriteRendererProperties, rendererIndex=1}
  ]
 }
 NE_VC_DriftTrail_CURVE_02006 {
  emitterUpdateScript [
   &EmitterState(Life Cycle Mode=::NewEnumerator1, Loop Duration=0.09, Recalculate Duration Each Loop=true, UseLoopDelay=false)
   &SpawnBurst_Instantaneous(Spawn Count=6, Use Spawn Probability=false, UseLoopCountLimit=false)
   &SpawnRate(SpawnRate=20, Use Spawn Probability=false) {enabled=false}
   /SetVariables(Emitter.IS FRONTEND=HlslExpression() {hlslExpression="User.GameModeID == 0 ? true : false"}) {isSetParameters=true}
  ]
  particleSpawnScript [
   &VC_PaintedVariantsOverride(
    Black=<rgba 0.01 0.01 0.01 1>
    BurntSienna=<rgba 0.25 0.0625 0 1>
    Cobalt=<rgba 0 0.0567 1.7 1>
    DefaultColor=&SelectLinearColorFromArray(
     Array Sampling Mode=::NewEnumerator0
     Direct Array Index=?
     Normalized Interpolation=false
     Color Selection Array=DataInterface() {propertyValues {ColorData [<rgba 0 0.8333 2 1>, <rgba 0.6 0 0.2 1>]}}
     Random Seed=0
    )
    ForestGreen=<rgba 0 0.5 0 1>
    Grey=<rgba 0.099 0.099 0.099 1>
    Lime=<rgba 0.25 0.6 0 1>
    Orange=<rgba 1.4 0.0933 0 1>
    Pink=<rgba 0.7 0.0839 0.4759 1>
    Saffron=<rgba 0.7 0.7 0 1>
    SkyBlue=<rgba 0 0.5323 0.8 1>
   )
   &InitializeParticle(
    Color=$Particles.Color
    Lifetime=0.2
    Mesh Scale=<5 5 5>
    Mesh Scale Mode=::NewEnumerator4
    Sprite Rotation Mode=::NewEnumerator2
    Sprite Size Mode=::NewEnumerator3
    UsePositionOffset=false
    Write Material Random=true
   )
   &PartitionParticles()
   &SkeletalMeshLocation(
    Bone / Socket Index=&Add_Integer(A=$Particles.PartitionParticles.Partition, B=4)
    Bone / Socket Sampling Mode=::NewEnumerator5
    Interpolate Mesh Sampling=false
    MaskSpawnGroup=false
    Mesh Sampling Type=::NewEnumerator6
    Sampled Position Offset=&AddVectorToPosition(Position=<0 0 0>, Vector=$User.offset)
    Skeletal Mesh=$System.SkeletalMesh
   )
   &AddVelocity(Velocity=&RandomRangeVector(Minimum=<1500 0 0>, Maximum=<1500 0 0>), Velocity Speed Scale=-1)
   &InitialMeshOrientation(
    Mesh Axis To Orient=&RandomRangeVector(Minimum=<0 -0.2 1>, Maximum=<0 0.2 1>)
    Mesh Orientation Mode=::NewEnumerator2
    Orientation Vector=&RandomRangeVector(Minimum=<0 1 1>, Maximum=<0 -1 1>)
    Rotation Coordinate Space=::NewEnumerator0
    Use Rotation Vector=false
   )
  ]
  particleUpdateScript [
   &ParticleState()
   &ScaleMeshSize(Scale Factor=&VectorFromCurve(
     VectorCurve=DataInterface() {propertyValues {
       XCurve {keys [<at 0 1 RCIM_Cubic>, {arriveTangent=-0.4, interpMode=RCIM_Cubic, leaveTangent=-0.4, time=0.8, value=0.65}, <at 1 0.6 RCIM_Cubic>]}
       YCurve {keys [
         <at 0 0.7 RCIM_Cubic>
         {arriveTangentWeight=0.178, interpMode=RCIM_Cubic, leaveTangentWeight=0.1068, tangentMode=RCTM_User, time=0.3, value=0.7}
         {arriveTangentWeight=0.1414, interpMode=RCIM_Cubic, leaveTangentWeight=0.1886, tangentMode=RCTM_User, time=0.5, value=0.8}
         {arriveTangent=-0.1027, interpMode=RCIM_Cubic, leaveTangent=-0.1027, tangentMode=RCTM_User, time=1, value=0}
        ]}
       ZCurve {keys [
         <at 0 0.7 RCIM_Cubic>
         {arriveTangentWeight=0.178, interpMode=RCIM_Cubic, leaveTangentWeight=0.1068, tangentMode=RCTM_User, time=0.3, value=0.7}
         {arriveTangentWeight=0.1414, interpMode=RCIM_Cubic, leaveTangentWeight=0.1886, tangentMode=RCTM_User, time=0.5, value=0.8}
         {arriveTangent=-0.1027, interpMode=RCIM_Cubic, leaveTangent=-0.1027, tangentMode=RCTM_User, time=1, value=0}
        ]}
      }}
     Scale Curve=&SelectVectorFromArray(
      Direct Array Index=$Particles.PartitionParticles.Partition
      Vector Selection Array=DataInterface() {propertyValues {FloatData [<0.75 -0.75 1>, <0.75 -0.75 1>]}}
     )
    ))
   &MeshRotationForce(
    Coordinate Space:ENiagaraMeshTransforms=::NewEnumerator2
    Rotation=&SelectVectorFromArray(
     Direct Array Index=$Particles.PartitionParticles.Partition
     Vector Selection Array=DataInterface() {propertyValues {FloatData [<10 0 0>, <-10 0 0>]}}
    )
   )
   &Drag(
    Drag=&FloatFromCurve(
     FloatCurve=DataInterface() {propertyValues {Curve {keys [{interpMode=RCIM_Cubic, tangentMode=RCTM_User, time=0.2, value=0}, {interpMode=RCIM_Cubic, tangentMode=RCTM_User, time=0.4, value=2}]}}}
     Scale Curve=2
    )
    Use Linear Drag=true
    Use Rotational Drag=true
   )
   &SolveForcesAndVelocity(Clamp Velocity=false, Limit Acceleration=false)
   &DynamicMaterialParameters(
    Index 0 Param 1=&FloatFromCurve(FloatCurve=DataInterface() {propertyValues {Curve {keys [
         <at 0.2 0 RCIM_Cubic>
         {arriveTangent=0.002, interpMode=RCIM_Cubic, leaveTangent=0.7645, tangentMode=RCTM_Break, time=0.6, value=0.1}
         {arriveTangent=2.25, interpMode=RCIM_Cubic, leaveTangent=2.25, time=0.8, value=0.3}
         {arriveTangent=0.644, tangentMode=RCTM_User, time=1, value=1}
        ]}}})
    Index 0 Param 2=&FloatFromCurve(FloatCurve=DataInterface() {propertyValues {Curve {keys [<at 0 1 RCIM_Cubic>,<at 0.5 0 RCIM_Cubic>]}}}, Scale Curve=0.5)
    Index 1 Param 1=&MakeCustomFloatFromBool(Bool=$User.bIsPainted, True Float=0, False Float=1)
    Index 1 Param 2=&MakeCustomFloatFromBool(Bool=$User.bIsPainted, True Float=&MakeFloatFromInt(Int=$User.Lines_Glow), False Float=100)
    Module.Index0Param0Write=true
    Module.Index0Param1Write=true
    Module.Index0Param2Write=true
    Module.Index0Param3Write=true
    Module.Index1Param0Write=true
    Module.Index1Param1Write=true
    Module.Index1Param2Write=true
    Module.Index1Param3Write=true
    Write Parameter Index 1=true
   )
   &ScaleColor(
    Scale RGB=&VectorFromFloat(Value=&MakeCustomFloatFromBool(
      Bool=$User.bIsPainted
      True Float=&FloatFromCurve(
       FloatCurve=DataInterface() {propertyValues {Curve {keys [{interpMode=RCIM_Cubic, leaveTangent=-2.4022, tangentMode=RCTM_Break, time=0, value=1}, <at 1 0.25 RCIM_Cubic>]}}}
       Scale Curve=3
      )
      False Float=&FloatFromCurve(
       FloatCurve=DataInterface() {propertyValues {Curve {keys [{arriveTangent=-5.2742, interpMode=RCIM_Cubic, leaveTangent=-5.2742, tangentMode=RCTM_User, time=0.3, value=2}, <at 1 1 RCIM_Cubic>]}}}
       Scale Curve=&RandomRangeFloat(Minimum=2, Maximum=2)
      )
     ))
    ScaleA=true
    ScaleRGB=true
   )
  ]
  rendererClasses [&NiagaraMeshRendererProperties, &NiagaraSpriteRendererProperties]
  renderers [
   {rendererClass=&NiagaraMeshRendererProperties, rendererIndex=0}
   {rendererClass=&NiagaraSpriteRendererProperties, rendererIndex=1}
  ]
 }
 NE_VC_DriftTrail_CURVE_03 {
  bEnabled=false
  emitterUpdateScript [
   &EmitterState(Life Cycle Mode=::NewEnumerator1, Loop Duration=0.1, Recalculate Duration Each Loop=true, UseLoopDelay=false)
   &SpawnRate(SpawnRate=10, Use Spawn Probability=false) {enabled=false, module=SpawnRate001}
   &SpawnBurst_Instantaneous(
    Spawn Count=&MakeCustomIntFromBool(Bool=$User.bIsFrontendPreview, True INT=6, False INT=0)
    Use Spawn Probability=false
    UseLoopCountLimit=false
   )
   &SpawnRate(SpawnRate=2, Use Spawn Probability=false) {enabled=false}
  ]
  particleSpawnScript [
   &VC_PaintedVariantsOverride(
    Black=<rgba 0.01 0.01 0.01 1>
    BurntSienna=<rgba 1 0.2116 0 1>
    Cobalt=<rgba 0 0.1333 2 1>
    DefaultColor=&SelectLinearColorFromArray(
     Array Sampling Mode=::NewEnumerator0
     Direct Array Index=?
     Color Selection Array=DataInterface() {propertyValues {ColorData [<rgba 0 0.42 1.5 1>, <rgba 0.5 0 0.4333 1>]}}
     Random Seed=0
    )
    ForestGreen=<rgba 0 0.7 0 1>
    Grey=<rgba 0.1 0.1 0.1 1>
    Lime=<rgba 0.5208 1.25 0 1>
    Orange=<rgba 1.5 0.075 0 1>
    Pink=<rgba 0.8 0.0958 0.5439 1>
    Purple=<rgba 0.1 0 0.6 10>
    Saffron=<rgba 0.7 0.7 0 1>
    SkyBlue=<rgba 0 0.6654 1 1>
    TitaniumWhite=<rgba 1 1 1 1>
   )
   &InitializeParticle(
    Color=$Particles.Color
    Lifetime=0.4
    Mesh Scale Max=<6 3 3>
    Mesh Scale Min=<6 3 3>
    Mesh Scale Mode=::NewEnumerator2
    Sprite Rotation Mode=::NewEnumerator2
    Sprite Size Mode=::NewEnumerator3
    UsePositionOffset=false
    Write Material Random=true
   )
   &PartitionParticles()
   &SkeletalMeshLocation(
    Bone / Socket Index=&Add_Integer(A=$Particles.PartitionParticles.Partition, B=4)
    Bone / Socket Sampling Mode=::NewEnumerator5
    Interpolate Mesh Sampling=false
    MaskSpawnGroup=false
    Mesh Sampling Type=::NewEnumerator6
    Sampled Position Offset=&AddVectorToPosition(Position=<0 0 0>, Vector=$User.offset)
    Skeletal Mesh=$System.SkeletalMesh
   )
   &AddVelocity(Velocity=&RandomRangeVector(Minimum=<50 0 0>, Maximum=<0 0 0>), Velocity Speed Scale=-1)
   &InitialMeshOrientation(
    Mesh Axis To Orient=&RandomRangeVector(Minimum=<0 -1 -1>, Maximum=<0 1 1>)
    Mesh Orientation Mode=::NewEnumerator2
    Orientation Vector=&RandomRangeVector(Minimum=<0 1 1>, Maximum=<0 -1 -1>)
    Rotation Coordinate Space=::NewEnumerator2
    Use Rotation Vector=false
   )
  ]
  particleUpdateScript [
   &ParticleState()
   &ScaleMeshSize(Scale Factor=&VectorFromCurve(VectorCurve=DataInterface() {propertyValues {
       XCurve {keys [<at 0 1>,<at 0.7 1>,<at 1 1>]}
       YCurve {keys [<at 0 0>,<at 0.2 1>,<at 0.7 1>,<at 1 1.2>]}
       ZCurve {keys [<at 0 0>,<at 0.2 1>,<at 0.7 1>,<at 1 1.2>]}
      }})) {enabled=false}
   &SolveForcesAndVelocity(Clamp Velocity=false, Limit Acceleration=false)
   &DynamicMaterialParameters(
    Index 0 Param 1=&FloatFromCurve(FloatCurve=DataInterface() {propertyValues {Curve {keys [
         {interpMode=RCIM_Cubic, leaveTangentWeight=0.3854, tangentMode=RCTM_User, time=0.2, value=0}
         {arriveTangent=0.1401, tangentMode=RCTM_User, time=0.9, value=1}
        ]}}})
    Index 0 Param 2=&FloatFromCurve(FloatCurve=DataInterface() {propertyValues {Curve {keys [
         {arriveTangent=-3.3715, interpMode=RCIM_Cubic, leaveTangent=-3.3715, leaveTangentWeight=1.1026, tangentMode=RCTM_User, time=0, value=1}
         {arriveTangentWeight=0.2828, interpMode=RCIM_Cubic, leaveTangentWeight=0.1886, tangentMode=RCTM_User, time=0.5, value=0.1}
         {arriveTangentWeight=0.1397, interpMode=RCIM_Cubic, tangentMode=RCTM_User, time=1, value=0}
        ]}}})
    Index 1 Param 1=&MakeCustomFloatFromBool(Bool=$User.bIsPainted, True Float=0, False Float=1)
    Module.Index0Param0Write=true
    Module.Index0Param1Write=true
    Module.Index0Param2Write=true
    Module.Index0Param3Write=true
    Module.Index1Param0Write=true
    Module.Index1Param1Write=true
    Module.Index1Param2Write=true
    Module.Index1Param3Write=true
    Write Parameter Index 1=true
   ) {enabled=false}
   &ScaleColor(
    Scale RGB=&VectorFromFloat(Value=&MakeCustomFloatFromBool(
      Bool=$User.bIsPainted
      True Float=&FloatFromCurve(
       FloatCurve=DataInterface() {propertyValues {Curve {keys [
           {interpMode=RCIM_Cubic, leaveTangent=-2.8244, tangentMode=RCTM_Break, time=0, value=1}
           {arriveTangentWeight=0.33, interpMode=RCIM_Cubic, leaveTangentWeight=0.1414, tangentMode=RCTM_User, time=0.8, value=0.1}
           <at 1 0.1 RCIM_Cubic>
          ]}}}
       Scale Curve=5
      )
      False Float=&FloatFromCurve(
       FloatCurve=DataInterface() {propertyValues {Curve {keys [{arriveTangent=-0.3719, interpMode=RCIM_Cubic, leaveTangent=-0.3719, tangentMode=RCTM_User, time=0.1, value=1}, <at 1 0.25 RCIM_Cubic>]}}}
       Scale Curve=&RandomRangeFloat(Minimum=4, Maximum=4)
      )
     ))
    ScaleA=true
    ScaleRGB=true
   )
   &DynamicMaterialParameters(
    Index 0 Param 1=&FloatFromCurve(FloatCurve=DataInterface() {propertyValues {Curve {keys [
         <at 0.45 0 RCIM_Cubic>
         {arriveTangent=0.002, interpMode=RCIM_Cubic, leaveTangent=0.7645, tangentMode=RCTM_Break, time=0.6, value=0.5}
         {arriveTangent=-0.0797, tangentMode=RCTM_User, time=1, value=1}
        ]}}})
    Index 0 Param 2=&FloatFromCurve(
     FloatCurve=DataInterface() {propertyValues {Curve {keys [
         <at 0 1 RCIM_Cubic>
         {arriveTangent=-1.3855, interpMode=RCIM_Cubic, leaveTangent=-1.3855, tangentMode=RCTM_User, time=0.5, value=0.4}
         <at 1 0 RCIM_Cubic>
        ]}}}
     Scale Curve=&RandomRangeFloat(Minimum=0.9, Maximum=1)
    )
    Index 1 Param 1=&MakeCustomFloatFromBool(Bool=$User.bIsPainted, True Float=0, False Float=1)
    Module.Index0Param0Write=true
    Module.Index0Param1Write=true
    Module.Index0Param2Write=true
    Module.Index0Param3Write=true
    Module.Index1Param0Write=true
    Module.Index1Param1Write=true
    Module.Index1Param2Write=true
    Module.Index1Param3Write=true
    Write Parameter Index 1=true
   ) {module=DynamicMaterialParameters001}
  ]
  rendererClasses [&NiagaraMeshRendererProperties, &NiagaraSpriteRendererProperties]
  renderers [
   {rendererClass=&NiagaraMeshRendererProperties, rendererIndex=0}
   {rendererClass=&NiagaraSpriteRendererProperties, rendererIndex=1}
  ]
 }
 NE_VC_DriftTrail_CURVE_03001 {
  emitterUpdateScript [
   &EmitterState(Life Cycle Mode=::NewEnumerator1, Loop Duration=0.1, Recalculate Duration Each Loop=true, UseLoopDelay=false)
   &SpawnRate(SpawnRate=10, Use Spawn Probability=false) {enabled=false, module=SpawnRate001}
   &SpawnBurst_Instantaneous(Spawn Count=12, Use Spawn Probability=false, UseLoopCountLimit=false)
   &SpawnRate(SpawnRate=2, Use Spawn Probability=false) {enabled=false}
   /SetVariables(Emitter.IS RR=HlslExpression() {hlslExpression="User.GameModeID == 2 ? true : false"}) {isSetParameters=true}
  ]
  particleSpawnScript [
   &VC_PaintedVariantsOverride(
    Black=<rgba 0.01 0.01 0.01 1>
    BurntSienna=<rgba 0.25 0.0625 0 1>
    Cobalt=<rgba 0 0.1333 2 1>
    DefaultColor=&SelectLinearColorFromArray(
     Array Sampling Mode=::NewEnumerator0
     Direct Array Index=?
     Color Selection Array=DataInterface() {propertyValues {ColorData [<rgba 0 1.6667 4 1>, <rgba 1.2 0 0.4 1>]}}
     Random Seed=0
    )
    ForestGreen=<rgba 0 0.7 0 1>
    Grey=<rgba 0.1 0.1 0.1 1>
    Lime=<rgba 0.5208 1.25 0 1>
    Orange=<rgba 1.5 0.075 0 1>
    Pink=<rgba 0.8 0.0958 0.5439 1>
    Saffron=<rgba 0.7 0.7 0 1>
    SkyBlue=<rgba 0 0.6654 1 1>
    TitaniumWhite=<rgba 1 1 1 1>
   )
   &InitializeParticle(
    Color=$Particles.Color
    Lifetime=0.4
    Mesh Scale Max=<6 3 3>
    Mesh Scale Min=<6 3 3>
    Mesh Scale Mode=::NewEnumerator2
    Sprite Rotation Mode=::NewEnumerator2
    Sprite Size Mode=::NewEnumerator3
    UsePositionOffset=false
    Write Material Random=true
   )
   &PartitionParticles()
   &SkeletalMeshLocation(
    Bone / Socket Index=&Add_Integer(A=$Particles.PartitionParticles.Partition, B=4)
    Bone / Socket Sampling Mode=::NewEnumerator5
    Interpolate Mesh Sampling=false
    MaskSpawnGroup=false
    Mesh Sampling Type=::NewEnumerator6
    Sampled Position Offset=&AddVectorToPosition(Position=<0 0 0>, Vector=$User.offset)
    Skeletal Mesh=$System.SkeletalMesh
   )
   &AddVelocity(Velocity=&RandomRangeVector(Minimum=<50 0 0>, Maximum=<0 0 0>), Velocity Speed Scale=-1)
   &InitialMeshOrientation(
    Mesh Axis To Orient=&RandomRangeVector(Minimum=<0 -1 -1>, Maximum=<0 1 1>)
    Mesh Orientation Mode=::NewEnumerator2
    Orientation Vector=&RandomRangeVector(Minimum=<0 1 1>, Maximum=<0 -1 -1>)
    Rotation Coordinate Space=::NewEnumerator2
    Use Rotation Vector=false
   )
  ]
  particleUpdateScript [
   &ParticleState()
   &ScaleMeshSize(Scale Factor=&VectorFromCurve(VectorCurve=DataInterface() {propertyValues {
       XCurve {keys [<at 0 1>,<at 0.7 1>,<at 1 1>]}
       YCurve {keys [<at 0 0>,<at 0.2 1>,<at 0.7 1>,<at 1 1.2>]}
       ZCurve {keys [<at 0 0>,<at 0.2 1>,<at 0.7 1>,<at 1 1.2>]}
      }})) {enabled=false}
   &SolveForcesAndVelocity(Clamp Velocity=false, Limit Acceleration=false)
   &DynamicMaterialParameters(
    Index 0 Param 1=&FloatFromCurve(FloatCurve=DataInterface() {propertyValues {Curve {keys [
         {interpMode=RCIM_Cubic, leaveTangentWeight=0.3854, tangentMode=RCTM_User, time=0.2, value=0}
         {arriveTangent=0.1401, tangentMode=RCTM_User, time=0.9, value=1}
        ]}}})
    Index 0 Param 2=&FloatFromCurve(FloatCurve=DataInterface() {propertyValues {Curve {keys [
         {arriveTangent=-3.3715, interpMode=RCIM_Cubic, leaveTangent=-3.3715, leaveTangentWeight=1.1026, tangentMode=RCTM_User, time=0, value=1}
         {arriveTangentWeight=0.2828, interpMode=RCIM_Cubic, leaveTangentWeight=0.1886, tangentMode=RCTM_User, time=0.5, value=0.1}
         {arriveTangentWeight=0.1397, interpMode=RCIM_Cubic, tangentMode=RCTM_User, time=1, value=0}
        ]}}})
    Index 1 Param 1=&MakeCustomFloatFromBool(Bool=$User.bIsPainted, True Float=0, False Float=1)
    Module.Index0Param0Write=true
    Module.Index0Param1Write=true
    Module.Index0Param2Write=true
    Module.Index0Param3Write=true
    Module.Index1Param0Write=true
    Module.Index1Param1Write=true
    Module.Index1Param2Write=true
    Module.Index1Param3Write=true
    Write Parameter Index 1=true
   ) {enabled=false}
   &ScaleColor(
    Scale RGB=&VectorFromFloat(Value=&MakeCustomFloatFromBool(
      Bool=$User.bIsPainted
      True Float=&FloatFromCurve(
       FloatCurve=DataInterface() {propertyValues {Curve {keys [
           {interpMode=RCIM_Cubic, leaveTangent=-2.8244, tangentMode=RCTM_Break, time=0, value=1}
           {arriveTangentWeight=0.33, interpMode=RCIM_Cubic, leaveTangentWeight=0.1414, tangentMode=RCTM_User, time=0.8, value=0.1}
           <at 1 0.1 RCIM_Cubic>
          ]}}}
       Scale Curve=5
      )
      False Float=&FloatFromCurve(
       FloatCurve=DataInterface() {propertyValues {Curve {keys [{arriveTangent=-0.3719, interpMode=RCIM_Cubic, leaveTangent=-0.3719, tangentMode=RCTM_User, time=0.1, value=1}, <at 1 0.25 RCIM_Cubic>]}}}
       Scale Curve=&RandomRangeFloat(Minimum=4, Maximum=4)
      )
     ))
    ScaleA=true
    ScaleRGB=true
   )
   &DynamicMaterialParameters(
    Index 0 Param 1=&FloatFromCurve(FloatCurve=DataInterface() {propertyValues {Curve {keys [
         <at 0.4 0 RCIM_Cubic>
         {arriveTangent=0.002, interpMode=RCIM_Cubic, leaveTangent=0.7645, tangentMode=RCTM_Break, time=0.5, value=0.5}
         {arriveTangent=-0.0797, tangentMode=RCTM_User, time=1, value=1}
        ]}}})
    Index 0 Param 2=&FloatFromCurve(
     FloatCurve=DataInterface() {propertyValues {Curve {keys [
         <at 0 1 RCIM_Cubic>
         {arriveTangent=-1.3855, interpMode=RCIM_Cubic, leaveTangent=-1.3855, tangentMode=RCTM_User, time=0.5, value=0.4}
         <at 1 0 RCIM_Cubic>
        ]}}}
     Scale Curve=&RandomRangeFloat(Minimum=0.9, Maximum=1)
    )
    Index 1 Param 1=&MakeCustomFloatFromBool(Bool=$User.bIsPainted, True Float=0, False Float=1)
    Index 1 Param 2=&MakeCustomFloatFromBool(Bool=$User.bIsPainted, True Float=&MakeFloatFromInt(Int=$User.Lines_Glow), False Float=100)
    Module.Index0Param0Write=true
    Module.Index0Param1Write=true
    Module.Index0Param2Write=true
    Module.Index0Param3Write=true
    Module.Index1Param0Write=true
    Module.Index1Param1Write=true
    Module.Index1Param2Write=true
    Module.Index1Param3Write=true
    Write Parameter Index 1=true
   ) {module=DynamicMaterialParameters001}
  ]
  rendererClasses [&NiagaraMeshRendererProperties, &NiagaraSpriteRendererProperties]
  renderers [
   {rendererClass=&NiagaraMeshRendererProperties, rendererIndex=0}
   {rendererClass=&NiagaraSpriteRendererProperties, rendererIndex=1}
  ]
 }
 NE_VC_DriftTrail_CURVE_03002 {
  emitterUpdateScript [
   &EmitterState(Life Cycle Mode=::NewEnumerator1, Loop Duration=0.01, Recalculate Duration Each Loop=true, UseLoopDelay=false)
   &SpawnBurst_Instantaneous(Spawn Count=5, Use Spawn Probability=false, UseLoopCountLimit=false)
   &SpawnRate(SpawnRate=30, Use Spawn Probability=false) {enabled=false}
   /SetVariables(Emitter.IS BR=HlslExpression() {hlslExpression="User.GameModeID == 1 ? true : false"}) {isSetParameters=true}
  ]
  particleSpawnScript [
   &VC_PaintedVariantsOverride(
    Black=<rgba 0.01 0.01 0.01 1>
    BurntSienna=<rgba 0.25 0.0625 0 1>
    Cobalt=<rgba 0 0.1333 2 1>
    DefaultColor=&SelectLinearColorFromArray(
     Array Sampling Mode=::NewEnumerator0
     Direct Array Index=?
     Color Selection Array=DataInterface() {propertyValues {ColorData [<rgba 0 1.6667 4 1>, <rgba 1.2 0 0.4 1>]}}
     Random Seed=0
    )
    ForestGreen=<rgba 0 0.7 0 1>
    Grey=<rgba 0.1 0.1 0.1 1>
    Lime=<rgba 0.5208 1.25 0 1>
    Orange=<rgba 1.5 0.075 0 1>
    Pink=<rgba 0.8 0.0958 0.5439 1>
    Saffron=<rgba 0.7 0.7 0 1>
    SkyBlue=<rgba 0 0.6654 1 1>
    TitaniumWhite=<rgba 1 1 1 1>
   )
   &InitializeParticle(
    Color=$Particles.Color
    Lifetime Max=0.25
    Lifetime Min=0.35
    Lifetime Mode=::NewEnumerator1
    Mesh Scale Max=<8 6 6>
    Mesh Scale Min=<5 4 4>
    Mesh Scale Mode=::NewEnumerator2
    Sprite Rotation Mode=::NewEnumerator2
    Sprite Size Mode=::NewEnumerator3
    UsePositionOffset=false
    Write Material Random=true
   )
   &PartitionParticles()
   &SkeletalMeshLocation(
    Bone / Socket Index=&Add_Integer(A=$Particles.PartitionParticles.Partition, B=4)
    Bone / Socket Sampling Mode=::NewEnumerator5
    Interpolate Mesh Sampling=false
    MaskSpawnGroup=false
    Mesh Sampling Type=::NewEnumerator6
    Sampled Position Offset=&AddVectorToPosition(Position=<0 0 10>, Vector=$User.offset)
    Skeletal Mesh=$System.SkeletalMesh
   )
   &AddVelocity(
    Velocity=&SelectVectorFromArray(
     Array Sampling Mode=::NewEnumerator0
     Direct Array Index=?
     Vector Selection Array=DataInterface() {propertyValues {FloatData [<1200 0 0>, <1200 0 0>]}}
     Random Seed=0
    )
    Velocity Speed Scale=-1
   )
   &InitialMeshOrientation(
    Mesh Axis To Orient=&RandomRangeVector(Minimum=<0 -1 -1>, Maximum=<0 1 1>)
    Mesh Orientation Mode=::NewEnumerator2
    Orientation Vector=&RandomRangeVector(Minimum=<0 1 1>, Maximum=<0 -1 -1>)
    Rotation Coordinate Space=::NewEnumerator2
    Use Rotation Vector=false
   )
  ]
  particleUpdateScript [
   &ParticleState()
   &ScaleMeshSize(Scale Factor=&Multiply_VectorByFloat(
     Vector=<1 0.75 0.75>
     Float=&FloatFromCurve(FloatCurve=DataInterface() {propertyValues {Curve {keys [{interpMode=RCIM_Cubic, tangentMode=RCTM_User, time=0, value=1}, {interpMode=RCIM_Cubic, tangentMode=RCTM_User, time=1, value=0}]}}})
    ))
   &InheritSourceMovement(
    Applied Position Scale=&Multiply_VectorByFloat(Vector=<1 1 1>, Float=1)
    ApplyForce=false
    ApplyPosition=true
    ApplyPositionFalloff=false
   )
   &SolveForcesAndVelocity(Clamp Velocity=false, Limit Acceleration=false)
   &DynamicMaterialParameters(
    Index 0 Param 1=&FloatFromCurve(FloatCurve=DataInterface() {propertyValues {Curve {keys [
         <at 0.4 0 RCIM_Cubic>
         {arriveTangent=0.002, interpMode=RCIM_Cubic, leaveTangent=0.7645, tangentMode=RCTM_Break, time=0.6, value=0.5}
         {arriveTangent=0.644, tangentMode=RCTM_User, time=1, value=1}
        ]}}})
    Index 0 Param 2=&FloatFromCurve(
     FloatCurve=DataInterface() {propertyValues {Curve {keys [
         <at 0 1 RCIM_Cubic>
         {arriveTangent=0.1281, interpMode=RCIM_Cubic, leaveTangent=0.1281, tangentMode=RCTM_User, time=0.4, value=0.25}
         <at 1 1 RCIM_Cubic>
        ]}}}
     Scale Curve=1.25
    )
    Index 1 Param 1=&MakeCustomFloatFromBool(Bool=$User.bIsPainted, True Float=0, False Float=1)
    Index 1 Param 2=&MakeCustomFloatFromBool(Bool=$User.bIsPainted, True Float=&MakeFloatFromInt(Int=$User.Lines_Glow), False Float=100)
    Module.Index0Param0Write=true
    Module.Index0Param1Write=true
    Module.Index0Param2Write=true
    Module.Index0Param3Write=true
    Module.Index1Param0Write=true
    Module.Index1Param1Write=true
    Module.Index1Param2Write=true
    Module.Index1Param3Write=true
    Write Parameter Index 1=true
   )
   &ScaleColor(
    Scale RGB=&VectorFromFloat(Value=&MakeCustomFloatFromBool(
      Bool=$User.bIsPainted
      True Float=&FloatFromCurve(
       FloatCurve=DataInterface() {propertyValues {Curve {keys [
           {interpMode=RCIM_Cubic, leaveTangent=-2.8244, tangentMode=RCTM_Break, time=0, value=1}
           {arriveTangentWeight=0.33, interpMode=RCIM_Cubic, leaveTangentWeight=0.1414, tangentMode=RCTM_User, time=0.6, value=0.1}
           <at 1 0 RCIM_Cubic>
          ]}}}
       Scale Curve=5
      )
      False Float=&FloatFromCurve(
       FloatCurve=DataInterface() {propertyValues {Curve {keys [<at 0 1>,<at 1 0.5>]}}}
       Scale Curve=&RandomRangeFloat(Minimum=4, Maximum=5)
      )
     ))
    ScaleA=true
    ScaleRGB=true
   )
  ]
  rendererClasses [&NiagaraMeshRendererProperties, &NiagaraSpriteRendererProperties]
  renderers [
   {rendererClass=&NiagaraMeshRendererProperties, rendererIndex=0}
   {rendererClass=&NiagaraSpriteRendererProperties, rendererIndex=1}
  ]
 }
 NE_VC_DriftTrail_CURVE_03003 {
  emitterUpdateScript [
   &EmitterState(Life Cycle Mode=::NewEnumerator1, Loop Duration=0.05, Recalculate Duration Each Loop=true, UseLoopDelay=false)
   &SpawnBurst_Instantaneous(Spawn Count=10, Use Spawn Probability=false, UseLoopCountLimit=false)
   &SpawnRate(SpawnRate=30, Use Spawn Probability=false) {enabled=false}
   /SetVariables(Emitter.IS BR=HlslExpression() {hlslExpression="User.GameModeID == 1 ? true : false"}) {isSetParameters=true}
  ]
  particleSpawnScript [
   &VC_PaintedVariantsOverride(
    Black=<rgba 0.01 0.01 0.01 1>
    BurntSienna=<rgba 0.25 0.0625 0 1>
    Cobalt=<rgba 0 0.1333 2 1>
    DefaultColor=&SelectLinearColorFromArray(
     Array Sampling Mode=::NewEnumerator0
     Direct Array Index=?
     Color Selection Array=DataInterface() {propertyValues {ColorData [<rgba 0 0.8333 2 1>, <rgba 0.6 0 0.2 1>]}}
     Random Seed=0
    )
    ForestGreen=<rgba 0 0.7 0 1>
    Grey=<rgba 0.1 0.1 0.1 1>
    Lime=<rgba 0.5208 1.25 0 1>
    Orange=<rgba 1.5 0.075 0 1>
    Pink=<rgba 0.8 0.0958 0.5439 1>
    Saffron=<rgba 0.7 0.7 0 1>
    SkyBlue=<rgba 0 0.6654 1 1>
    TitaniumWhite=<rgba 1 1 1 1>
   )
   &InitializeParticle(
    Color=$Particles.Color
    Lifetime=0.3
    Mesh Scale Max=<3 3 3>
    Mesh Scale Min=<3 5 5>
    Mesh Scale Mode=::NewEnumerator2
    Sprite Rotation Mode=::NewEnumerator2
    Sprite Size Mode=::NewEnumerator3
    UsePositionOffset=false
    Write Material Random=true
   )
   &PartitionParticles()
   &SkeletalMeshLocation(
    Bone / Socket Index=&Add_Integer(A=$Particles.PartitionParticles.Partition, B=4)
    Bone / Socket Sampling Mode=::NewEnumerator5
    Interpolate Mesh Sampling=false
    MaskSpawnGroup=false
    Mesh Sampling Type=::NewEnumerator6
    Sampled Position Offset=&AddVectorToPosition(Position=<-50 0 10>, Vector=$User.offset)
    Skeletal Mesh=$System.SkeletalMesh
   )
   &AddVelocity(Velocity=&RandomRangeVector(Minimum=<1000 0 0>, Maximum=<1000 0 0>), Velocity Speed Scale=-1)
   &InitialMeshOrientation(
    Mesh Axis To Orient=&RandomRangeVector(Minimum=<0 -1 -1>, Maximum=<0 1 1>)
    Mesh Orientation Mode=::NewEnumerator2
    Orientation Vector=&RandomRangeVector(Minimum=<0 1 1>, Maximum=<0 -1 -1>)
    Rotation Coordinate Space=::NewEnumerator2
    Use Rotation Vector=false
   )
  ]
  particleUpdateScript [
   &ParticleState()
   &ScaleMeshSize(Scale Factor=&VectorFromCurve(
     VectorCurve=DataInterface() {propertyValues {
       XCurve {keys [<at 0.35 1 RCIM_Cubic>, {arriveTangent=-2.5362, interpMode=RCIM_Cubic, leaveTangent=-2.5362, tangentMode=RCTM_User, time=0.75, value=0.25}]}
       YCurve {keys [<at 0 1>,<at 0.35 1>,<at 0.7 1.5>]}
       ZCurve {keys [<at 0 1>,<at 0.35 1>,<at 0.7 1.5>]}
      }}
     Scale Curve=<1 0.75 0.75>
    ))
   &InheritSourceMovement(
    Applied Position Scale=&Multiply_VectorByFloat(Vector=<1 1 1>, Float=1)
    ApplyForce=false
    ApplyPosition=true
    ApplyPositionFalloff=false
   )
   &SolveForcesAndVelocity(Clamp Velocity=false, Limit Acceleration=false)
   &DynamicMaterialParameters(
    Index 0 Param 1=&FloatFromCurve(
     FloatCurve=DataInterface() {propertyValues {Curve {keys [
         {arriveTangent=0.8333, interpMode=RCIM_Cubic, time=0.5, value=0}
         {arriveTangent=2, interpMode=RCIM_Cubic, leaveTangent=2, time=0.6, value=0.5}
         {arriveTangent=0.302, tangentMode=RCTM_User, time=1, value=1}
        ]}}}
     Scale Curve=&SelectFloatFromArray(
      Array Sampling Mode=::NewEnumerator0
      Direct Array Index=?
      Float Selection Array=DataInterface() {propertyValues {FloatData [0.5,0]}}
      Random Seed=0
     )
    )
    Index 0 Param 2=&FloatFromCurve(
     FloatCurve=DataInterface() {propertyValues {Curve {keys [<at 0 1 RCIM_Cubic>, {arriveTangent=2.0948, interpMode=RCIM_Cubic, leaveTangent=2.0948, tangentMode=RCTM_User, time=1, value=0}]}}}
     Scale Curve=0.5
    )
    Index 1 Param 1=&MakeCustomFloatFromBool(Bool=$User.bIsPainted, True Float=0, False Float=1)
    Index 1 Param 2=&MakeCustomFloatFromBool(Bool=$User.bIsPainted, True Float=&MakeFloatFromInt(Int=$User.Lines_Glow), False Float=100)
    Module.Index0Param0Write=true
    Module.Index0Param1Write=true
    Module.Index0Param2Write=true
    Module.Index0Param3Write=true
    Module.Index1Param0Write=true
    Module.Index1Param1Write=true
    Module.Index1Param2Write=true
    Module.Index1Param3Write=true
    Write Parameter Index 1=true
   )
   &ScaleColor(
    Scale RGB=&VectorFromFloat(Value=&MakeCustomFloatFromBool(
      Bool=$User.bIsPainted
      True Float=&FloatFromCurve(
       FloatCurve=DataInterface() {propertyValues {Curve {keys [
           {interpMode=RCIM_Cubic, leaveTangent=-2.4224, tangentMode=RCTM_Break, time=0, value=1}
           {arriveTangent=-0.1043, interpMode=RCIM_Cubic, leaveTangent=-0.1043, tangentMode=RCTM_User, time=1, value=0.1}
          ]}}}
       Scale Curve=4
      )
      False Float=&FloatFromCurve(
       FloatCurve=DataInterface() {propertyValues {Curve {keys [
           {interpMode=RCIM_Cubic, leaveTangentWeight=0.2018, tangentMode=RCTM_User, time=0, value=2}
           {arriveTangent=-1.65, arriveTangentWeight=0.3, interpMode=RCIM_Cubic, leaveTangent=-1.65, tangentMode=RCTM_User, time=1, value=0.35}
          ]}}}
       Scale Curve=&RandomRangeFloat(Minimum=4, Maximum=5)
      )
     ))
    ScaleA=true
    ScaleRGB=true
   )
  ]
  rendererClasses [&NiagaraMeshRendererProperties, &NiagaraSpriteRendererProperties]
  renderers [
   {rendererClass=&NiagaraMeshRendererProperties, rendererIndex=0}
   {rendererClass=&NiagaraSpriteRendererProperties, rendererIndex=1}
  ]
 }
 NE_VC_DriftTrail_CURVE_03004 {
  bEnabled=false
  emitterUpdateScript [
   &EmitterState(Life Cycle Mode=::NewEnumerator1, Loop Duration=0.1, Recalculate Duration Each Loop=true, UseLoopDelay=false)
   &SpawnBurst_Instantaneous(
    Spawn Count=&MakeCustomIntFromBool(Bool=$User.bIsFrontendPreview, True INT=0, False INT=5)
    Use Spawn Probability=false
    UseLoopCountLimit=false
   )
   &SpawnRate(SpawnRate=30, Use Spawn Probability=false) {enabled=false}
  ]
  particleSpawnScript [
   &VC_PaintedVariantsOverride(
    Black=<rgba 0.01 0.01 0.01 1>
    BurntSienna=<rgba 1 0.2116 0 1>
    Cobalt=<rgba 0 0.1333 2 1>
    DefaultColor=&SelectLinearColorFromArray(
     Array Sampling Mode=::NewEnumerator0
     Direct Array Index=?
     Color Selection Array=DataInterface() {propertyValues {ColorData [<rgba 0 0.42 1.5 1>, <rgba 0.5 0 0.4333 1>]}}
     Random Seed=0
    )
    ForestGreen=<rgba 0 0.7 0 1>
    Grey=<rgba 0.1 0.1 0.1 1>
    Lime=<rgba 0.5208 1.25 0 1>
    Orange=<rgba 1.5 0.075 0 1>
    Pink=<rgba 0.8 0.0958 0.5439 1>
    Purple=<rgba 0.1 0 0.6 10>
    Saffron=<rgba 0.7 0.7 0 1>
    SkyBlue=<rgba 0 0.6654 1 1>
    TitaniumWhite=<rgba 1 1 1 1>
   )
   &InitializeParticle(
    Color=$Particles.Color
    Lifetime Max=0.3
    Lifetime Min=0.5
    Lifetime Mode=::NewEnumerator1
    Mesh Scale Max=<8 6 6>
    Mesh Scale Min=<5 3 3>
    Mesh Scale Mode=::NewEnumerator2
    Sprite Rotation Mode=::NewEnumerator2
    Sprite Size Mode=::NewEnumerator3
    UsePositionOffset=false
    Write Material Random=true
   )
   &PartitionParticles()
   &SkeletalMeshLocation(
    Bone / Socket Index=&Add_Integer(A=$Particles.PartitionParticles.Partition, B=4)
    Bone / Socket Sampling Mode=::NewEnumerator5
    Interpolate Mesh Sampling=false
    MaskSpawnGroup=false
    Mesh Sampling Type=::NewEnumerator6
    Sampled Position Offset=&AddVectorToPosition(Position=<0 0 0>, Vector=$User.offset)
    Skeletal Mesh=$System.SkeletalMesh
   )
   &AddVelocity(Velocity=&RandomRangeVector(Minimum=<250 0 0>, Maximum=<0 0 0>), Velocity Speed Scale=-1)
   &InitialMeshOrientation(
    Mesh Axis To Orient=&RandomRangeVector(Minimum=<0 -1 -1>, Maximum=<0 1 1>)
    Mesh Orientation Mode=::NewEnumerator2
    Orientation Vector=&RandomRangeVector(Minimum=<0 1 1>, Maximum=<0 -1 -1>)
    Rotation Coordinate Space=::NewEnumerator2
    Use Rotation Vector=false
   )
  ]
  particleUpdateScript [
   &ParticleState()
   &ScaleMeshSize(Scale Factor=&VectorFromCurve(
     VectorCurve=DataInterface() {propertyValues {XCurve {keys [<at 0 1>,<at 0.75 1>,<at 1 0>]}, YCurve {keys [<at 0 1>,<at 0.75 1>,<at 1 2>]}, ZCurve {keys [<at 0 1>,<at 0.75 1>,<at 1 2>]}}}
     Scale Curve=<1 0.75 0.75>
    ))
   &SolveForcesAndVelocity(Clamp Velocity=false, Limit Acceleration=false)
   &DynamicMaterialParameters(
    Index 0 Param 1=&FloatFromCurve(
     FloatCurve=DataInterface() {propertyValues {Curve {keys [
         <at 0.25 0 RCIM_Cubic>
         {arriveTangent=0.002, interpMode=RCIM_Cubic, leaveTangent=0.7645, tangentMode=RCTM_Break, time=0.6, value=0.8}
         {arriveTangent=0.644, tangentMode=RCTM_User, time=1, value=1}
        ]}}}
     Scale Curve=1.5
    )
    Index 0 Param 2=&FloatFromCurve(
     FloatCurve=DataInterface() {propertyValues {Curve {keys [
         <at 0 1 RCIM_Cubic>
         {arriveTangent=0.1281, interpMode=RCIM_Cubic, leaveTangent=0.1281, tangentMode=RCTM_User, time=0.5, value=0.1}
         <at 0.8 0 RCIM_Cubic>
        ]}}}
     Scale Curve=1.2
    )
    Index 1 Param 1=&MakeCustomFloatFromBool(Bool=$User.bIsPainted, True Float=0, False Float=1)
    Module.Index0Param0Write=true
    Module.Index0Param1Write=true
    Module.Index0Param2Write=true
    Module.Index0Param3Write=true
    Module.Index1Param0Write=true
    Module.Index1Param1Write=true
    Module.Index1Param2Write=true
    Module.Index1Param3Write=true
    Write Parameter Index 1=true
   )
   &ScaleColor(
    Scale RGB=&VectorFromFloat(Value=&MakeCustomFloatFromBool(
      Bool=$User.bIsPainted
      True Float=&FloatFromCurve(
       FloatCurve=DataInterface() {propertyValues {Curve {keys [
           {interpMode=RCIM_Cubic, leaveTangent=-2.8244, tangentMode=RCTM_Break, time=0, value=1}
           {arriveTangentWeight=0.33, interpMode=RCIM_Cubic, leaveTangentWeight=0.1414, tangentMode=RCTM_User, time=0.6, value=0.1}
           <at 1 0 RCIM_Cubic>
          ]}}}
       Scale Curve=5
      )
      False Float=&FloatFromCurve(
       FloatCurve=DataInterface() {propertyValues {Curve {keys [{arriveTangent=-0.3719, interpMode=RCIM_Cubic, leaveTangent=-0.3719, tangentMode=RCTM_User, time=0.1, value=1}, <at 1 -0.5 RCIM_Cubic>]}}}
       Scale Curve=&RandomRangeFloat(Minimum=4, Maximum=5)
      )
     ))
    ScaleA=true
    ScaleRGB=true
   )
  ]
  rendererClasses [&NiagaraMeshRendererProperties, &NiagaraSpriteRendererProperties]
  renderers [
   {rendererClass=&NiagaraMeshRendererProperties, rendererIndex=0}
   {rendererClass=&NiagaraSpriteRendererProperties, rendererIndex=1}
  ]
 }
 NE_VC_DriftTrail_CURVE_03005 {
  bEnabled=false
  emitterUpdateScript [
   &EmitterState(Life Cycle Mode=::NewEnumerator1, Loop Duration=0.1, Recalculate Duration Each Loop=true, UseLoopDelay=false)
   &SpawnBurst_Instantaneous(
    Spawn Count=&MakeCustomIntFromBool(Bool=$User.bIsFrontendPreview, True INT=0, False INT=10)
    Use Spawn Probability=false
    UseLoopCountLimit=false
   )
   &SpawnRate(SpawnRate=30, Use Spawn Probability=false) {enabled=false}
  ]
  particleSpawnScript [
   &VC_PaintedVariantsOverride(
    Black=<rgba 0.01 0.01 0.01 1>
    BurntSienna=<rgba 1 0.2116 0 1>
    Cobalt=<rgba 0 0.1333 2 1>
    DefaultColor=&SelectLinearColorFromArray(
     Array Sampling Mode=::NewEnumerator0
     Direct Array Index=?
     Color Selection Array=DataInterface() {propertyValues {ColorData [<rgba 0 0.42 1.5 1>, <rgba 0.5 0 0.4333 1>]}}
     Random Seed=0
    )
    ForestGreen=<rgba 0 0.7 0 1>
    Grey=<rgba 0.1 0.1 0.1 1>
    Lime=<rgba 0.5208 1.25 0 1>
    Orange=<rgba 1.5 0.075 0 1>
    Pink=<rgba 0.8 0.0958 0.5439 1>
    Purple=<rgba 0.1167 0 0.7 10>
    Saffron=<rgba 0.7 0.7 0 1>
    SkyBlue=<rgba 0 0.6654 1 1>
    TitaniumWhite=<rgba 1 1 1 1>
   )
   &InitializeParticle(
    Color=$Particles.Color
    Lifetime=0.3
    Mesh Scale Max=<3 3 3>
    Mesh Scale Min=<3 5 5>
    Mesh Scale Mode=::NewEnumerator2
    Sprite Rotation Mode=::NewEnumerator2
    Sprite Size Mode=::NewEnumerator3
    UsePositionOffset=false
    Write Material Random=true
   )
   &PartitionParticles()
   &SkeletalMeshLocation(
    Bone / Socket Index=&Add_Integer(A=$Particles.PartitionParticles.Partition, B=4)
    Bone / Socket Sampling Mode=::NewEnumerator5
    Interpolate Mesh Sampling=false
    MaskSpawnGroup=false
    Mesh Sampling Type=::NewEnumerator6
    Sampled Position Offset=&AddVectorToPosition(Position=<-50 0 0>, Vector=$User.offset)
    Skeletal Mesh=$System.SkeletalMesh
   )
   &AddVelocity(
    Velocity=&MakeCustomVectorFromBool(Bool=$User.bIsFrontendPreview, True Vector=<750 0 0>, False Vector=<1000 0 0>)
    Velocity Speed Scale=-1
   )
   &InitialMeshOrientation(
    Mesh Axis To Orient=&RandomRangeVector(Minimum=<0 -1 -1>, Maximum=<0 1 1>)
    Mesh Orientation Mode=::NewEnumerator2
    Orientation Vector=&RandomRangeVector(Minimum=<0 1 1>, Maximum=<0 -1 1>)
    Rotation Coordinate Space=::NewEnumerator2
    Use Rotation Vector=false
   )
  ]
  particleUpdateScript [
   &ParticleState()
   &ScaleMeshSize(Scale Factor=&VectorFromCurve(
     VectorCurve=DataInterface() {propertyValues {
       XCurve {keys [<at 0.5 1 RCIM_Cubic>, {arriveTangent=-2.5362, interpMode=RCIM_Cubic, leaveTangent=-2.5362, tangentMode=RCTM_User, time=1, value=0.25}]}
       YCurve {keys [<at 0 1>,<at 0.5 1>,<at 1 1.5>]}
       ZCurve {keys [<at 0 1>,<at 0.5 1>,<at 1 1.5>]}
      }}
     Scale Curve=<1 0.75 0.75>
    ))
   &ScaleVelocity(Velocity Scale=&VectorFromCurve(
     VectorCurve=DataInterface() {propertyValues {XCurve {keys [<at 0.2 0 RCIM_Cubic>,<at 1 1 RCIM_Cubic>]}, YCurve {keys [<at 0 1>,<at 1 1>]}, ZCurve {keys [<at 0 1>,<at 1 1>]}}}
     Scale Curve=<2 1 1>
    ))
   &SolveForcesAndVelocity(Clamp Velocity=false, Limit Acceleration=false)
   &DynamicMaterialParameters(
    Index 0 Param 1=&FloatFromCurve(
     FloatCurve=DataInterface() {propertyValues {Curve {keys [
         {arriveTangent=0.8333, interpMode=RCIM_Cubic, time=0.5, value=0}
         {arriveTangent=2, interpMode=RCIM_Cubic, leaveTangent=2, time=0.6, value=0.5}
         {arriveTangent=0.302, tangentMode=RCTM_User, time=1, value=1}
        ]}}}
     Scale Curve=&SelectFloatFromArray(
      Array Sampling Mode=::NewEnumerator0
      Direct Array Index=?
      Float Selection Array=DataInterface() {propertyValues {FloatData [0.5,0]}}
      Random Seed=0
     )
    )
    Index 0 Param 2=&FloatFromCurve(
     FloatCurve=DataInterface() {propertyValues {Curve {keys [<at 0 1 RCIM_Cubic>, {arriveTangent=2.0948, interpMode=RCIM_Cubic, leaveTangent=2.0948, tangentMode=RCTM_User, time=1, value=0}]}}}
     Scale Curve=0.5
    )
    Index 1 Param 1=&MakeCustomFloatFromBool(Bool=$User.bIsPainted, True Float=0, False Float=1)
    Module.Index0Param0Write=true
    Module.Index0Param1Write=true
    Module.Index0Param2Write=true
    Module.Index0Param3Write=true
    Module.Index1Param0Write=true
    Module.Index1Param1Write=true
    Module.Index1Param2Write=true
    Module.Index1Param3Write=true
    Write Parameter Index 1=true
   )
   &ScaleColor(
    Scale RGB=&VectorFromFloat(Value=&MakeCustomFloatFromBool(
      Bool=$User.bIsPainted
      True Float=&FloatFromCurve(
       FloatCurve=DataInterface() {propertyValues {Curve {keys [
           {interpMode=RCIM_Cubic, leaveTangent=-2.4224, tangentMode=RCTM_Break, time=0, value=1}
           {arriveTangent=-0.1043, interpMode=RCIM_Cubic, leaveTangent=-0.1043, tangentMode=RCTM_User, time=1, value=0.1}
          ]}}}
       Scale Curve=4
      )
      False Float=&FloatFromCurve(
       FloatCurve=DataInterface() {propertyValues {Curve {keys [
           {interpMode=RCIM_Cubic, leaveTangentWeight=0.2018, tangentMode=RCTM_User, time=0, value=2}
           {arriveTangent=-1.65, arriveTangentWeight=0.3, interpMode=RCIM_Cubic, leaveTangent=-1.65, tangentMode=RCTM_User, time=1, value=0.35}
          ]}}}
       Scale Curve=&RandomRangeFloat(Minimum=4, Maximum=5)
      )
     ))
    ScaleA=true
    ScaleRGB=true
   )
  ]
  rendererClasses [&NiagaraMeshRendererProperties, &NiagaraSpriteRendererProperties]
  renderers [
   {rendererClass=&NiagaraMeshRendererProperties, rendererIndex=0}
   {rendererClass=&NiagaraSpriteRendererProperties, rendererIndex=1}
  ]
 }
 NE_VC_DriftTrail_CURVE_03006 {
  emitterUpdateScript [
   &EmitterState(Life Cycle Mode=::NewEnumerator1, Loop Duration=0.1, Recalculate Duration Each Loop=true, UseLoopDelay=false)
   &SpawnBurst_Instantaneous(Spawn Count=10, Use Spawn Probability=false, UseLoopCountLimit=false)
   &SpawnRate(SpawnRate=30, Use Spawn Probability=false) {enabled=false}
   /SetVariables(Emitter.IS RR=HlslExpression() {hlslExpression="User.GameModeID == 2 ? true : false"}) {isSetParameters=true}
  ]
  particleSpawnScript [
   &VC_PaintedVariantsOverride(
    Black=<rgba 0.01 0.01 0.01 1>
    BurntSienna=<rgba 0.25 0.0625 0 1>
    Cobalt=<rgba 0 0.1333 2 1>
    DefaultColor=&SelectLinearColorFromArray(
     Array Sampling Mode=::NewEnumerator0
     Direct Array Index=?
     Color Selection Array=DataInterface() {propertyValues {ColorData [<rgba 0 0.8333 2 1>, <rgba 0.6 0 0.2 1>]}}
     Random Seed=0
    )
    ForestGreen=<rgba 0 0.7 0 1>
    Grey=<rgba 0.1 0.1 0.1 1>
    Lime=<rgba 0.5208 1.25 0 1>
    Orange=<rgba 1.5 0.075 0 1>
    Pink=<rgba 0.8 0.0958 0.5439 1>
    Saffron=<rgba 0.7 0.7 0 1>
    SkyBlue=<rgba 0 0.6654 1 1>
    TitaniumWhite=<rgba 1 1 1 1>
   )
   &InitializeParticle(
    Color=$Particles.Color
    Lifetime=0.3
    Mesh Scale Max=<3 3 3>
    Mesh Scale Min=<3 5 5>
    Mesh Scale Mode=::NewEnumerator2
    Sprite Rotation Mode=::NewEnumerator2
    Sprite Size Mode=::NewEnumerator3
    UsePositionOffset=false
    Write Material Random=true
   )
   &PartitionParticles()
   &SkeletalMeshLocation(
    Bone / Socket Index=&Add_Integer(A=$Particles.PartitionParticles.Partition, B=4)
    Bone / Socket Sampling Mode=::NewEnumerator5
    Interpolate Mesh Sampling=false
    MaskSpawnGroup=false
    Mesh Sampling Type=::NewEnumerator6
    Sampled Position Offset=&AddVectorToPosition(Position=<-50 0 0>, Vector=$User.offset)
    Skeletal Mesh=$System.SkeletalMesh
   )
   &AddVelocity(
    Velocity=&MakeCustomVectorFromBool(Bool=$User.bIsFrontendPreview, True Vector=<750 0 0>, False Vector=<1000 0 0>)
    Velocity Speed Scale=-1
   )
   &InitialMeshOrientation(
    Mesh Axis To Orient=&RandomRangeVector(Minimum=<0 -1 -1>, Maximum=<0 1 1>)
    Mesh Orientation Mode=::NewEnumerator2
    Orientation Vector=&RandomRangeVector(Minimum=<0 1 1>, Maximum=<0 -1 1>)
    Rotation Coordinate Space=::NewEnumerator2
    Use Rotation Vector=false
   )
  ]
  particleUpdateScript [
   &ParticleState()
   &ScaleMeshSize(Scale Factor=&VectorFromCurve(
     VectorCurve=DataInterface() {propertyValues {
       XCurve {keys [<at 0.5 1 RCIM_Cubic>, {arriveTangent=-2.5362, interpMode=RCIM_Cubic, leaveTangent=-2.5362, tangentMode=RCTM_User, time=1, value=0.25}]}
       YCurve {keys [<at 0 1>,<at 0.5 1>,<at 1 1.5>]}
       ZCurve {keys [<at 0 1>,<at 0.5 1>,<at 1 1.5>]}
      }}
     Scale Curve=<1 0.75 0.75>
    ))
   &ScaleVelocity(Velocity Scale=&VectorFromCurve(
     VectorCurve=DataInterface() {propertyValues {XCurve {keys [<at 0.2 0 RCIM_Cubic>,<at 1 1 RCIM_Cubic>]}, YCurve {keys [<at 0 1>,<at 1 1>]}, ZCurve {keys [<at 0 1>,<at 1 1>]}}}
     Scale Curve=<2 1 1>
    ))
   &SolveForcesAndVelocity(Clamp Velocity=false, Limit Acceleration=false)
   &DynamicMaterialParameters(
    Index 0 Param 1=&FloatFromCurve(
     FloatCurve=DataInterface() {propertyValues {Curve {keys [
         {arriveTangent=0.8333, interpMode=RCIM_Cubic, time=0.5, value=0}
         {arriveTangent=2, interpMode=RCIM_Cubic, leaveTangent=2, time=0.6, value=0.5}
         {arriveTangent=0.302, tangentMode=RCTM_User, time=1, value=1}
        ]}}}
     Scale Curve=&SelectFloatFromArray(
      Array Sampling Mode=::NewEnumerator0
      Direct Array Index=?
      Float Selection Array=DataInterface() {propertyValues {FloatData [0.5,0]}}
      Random Seed=0
     )
    )
    Index 0 Param 2=&FloatFromCurve(
     FloatCurve=DataInterface() {propertyValues {Curve {keys [<at 0 1 RCIM_Cubic>, {arriveTangent=2.0948, interpMode=RCIM_Cubic, leaveTangent=2.0948, tangentMode=RCTM_User, time=1, value=0}]}}}
     Scale Curve=0.5
    )
    Index 1 Param 1=&MakeCustomFloatFromBool(Bool=$User.bIsPainted, True Float=0, False Float=1)
    Index 1 Param 2=&MakeCustomFloatFromBool(Bool=$User.bIsPainted, True Float=&MakeFloatFromInt(Int=$User.Lines_Glow), False Float=100)
    Module.Index0Param0Write=true
    Module.Index0Param1Write=true
    Module.Index0Param2Write=true
    Module.Index0Param3Write=true
    Module.Index1Param0Write=true
    Module.Index1Param1Write=true
    Module.Index1Param2Write=true
    Module.Index1Param3Write=true
    Write Parameter Index 1=true
   )
   &ScaleColor(
    Scale RGB=&VectorFromFloat(Value=&MakeCustomFloatFromBool(
      Bool=$User.bIsPainted
      True Float=&FloatFromCurve(
       FloatCurve=DataInterface() {propertyValues {Curve {keys [
           {interpMode=RCIM_Cubic, leaveTangent=-2.4224, tangentMode=RCTM_Break, time=0, value=1}
           {arriveTangent=-0.1043, interpMode=RCIM_Cubic, leaveTangent=-0.1043, tangentMode=RCTM_User, time=1, value=0.1}
          ]}}}
       Scale Curve=4
      )
      False Float=&FloatFromCurve(
       FloatCurve=DataInterface() {propertyValues {Curve {keys [
           {interpMode=RCIM_Cubic, leaveTangentWeight=0.2018, tangentMode=RCTM_User, time=0, value=2}
           {arriveTangent=-1.65, arriveTangentWeight=0.3, interpMode=RCIM_Cubic, leaveTangent=-1.65, tangentMode=RCTM_User, time=1, value=0.35}
          ]}}}
       Scale Curve=&RandomRangeFloat(Minimum=4, Maximum=5)
      )
     ))
    ScaleA=true
    ScaleRGB=true
   )
  ]
  rendererClasses [&NiagaraMeshRendererProperties, &NiagaraSpriteRendererProperties]
  renderers [
   {rendererClass=&NiagaraMeshRendererProperties, rendererIndex=0}
   {rendererClass=&NiagaraSpriteRendererProperties, rendererIndex=1}
  ]
 }
 NE_VC_DriftTrail_CURVE_03007 {
  emitterUpdateScript [
   &EmitterState(Life Cycle Mode=::NewEnumerator1, Loop Duration=0.2, Recalculate Duration Each Loop=true, UseLoopDelay=false)
   &SpawnRate(SpawnRate=10, Use Spawn Probability=false) {enabled=false, module=SpawnRate001}
   &SpawnBurst_Instantaneous(Spawn Count=6, Use Spawn Probability=false, UseLoopCountLimit=false)
   &SpawnRate(SpawnRate=16, Use Spawn Probability=false) {enabled=false}
   /SetVariables(Emitter.IS FRONTEND=HlslExpression() {hlslExpression="User.GameModeID == 0 ? true : false"}) {isSetParameters=true}
  ]
  particleSpawnScript [
   &VC_PaintedVariantsOverride(
    Black=<rgba 0.01 0.01 0.01 1>
    BurntSienna=<rgba 0.25 0.0625 0 1>
    Cobalt=<rgba 0 0.1333 2 1>
    DefaultColor=&SelectLinearColorFromArray(
     Array Sampling Mode=::NewEnumerator0
     Direct Array Index=?
     Color Selection Array=DataInterface() {propertyValues {ColorData [<rgba 0 1.6667 4 1>, <rgba 1.2 0 0.4 1>]}}
     Random Seed=0
    )
    ForestGreen=<rgba 0 0.7 0 1>
    Grey=<rgba 0.065 0.065 0.065 1>
    Lime=<rgba 0.5208 1.25 0 1>
    Orange=<rgba 1.5 0.1 0 1>
    Pink=<rgba 0.8 0.0958 0.5439 1>
    Saffron=<rgba 0.7 0.7 0 1>
    SkyBlue=<rgba 0 0.6654 1 1>
    TitaniumWhite=<rgba 1 1 1 1>
   )
   &InitializeParticle(
    Color=$Particles.Color
    Lifetime=0.4
    Mesh Scale Max=<6 3 3>
    Mesh Scale Min=<6 3 3>
    Mesh Scale Mode=::NewEnumerator2
    Sprite Rotation Mode=::NewEnumerator2
    Sprite Size Mode=::NewEnumerator3
    UsePositionOffset=false
    Write Material Random=true
   )
   &PartitionParticles()
   &SkeletalMeshLocation(
    Bone / Socket Index=&Add_Integer(A=$Particles.PartitionParticles.Partition, B=4)
    Bone / Socket Sampling Mode=::NewEnumerator5
    Interpolate Mesh Sampling=false
    MaskSpawnGroup=false
    Mesh Sampling Type=::NewEnumerator6
    Sampled Position Offset=&AddVectorToPosition(Position=<0 0 0>, Vector=$User.offset)
    Skeletal Mesh=$System.SkeletalMesh
   )
   &AddVelocity(Velocity=&RandomRangeVector(Minimum=<100 0 0>, Maximum=<100 0 0>), Velocity Speed Scale=-1)
   &InitialMeshOrientation(
    Mesh Axis To Orient=&RandomRangeVector(Minimum=<0 1 1>, Maximum=<0 1 -1>)
    Mesh Orientation Mode=::NewEnumerator2
    Orientation Vector=&RandomRangeVector(Minimum=<0 1 1>, Maximum=<0 1 1>)
    Rotation=<90 65 0>
    Rotation Coordinate Space=::NewEnumerator2
    Use Rotation Vector=false
   )
  ]
  particleUpdateScript [
   &ParticleState()
   &ScaleMeshSize(Scale Factor=&VectorFromCurve(VectorCurve=DataInterface() {propertyValues {
       XCurve {keys [<at 0 1>,<at 1 1>]}
       YCurve {keys [<at 0.1 0 RCIM_Cubic>,<at 0.7 1 RCIM_Cubic>]}
       ZCurve {keys [<at 0.1 0 RCIM_Cubic>,<at 0.7 1 RCIM_Cubic>]}
      }})) {enabled=false}
   &SolveForcesAndVelocity(Clamp Velocity=false, Limit Acceleration=false)
   &DynamicMaterialParameters(
    Index 0 Param 1=&FloatFromCurve(FloatCurve=DataInterface() {propertyValues {Curve {keys [
         {interpMode=RCIM_Cubic, leaveTangentWeight=0.3854, tangentMode=RCTM_User, time=0.2, value=0}
         {arriveTangent=0.1401, tangentMode=RCTM_User, time=0.9, value=1}
        ]}}})
    Index 0 Param 2=&FloatFromCurve(FloatCurve=DataInterface() {propertyValues {Curve {keys [
         {arriveTangent=-3.3715, interpMode=RCIM_Cubic, leaveTangent=-3.3715, leaveTangentWeight=1.1026, tangentMode=RCTM_User, time=0, value=1}
         {arriveTangentWeight=0.2828, interpMode=RCIM_Cubic, leaveTangentWeight=0.1886, tangentMode=RCTM_User, time=0.5, value=0.1}
         {arriveTangentWeight=0.1397, interpMode=RCIM_Cubic, tangentMode=RCTM_User, time=1, value=0}
        ]}}})
    Index 1 Param 1=&MakeCustomFloatFromBool(Bool=$User.bIsPainted, True Float=0, False Float=1)
    Module.Index0Param0Write=true
    Module.Index0Param1Write=true
    Module.Index0Param2Write=true
    Module.Index0Param3Write=true
    Module.Index1Param0Write=true
    Module.Index1Param1Write=true
    Module.Index1Param2Write=true
    Module.Index1Param3Write=true
    Write Parameter Index 1=true
   ) {enabled=false}
   &ScaleColor(
    Scale RGB=&VectorFromFloat(Value=&MakeCustomFloatFromBool(
      Bool=$User.bIsPainted
      True Float=&FloatFromCurve(
       FloatCurve=DataInterface() {propertyValues {Curve {keys [{interpMode=RCIM_Cubic, leaveTangent=-2.5187, tangentMode=RCTM_Break, time=0, value=1}, <at 1 0.15 RCIM_Cubic>]}}}
       Scale Curve=5
      )
      False Float=&FloatFromCurve(
       FloatCurve=DataInterface() {propertyValues {Curve {keys [{arriveTangent=-0.3719, interpMode=RCIM_Cubic, leaveTangent=-0.3719, tangentMode=RCTM_User, time=0.1, value=1}, <at 1 0.25 RCIM_Cubic>]}}}
       Scale Curve=&RandomRangeFloat(Minimum=3, Maximum=2)
      )
     ))
    ScaleA=true
    ScaleRGB=true
   )
   &DynamicMaterialParameters(
    Index 0 Param 1=&FloatFromCurve(FloatCurve=DataInterface() {propertyValues {Curve {keys [<at 0.2 0 RCIM_Cubic>, {arriveTangent=0.1221, tangentMode=RCTM_User, time=1, value=1}]}}})
    Index 0 Param 2=&FloatFromCurve(
     FloatCurve=DataInterface() {propertyValues {Curve {keys [
         <at 0 1 RCIM_Cubic>
         {arriveTangent=-1.3855, interpMode=RCIM_Cubic, leaveTangent=-1.3855, tangentMode=RCTM_User, time=0.5, value=0.45}
         <at 1 0 RCIM_Cubic>
        ]}}}
     Scale Curve=&RandomRangeFloat(Minimum=0.8, Maximum=0.8)
    )
    Index 1 Param 1=&MakeCustomFloatFromBool(Bool=$User.bIsPainted, True Float=0, False Float=1)
    Index 1 Param 2=&MakeCustomFloatFromBool(Bool=$User.bIsPainted, True Float=&MakeFloatFromInt(Int=$User.Lines_Glow), False Float=100)
    Module.Index0Param0Write=true
    Module.Index0Param1Write=true
    Module.Index0Param2Write=true
    Module.Index0Param3Write=true
    Module.Index1Param0Write=true
    Module.Index1Param1Write=true
    Module.Index1Param2Write=true
    Module.Index1Param3Write=true
    Write Parameter Index 1=true
   ) {module=DynamicMaterialParameters001}
  ]
  rendererClasses [&NiagaraMeshRendererProperties, &NiagaraSpriteRendererProperties]
  renderers [
   {rendererClass=&NiagaraMeshRendererProperties, rendererIndex=0}
   {rendererClass=&NiagaraSpriteRendererProperties, rendererIndex=1}
  ]
 }
 NE_VC_DriftTrail_CURVE_1 {
  bEnabled=false
  emitterUpdateScript [
   &EmitterState(Life Cycle Mode=::NewEnumerator1, Loop Duration=0.05, Recalculate Duration Each Loop=true, UseLoopDelay=false)
   &SpawnBurst_Instantaneous(
    Spawn Count=&MakeCustomIntFromBool(Bool=$User.bIsFrontendPreview, True INT=6, False INT=0)
    Use Spawn Probability=false
    UseLoopCountLimit=false
   )
   &SpawnRate(
    SpawnRate=&MakeCustomFloatFromBool(Bool=$User.bIsFrontendPreview, True Float=120, False Float=0)
    Use Spawn Probability=false
   ) {enabled=false}
  ]
  particleSpawnScript [
   &VC_PaintedVariantsOverride(
    Black=<rgba 0.01 0.01 0.01 1>
    BurntSienna=<rgba 1 0.2116 0 1>
    Cobalt=<rgba 0 0.125 1.5 1>
    DefaultColor=&SelectLinearColorFromArray(
     Array Sampling Mode=::NewEnumerator0
     Direct Array Index=?
     Color Selection Array=DataInterface() {propertyValues {ColorData [<rgba 0 0.42 1.5 1>, <rgba 0.5 0 0.4333 1>]}}
     Random Seed=0
    )
    ForestGreen=<rgba 0 0.7 0 1>
    Grey=<rgba 0.1 0.1 0.1 1>
    Lime=<rgba 0.5208 1.25 0 1>
    Orange=<rgba 1.25 0.0417 0 1>
    Pink=<rgba 0.8 0.0958 0.5439 1>
    Purple=<rgba 0.1 0 0.6 10>
    Saffron=<rgba 0.7 0.7 0 1>
    SkyBlue=<rgba 0 0.6654 1 1>
    TitaniumWhite=<rgba 1 1 1 1>
   )
   &InitializeParticle(
    Color=$Particles.Color
    Lifetime Max=0.4
    Lifetime Min=0.3
    Lifetime Mode=::NewEnumerator1
    Mesh Scale Max=&RandomRangeVector(Minimum=<5 3 3>, Maximum=<5 5 5>)
    Mesh Scale Min=&RandomRangeVector(Minimum=<2 5 5>, Maximum=<2 5 7>)
    Mesh Scale Mode=::NewEnumerator2
    Sprite Rotation Mode=::NewEnumerator2
    Sprite Size Mode=::NewEnumerator3
    UsePositionOffset=false
    Write Material Random=true
   )
   &PartitionParticles()
   &SkeletalMeshLocation(
    Bone / Socket Index=&Add_Integer(A=$Particles.PartitionParticles.Partition, B=4)
    Bone / Socket Sampling Mode=::NewEnumerator5
    Interpolate Mesh Sampling=false
    MaskSpawnGroup=false
    Mesh Sampling Type=::NewEnumerator6
    Sampled Position Offset=&AddVectorToPosition(Position=<0 0 0>, Vector=$User.offset)
    Skeletal Mesh=$System.SkeletalMesh
   )
   &AddVelocity(Velocity=<1000 0 0>, Velocity Speed Scale=-1) {enabled=false}
   &InitialMeshOrientation(
    Mesh Axis To Orient=&RandomRangeVector(Minimum=<0 -1 -1>, Maximum=<0 1 1>)
    Mesh Orientation Mode=::NewEnumerator2
    Orientation Vector=&RandomRangeVector(Minimum=<0 1 1>, Maximum=<0 -1 -1>, Randomness Mode=::NewEnumerator1)
    Rotation=<90 65 0>
    Rotation Coordinate Space=::NewEnumerator2
    Use Rotation Vector=false
   )
  ]
  particleUpdateScript [
   &ParticleState()
   &ScaleMeshSize(Scale Factor=&SelectVectorFromArray(
     Direct Array Index=$Particles.PartitionParticles.Partition
     Vector Selection Array=DataInterface() {propertyValues {FloatData [<1 -0.75 0.75>, <1 0.75 0.75>]}}
    ))
   &MeshRotationForce(Rotation=&SelectVectorFromArray(
     Direct Array Index=$Particles.PartitionParticles.Partition
     Vector Selection Array=DataInterface() {propertyValues {FloatData [<-10 0 0>, <10 0 0>]}}
    )) {enabled=false}
   &SolveForcesAndVelocity(Clamp Velocity=false, Limit Acceleration=false)
   &DynamicMaterialParameters(
    Index 0 Param 1=&FloatFromCurve(
     FloatCurve=DataInterface() {propertyValues {Curve {keys [
         <at 0.4 0 RCIM_Cubic>
         {arriveTangent=0.002, interpMode=RCIM_Cubic, leaveTangent=0.7645, tangentMode=RCTM_Break, time=0.5, value=0.55}
         {arriveTangent=0.644, tangentMode=RCTM_User, time=1, value=0.95}
        ]}}}
     Scale Curve=0.9
    )
    Index 0 Param 2=&FloatFromCurve(
     FloatCurve=DataInterface() {propertyValues {Curve {keys [
         <at 0 1 RCIM_Cubic>
         {arriveTangent=-1.3855, interpMode=RCIM_Cubic, leaveTangent=-1.3855, tangentMode=RCTM_User, time=0.5, value=0.4}
         <at 1 0 RCIM_Cubic>
        ]}}}
     Scale Curve=&RandomRangeFloat(Minimum=0.9, Maximum=0.9)
    )
    Index 1 Param 1=&MakeCustomFloatFromBool(Bool=$User.bIsPainted, True Float=0, False Float=1)
    Module.Index0Param0Write=true
    Module.Index0Param1Write=true
    Module.Index0Param2Write=true
    Module.Index0Param3Write=true
    Module.Index1Param0Write=true
    Module.Index1Param1Write=true
    Module.Index1Param2Write=true
    Module.Index1Param3Write=true
    Write Parameter Index 1=true
   )
   &ScaleColor(
    Scale RGB=&VectorFromFloat(Value=&MakeCustomFloatFromBool(
      Bool=$User.bIsPainted
      True Float=&FloatFromCurve(
       FloatCurve=DataInterface() {propertyValues {Curve {keys [{interpMode=RCIM_Cubic, leaveTangent=-1.6688, tangentMode=RCTM_Break, time=0, value=1.25}, <at 0.9 0.15 RCIM_Cubic>]}}}
       Scale Curve=4
      )
      False Float=&FloatFromCurve(
       FloatCurve=DataInterface() {propertyValues {Curve {keys [
           {arriveTangent=-1.9978, interpMode=RCIM_Cubic, leaveTangent=-1.9978, tangentMode=RCTM_User, time=0, value=1}
           {
            arriveTangentWeight=0.311
            interpMode=RCIM_Cubic
            leaveTangent=-2.5749
            leaveTangentWeight=0.0777
            tangentMode=RCTM_Break
            time=0.8
            value=0.4
           }
           {arriveTangentWeight=0.2828, interpMode=RCIM_Cubic, leaveTangentWeight=0.0471, tangentMode=RCTM_User, time=1, value=0.2}
          ]}}}
       Scale Curve=&RandomRangeFloat(Minimum=5, Maximum=5)
      )
     ))
    ScaleA=true
    ScaleRGB=true
   )
  ]
  rendererClasses [&NiagaraMeshRendererProperties, &NiagaraSpriteRendererProperties]
  renderers [
   {rendererClass=&NiagaraMeshRendererProperties, rendererIndex=0}
   {rendererClass=&NiagaraSpriteRendererProperties, rendererIndex=1}
  ]
 }
 NE_VC_DriftTrail_Ribbon_Parent {
  bEnabled=false
  emitterUpdateScript [
   &EmitterState()
   &VC_SpawnRibbon_v1(
    Number of Ribbons=2
    Spawn Time=-1
    SpawnRate=&MakeCustomFloatFromBool(Bool=$User.bIsFrontendPreview, True Float=50, False Float=0)
   )
  ]
  particleSpawnScript [
   &VC_PaintedVariantsOverride(
    Black=<rgba 0.01 0.01 0.01 1>
    BurntSienna=<rgba 1 0.2116 0 1>
    Cobalt=<rgba 0 0.0667 2 1>
    Crimson=<rgba 0.7 0 0.019 1>
    DefaultColor=<rgba 2 2 2 1>
    ForestGreen=<rgba 0 0.7 0 1>
    Grey=<rgba 0.1 0.1 0.1 1>
    Lime=<rgba 0.4167 1 0 1>
    Orange=<rgba 1.5 0.075 0 1>
    Pink=<rgba 0.8 0.0958 0.5439 1>
    Purple=<rgba 0.1517 0 0.7 10>
    Saffron=<rgba 0.7 0.7 0 1>
    SkyBlue=<rgba 0 0.6654 1 1>
    TitaniumWhite=<rgba 1 1 1 1>
   )
   &VC_InitializeRibbon_v1(
    Color=$Particles.Color
    Custom Offset=true
    Enable Skeletal Position=true
    Flip Facing=true
    Lifetime=0.19
    Local Vector Offset=&AddVectorToPosition(Position=<120 -3 0>, Vector=$User.offset)
    Offset Amount=1
    Ribbon Width=20
   )
   &AddVelocity(Velocity=<6000 0 0>, Velocity Speed Scale=-1) {module=AddVelocity001}
  ]
  particleUpdateScript [
   &ParticleState()
   &InheritSourceMovement(
    Applied Position Scale=&Multiply_VectorByFloat(
     Vector=<1 1 1>
     Float=&FloatFromCurve(FloatCurve=DataInterface() {propertyValues {Curve {keys [{interpMode=RCIM_Cubic, tangentMode=RCTM_User, time=0, value=1}, {interpMode=RCIM_Cubic, tangentMode=RCTM_User, time=1, value=0}]}}})
    )
    ApplyForce=false
    ApplyPosition=true
    ApplyPositionFalloff=false
   )
   &SolveForcesAndVelocity(Clamp Velocity=false, Limit Acceleration=false)
   &VC_UpdateRibbon_v1(Enable Skeletal Position=true)
   &ScaleColor(
    Scale RGB=&VectorFromFloat(Value=&MakeCustomFloatFromBool(
      Bool=$User.bIsPainted
      True Float=&FloatFromCurve(
       FloatCurve=DataInterface() {propertyValues {Curve {keys [{interpMode=RCIM_Cubic, leaveTangent=-3.8624, tangentMode=RCTM_Break, time=0.3, value=1}, <at 1 0 RCIM_Cubic>]}}}
       Scale Curve=2.5
      )
      False Float=&FloatFromCurve(
       FloatCurve=DataInterface() {propertyValues {Curve {keys [
           {arriveTangent=-3.5687, interpMode=RCIM_Cubic, leaveTangent=-3.5687, tangentMode=RCTM_User, time=0.4, value=1}
           {arriveTangent=0.0002, interpMode=RCIM_Cubic, leaveTangent=0.0002, tangentMode=RCTM_User, time=0.9, value=0}
          ]}}}
       Scale Curve=3
      )
     ))
    ScaleA=true
    ScaleRGB=true
   )
   &DynamicMaterialParameters(
    Index 1 Param 1=&MakeCustomFloatFromBool(Bool=$User.bIsPainted, True Float=0, False Float=1)
    Module.Index0Param0Write=true
    Module.Index0Param1Write=true
    Module.Index0Param2Write=true
    Module.Index0Param3Write=true
    Module.Index1Param0Write=true
    Module.Index1Param1Write=true
    Module.Index1Param2Write=true
    Module.Index1Param3Write=true
    Write Parameter Index 1=true
   ) {module=DynamicMaterialParameters001}
   &ScaleRibbonWidth(Ribbon Width Scale=&FloatFromCurve(
     FloatCurve=DataInterface() {propertyValues {Curve {keys [
         {arriveTangent=-1.1867, interpMode=RCIM_Cubic, leaveTangent=-1.1867, tangentMode=RCTM_User, time=0, value=1}
         {arriveTangent=-1.0804, interpMode=RCIM_Cubic, leaveTangent=-1.0804, tangentMode=RCTM_User, time=0.1, value=0.9}
         <at 1 0.25 RCIM_Cubic>
        ]}}}
     Scale Curve=&MakeCustomFloatFromBool(Bool=$User.bIsPainted, True Float=1.5, False Float=1)
    ))
  ]
  rendererClasses [&NiagaraRibbonRendererProperties]
  renderers [{rendererClass=&NiagaraRibbonRendererProperties}]
 }
 NE_VC_DriftTrail_Ribbon_Parent001 {
  emitterUpdateScript [
   &EmitterState()
   &VC_RateQualityHandler(High Spawn Rate=60, Low Spawn Rate=50, Medium Spawn Rate=50, Use Custom Rate?=true)
   &VC_SpawnRibbon_v1(
    Number of Ribbons=2
    Spawn Time=-1
    SpawnRate=&MakeCustomFloatFromBool(Bool=$User.bIsFrontendPreview, True Float=0, False Float=$Emitter.Spawn Rate)
   )
   /SetVariables(Emitter.IS BR=HlslExpression() {hlslExpression="User.GameModeID == 1 ? true : false"}) {isSetParameters=true}
  ]
  particleSpawnScript [
   &VC_PaintedVariantsOverride(
    Black=<rgba 0.01 0.01 0.01 1>
    BurntSienna=<rgba 1 0.2116 0 1>
    Cobalt=<rgba 0 0.0667 2 1>
    Crimson=<rgba 0.7 0 0.019 1>
    DefaultColor=<rgba 2 2 2 1>
    ForestGreen=<rgba 0 0.7 0 1>
    Grey=<rgba 0.1 0.1 0.1 1>
    Lime=<rgba 0.4167 1 0 1>
    Orange=<rgba 1.5 0.075 0 1>
    Pink=<rgba 0.8 0.0958 0.5439 1>
    Purple=<rgba 0.1517 0 0.7 10>
    Saffron=<rgba 0.7 0.7 0 1>
    SkyBlue=<rgba 0 0.6654 1 1>
    TitaniumWhite=<rgba 1 1 1 1>
   )
   &VC_InitializeRibbon_v1(
    Color=$Particles.Color
    Custom Offset=true
    Enable Skeletal Position=true
    Flip Facing=true
    Lifetime=0.65
    Local Vector Offset=&AddVectorToPosition(Position=<25 -3 10>, Vector=$User.offset)
    Offset Amount=1
    Ribbon Width=20
   )
   &AddVelocity(Velocity=<1600 0 0>, Velocity Speed Scale=-1) {module=AddVelocity001}
  ]
  particleUpdateScript [
   &ParticleState()
   &InheritSourceMovement(
    Applied Position Scale=&Multiply_VectorByFloat(
     Vector=<1 1 1>
     Float=&FloatFromCurve(FloatCurve=DataInterface() {propertyValues {Curve {keys [
          {interpMode=RCIM_Cubic, leaveTangentWeight=0.3706, tangentMode=RCTM_User, time=0.5, value=1}
          {interpMode=RCIM_Cubic, leaveTangent=-0.1455, time=1, value=0}
         ]}}})
    )
    ApplyForce=false
    ApplyPosition=true
    ApplyPositionFalloff=false
   )
   &SolveForcesAndVelocity(Clamp Velocity=false, Limit Acceleration=false)
   &VC_UpdateRibbon_v1(Enable Skeletal Position=true)
   &ScaleColor(
    Scale RGB=&VectorFromFloat(Value=&MakeCustomFloatFromBool(
      Bool=$User.bIsPainted
      True Float=&FloatFromCurve(
       FloatCurve=DataInterface() {propertyValues {Curve {keys [{interpMode=RCIM_Cubic, leaveTangent=-3.8624, tangentMode=RCTM_Break, time=0.3, value=1}, <at 1 0.1 RCIM_Cubic>]}}}
       Scale Curve=2.5
      )
      False Float=&FloatFromCurve(
       FloatCurve=DataInterface() {propertyValues {Curve {keys [
           {arriveTangent=-1.2481, interpMode=RCIM_Cubic, leaveTangent=-1.2481, tangentMode=RCTM_User, time=0, value=1}
           {arriveTangent=0.0002, interpMode=RCIM_Cubic, leaveTangent=0.0002, tangentMode=RCTM_User, time=1, value=0.1}
          ]}}}
       Scale Curve=4
      )
     ))
    ScaleA=true
    ScaleRGB=true
   )
   &DynamicMaterialParameters(
    Index 1 Param 1=&MakeCustomFloatFromBool(Bool=$User.bIsPainted, True Float=0, False Float=1)
    Module.Index0Param0Write=true
    Module.Index0Param1Write=true
    Module.Index0Param2Write=true
    Module.Index0Param3Write=true
    Module.Index1Param0Write=true
    Module.Index1Param1Write=true
    Module.Index1Param2Write=true
    Module.Index1Param3Write=true
    Write Parameter Index 1=true
   ) {module=DynamicMaterialParameters001}
   &ScaleRibbonWidth(Ribbon Width Scale=&FloatFromCurve(
     FloatCurve=DataInterface() {propertyValues {Curve {keys [{arriveTangent=-0.1091, interpMode=RCIM_Cubic, leaveTangent=-0.1091, tangentMode=RCTM_User, time=0, value=1}, <at 1 0.25 RCIM_Cubic>]}}}
     Scale Curve=&MakeCustomFloatFromBool(Bool=$User.bIsPainted, True Float=1.5, False Float=1.5)
    ))
  ]
  rendererClasses [&NiagaraRibbonRendererProperties]
  renderers [{rendererClass=&NiagaraRibbonRendererProperties}]
 }
 NE_VC_DriftTrail_Ribbon_Parent002 {
  bEnabled=false
  emitterUpdateScript [
   &EmitterState()
   &VC_SpawnRibbon_v1(
    Number of Ribbons=2
    Spawn Time=-1
    SpawnRate=&MakeCustomFloatFromBool(Bool=$User.bIsFrontendPreview, True Float=0, False Float=100)
   )
  ]
  particleSpawnScript [
   &VC_PaintedVariantsOverride(
    Black=<rgba 0.01 0.01 0.01 1>
    BurntSienna=<rgba 1 0.2116 0 1>
    Cobalt=<rgba 0 0.0667 2 1>
    Crimson=<rgba 0.7 0 0.019 1>
    DefaultColor=<rgba 2 2 2 1>
    ForestGreen=<rgba 0 0.7 0 1>
    Grey=<rgba 0.1 0.1 0.1 1>
    Lime=<rgba 0.4167 1 0 1>
    Orange=<rgba 1.5 0.075 0 1>
    Pink=<rgba 0.8 0.0958 0.5439 1>
    Purple=<rgba 0.1517 0 0.7 10>
    Saffron=<rgba 0.7 0.7 0 1>
    SkyBlue=<rgba 0 0.6654 1 1>
    TitaniumWhite=<rgba 1 1 1 1>
   )
   &VC_InitializeRibbon_v1(
    Color=$Particles.Color
    Custom Offset=true
    Enable Skeletal Position=true
    Flip Facing=true
    Lifetime=0.65
    Local Vector Offset=&AddVectorToPosition(Position=<25 -3 10>, Vector=$User.offset)
    Offset Amount=1
    Ribbon Width=20
   )
   &AddVelocity(Velocity=<1600 0 0>, Velocity Speed Scale=-1) {module=AddVelocity001}
  ]
  particleUpdateScript [
   &ParticleState()
   &InheritSourceMovement(
    Applied Position Scale=&Multiply_VectorByFloat(
     Vector=<1 1 1>
     Float=&FloatFromCurve(FloatCurve=DataInterface() {propertyValues {Curve {keys [
          {interpMode=RCIM_Cubic, leaveTangentWeight=0.3706, tangentMode=RCTM_User, time=0.5, value=1}
          {interpMode=RCIM_Cubic, leaveTangent=-0.1455, time=1, value=0}
         ]}}})
    )
    ApplyForce=false
    ApplyPosition=true
    ApplyPositionFalloff=false
   ) {enabled=false}
   &SolveForcesAndVelocity(Clamp Velocity=false, Limit Acceleration=false)
   &VC_UpdateRibbon_v1(Enable Skeletal Position=true)
   &ScaleColor(
    Scale RGB=&VectorFromFloat(Value=&MakeCustomFloatFromBool(
      Bool=$User.bIsPainted
      True Float=&FloatFromCurve(
       FloatCurve=DataInterface() {propertyValues {Curve {keys [{interpMode=RCIM_Cubic, leaveTangent=-3.8624, tangentMode=RCTM_Break, time=0.3, value=1}, <at 1 0.1 RCIM_Cubic>]}}}
       Scale Curve=2.5
      )
      False Float=&FloatFromCurve(
       FloatCurve=DataInterface() {propertyValues {Curve {keys [
           {arriveTangent=-1.2481, interpMode=RCIM_Cubic, leaveTangent=-1.2481, tangentMode=RCTM_User, time=0, value=1}
           {arriveTangent=0.0002, interpMode=RCIM_Cubic, leaveTangent=0.0002, tangentMode=RCTM_User, time=1, value=0.1}
          ]}}}
       Scale Curve=4
      )
     ))
    ScaleA=true
    ScaleRGB=true
   )
   &DynamicMaterialParameters(
    Index 1 Param 1=&MakeCustomFloatFromBool(Bool=$User.bIsPainted, True Float=0, False Float=1)
    Module.Index0Param0Write=true
    Module.Index0Param1Write=true
    Module.Index0Param2Write=true
    Module.Index0Param3Write=true
    Module.Index1Param0Write=true
    Module.Index1Param1Write=true
    Module.Index1Param2Write=true
    Module.Index1Param3Write=true
    Write Parameter Index 1=true
   ) {module=DynamicMaterialParameters001}
   &ScaleRibbonWidth(Ribbon Width Scale=&FloatFromCurve(
     FloatCurve=DataInterface() {propertyValues {Curve {keys [{arriveTangent=-0.1091, interpMode=RCIM_Cubic, leaveTangent=-0.1091, tangentMode=RCTM_User, time=0, value=1}, <at 1 0.25 RCIM_Cubic>]}}}
     Scale Curve=&MakeCustomFloatFromBool(Bool=$User.bIsPainted, True Float=1.5, False Float=1.5)
    ))
  ]
  rendererClasses [&NiagaraRibbonRendererProperties]
  renderers [{rendererClass=&NiagaraRibbonRendererProperties}]
 }
 NE_VC_DriftTrail_Ribbon_Parent003 {
  emitterUpdateScript [
   &EmitterState()
   &VC_SpawnRibbon_v1(Number of Ribbons=2, Spawn Time=-1, SpawnRate=100)
   /SetVariables(Emitter.IS RR=HlslExpression() {hlslExpression="User.GameModeID == 2 ? true : false"}) {isSetParameters=true}
  ]
  particleSpawnScript [
   &VC_PaintedVariantsOverride(
    Black=<rgba 0.01 0.01 0.01 1>
    BurntSienna=<rgba 1 0.2116 0 1>
    Cobalt=<rgba 0 0.0667 2 1>
    Crimson=<rgba 0.7 0 0.019 1>
    DefaultColor=<rgba 2 2 2 1>
    ForestGreen=<rgba 0 0.7 0 1>
    Grey=<rgba 0.1 0.1 0.1 1>
    Lime=<rgba 0.4167 1 0 1>
    Orange=<rgba 1.5 0.075 0 1>
    Pink=<rgba 0.8 0.0958 0.5439 1>
    Purple=<rgba 0.1517 0 0.7 10>
    Saffron=<rgba 0.7 0.7 0 1>
    SkyBlue=<rgba 0 0.6654 1 1>
    TitaniumWhite=<rgba 1 1 1 1>
   )
   &VC_InitializeRibbon_v1(
    Color=$Particles.Color
    Custom Offset=true
    Enable Skeletal Position=true
    Flip Facing=true
    Lifetime=0.65
    Local Vector Offset=&AddVectorToPosition(Position=<25 -3 0>, Vector=$User.offset)
    Offset Amount=1
    Ribbon Width=20
   )
   &AddVelocity(Velocity=<1600 0 0>, Velocity Speed Scale=-1) {module=AddVelocity001}
  ]
  particleUpdateScript [
   &ParticleState()
   &InheritSourceMovement(
    Applied Position Scale=&Multiply_VectorByFloat(
     Vector=<1 1 1>
     Float=&FloatFromCurve(FloatCurve=DataInterface() {propertyValues {Curve {keys [
          {interpMode=RCIM_Cubic, leaveTangentWeight=0.3706, tangentMode=RCTM_User, time=0.5, value=1}
          {interpMode=RCIM_Cubic, leaveTangent=-0.1455, time=1, value=0}
         ]}}})
    )
    ApplyForce=false
    ApplyPosition=true
    ApplyPositionFalloff=false
   ) {enabled=false}
   &SolveForcesAndVelocity(Clamp Velocity=false, Limit Acceleration=false)
   &VC_UpdateRibbon_v1(Enable Skeletal Position=true)
   &ScaleColor(
    Scale RGB=&VectorFromFloat(Value=&MakeCustomFloatFromBool(
      Bool=$User.bIsPainted
      True Float=&FloatFromCurve(
       FloatCurve=DataInterface() {propertyValues {Curve {keys [{interpMode=RCIM_Cubic, leaveTangent=-3.8624, tangentMode=RCTM_Break, time=0.3, value=1}, <at 1 0.1 RCIM_Cubic>]}}}
       Scale Curve=2.5
      )
      False Float=&FloatFromCurve(
       FloatCurve=DataInterface() {propertyValues {Curve {keys [
           {arriveTangent=-1.2481, interpMode=RCIM_Cubic, leaveTangent=-1.2481, tangentMode=RCTM_User, time=0, value=1}
           {arriveTangent=0.0002, interpMode=RCIM_Cubic, leaveTangent=0.0002, tangentMode=RCTM_User, time=1, value=0.1}
          ]}}}
       Scale Curve=4
      )
     ))
    ScaleA=true
    ScaleRGB=true
   )
   &DynamicMaterialParameters(
    Index 1 Param 1=&MakeCustomFloatFromBool(Bool=$User.bIsPainted, True Float=0, False Float=1)
    Module.Index0Param0Write=true
    Module.Index0Param1Write=true
    Module.Index0Param2Write=true
    Module.Index0Param3Write=true
    Module.Index1Param0Write=true
    Module.Index1Param1Write=true
    Module.Index1Param2Write=true
    Module.Index1Param3Write=true
    Write Parameter Index 1=true
   ) {module=DynamicMaterialParameters001}
   &ScaleRibbonWidth(Ribbon Width Scale=&FloatFromCurve(
     FloatCurve=DataInterface() {propertyValues {Curve {keys [{arriveTangent=-0.1091, interpMode=RCIM_Cubic, leaveTangent=-0.1091, tangentMode=RCTM_User, time=0, value=1}, <at 1 0.25 RCIM_Cubic>]}}}
     Scale Curve=&MakeCustomFloatFromBool(Bool=$User.bIsPainted, True Float=1.5, False Float=1.5)
    ))
  ]
  rendererClasses [&NiagaraRibbonRendererProperties]
  renderers [{rendererClass=&NiagaraRibbonRendererProperties}]
 }
 NE_VC_DriftTrail_Ribbon_Parent004 {
  emitterUpdateScript [
   &EmitterState()
   &VC_SpawnRibbon_v1(Number of Ribbons=2, Spawn Time=-1, SpawnRate=60)
   /SetVariables(Emitter.IS FRONTEND=HlslExpression() {hlslExpression="User.GameModeID == 0 ? true : false"}) {isSetParameters=true}
  ]
  particleSpawnScript [
   &VC_PaintedVariantsOverride(
    Black=<rgba 0.01 0.01 0.01 1>
    BurntSienna=<rgba 1 0.2116 0 1>
    Cobalt=<rgba 0 0.0667 2 1>
    Crimson=<rgba 0.7 0 0.019 1>
    DefaultColor=<rgba 2 2 2 1>
    ForestGreen=<rgba 0 0.7 0 1>
    Grey=<rgba 0.1 0.1 0.1 1>
    Lime=<rgba 0.4167 1 0 1>
    Orange=<rgba 1.5 0.114 0.015 1>
    Pink=<rgba 0.8 0.0958 0.5439 1>
    Purple=<rgba 0.1517 0 0.7 10>
    Saffron=<rgba 0.7 0.7 0 1>
    SkyBlue=<rgba 0 0.6654 1 1>
    TitaniumWhite=<rgba 1 1 1 1>
   )
   &VC_InitializeRibbon_v1(
    Color=$Particles.Color
    Custom Offset=true
    Enable Skeletal Position=true
    Flip Facing=true
    Lifetime=0.181
    Local Vector Offset=&AddVectorToPosition(Position=<120 -3 0>, Vector=$User.offset)
    Offset Amount=1
    Ribbon Width=20
   )
   &AddVelocity(Velocity=<6000 0 0>, Velocity Speed Scale=-1) {module=AddVelocity001}
  ]
  particleUpdateScript [
   &ParticleState()
   &InheritSourceMovement(
    Applied Position Scale=&Multiply_VectorByFloat(
     Vector=<1 1 1>
     Float=&FloatFromCurve(FloatCurve=DataInterface() {propertyValues {Curve {keys [{interpMode=RCIM_Cubic, tangentMode=RCTM_User, time=0, value=1}, {interpMode=RCIM_Cubic, tangentMode=RCTM_User, time=1, value=0}]}}})
    )
    ApplyForce=false
    ApplyPosition=true
    ApplyPositionFalloff=false
   )
   &SolveForcesAndVelocity(Clamp Velocity=false, Limit Acceleration=false)
   &VC_UpdateRibbon_v1(Enable Skeletal Position=true)
   &ScaleColor(
    Scale RGB=&VectorFromFloat(Value=&MakeCustomFloatFromBool(
      Bool=$User.bIsPainted
      True Float=&FloatFromCurve(
       FloatCurve=DataInterface() {propertyValues {Curve {keys [
           {interpMode=RCIM_Cubic, leaveTangent=-3.8624, tangentMode=RCTM_Break, time=0.4, value=1}
           {arriveTangent=-0.8405, interpMode=RCIM_Cubic, leaveTangent=-6.2986, tangentMode=RCTM_Break, time=0.9, value=0.15}
           {arriveTangentWeight=0.0417, interpMode=RCIM_Cubic, tangentMode=RCTM_User, time=1, value=0}
          ]}}}
       Scale Curve=2.5
      )
      False Float=&FloatFromCurve(
       FloatCurve=DataInterface() {propertyValues {Curve {keys [
           {arriveTangent=-5.0663, interpMode=RCIM_Cubic, leaveTangent=-5.0663, tangentMode=RCTM_User, time=0.4, value=0.9}
           {arriveTangent=0.0002, interpMode=RCIM_Cubic, leaveTangent=0.0002, tangentMode=RCTM_User, time=0.9, value=0.15}
          ]}}}
       Scale Curve=3
      )
     ))
    ScaleA=true
    ScaleRGB=true
   )
   &DynamicMaterialParameters(
    Index 1 Param 1=&MakeCustomFloatFromBool(Bool=$User.bIsPainted, True Float=0, False Float=0.7)
    Index 1 Param 2=&MakeCustomFloatFromBool(Bool=$User.bIsPainted, True Float=0.45, False Float=0.7)
    Module.Index0Param0Write=true
    Module.Index0Param1Write=true
    Module.Index0Param2Write=true
    Module.Index0Param3Write=true
    Module.Index1Param0Write=true
    Module.Index1Param1Write=true
    Module.Index1Param2Write=true
    Module.Index1Param3Write=true
    Write Parameter Index 1=true
   ) {module=DynamicMaterialParameters001}
   &ScaleRibbonWidth(Ribbon Width Scale=&FloatFromCurve(
     FloatCurve=DataInterface() {propertyValues {Curve {keys [
         {arriveTangent=-1.1867, interpMode=RCIM_Cubic, leaveTangent=-1.1867, tangentMode=RCTM_User, time=0, value=1}
         {arriveTangent=-1.0804, interpMode=RCIM_Cubic, leaveTangent=-1.0804, tangentMode=RCTM_User, time=0.1, value=0.9}
         <at 1 0.25 RCIM_Cubic>
        ]}}}
     Scale Curve=&MakeCustomFloatFromBool(Bool=$User.bIsPainted, True Float=1.5, False Float=1.2)
    ))
  ]
  rendererClasses [&NiagaraRibbonRendererProperties]
  renderers [{rendererClass=&NiagaraRibbonRendererProperties}]
 }
 NE_VC_DriftTrail_WheelWrap_Parent001 {
  emitterUpdateScript [
   &EmitterState(Life Cycle Mode=::NewEnumerator1, Loop Delay=0.1, Loop Duration=0.2, Recalculate Duration Each Loop=true, UseLoopDelay=false)
   &SpawnRate(SpawnRate=4, Use Spawn Probability=false) {enabled=false}
   &SpawnBurst_Instantaneous(Use Spawn Probability=true, UseLoopCountLimit=false)
   &VC_WheelWrapScale()
  ]
  particleSpawnScript [
   &VC_PaintedVariantsOverride(
    Black=<rgba 0.01 0.01 0.01 1>
    BurntSienna=<rgba 1 0.2116 0 1>
    Cobalt=<rgba 0 0.0667 2 1>
    Crimson=<rgba 0.7 0 0.019 1>
    ForestGreen=<rgba 0 0.7 0 1>
    Grey=<rgba 0.075 0.075 0.075 1>
    Lime=<rgba 0.4167 1 0 1>
    Orange=<rgba 1.5 0.075 0 1>
    Pink=<rgba 0.8 0.0958 0.5439 1>
    Purple=<rgba 0.1517 0 0.7 10>
    Saffron=<rgba 0.7 0.7 0 1>
    SkyBlue=<rgba 0 0.6654 1 1>
    TitaniumWhite=<rgba 1 1 1 1>
   )
   &InitializeParticle(
    Color=$Particles.Color
    Lifetime=0.2
    Mesh Scale=$Emitter.WheelFXScale
    Mesh Scale Mode=::NewEnumerator4
    Sprite Rotation Mode=::NewEnumerator2
    Sprite Size Mode=::NewEnumerator3
    UsePositionOffset=false
    Write Material Random=true
   )
   &PartitionParticles(Number Of Partitions=1)
   &SkeletalMeshLocation(
    Bone / Socket Sampling Mode=::NewEnumerator5
    Interpolate Mesh Sampling=false
    MaskSpawnGroup=false
    Mesh Sampling Type=::NewEnumerator6
    Skeletal Mesh=$System.SkeletalMesh
    Use Sampled Mesh Attribute as Ribbon ID=false
   )
   &InitialMeshOrientation(Mesh Orientation Mode=::NewEnumerator2, Rotation Coordinate Space=::NewEnumerator2, Use Rotation Vector=true)
  ]
  particleUpdateScript [
   &ParticleState()
   &UpdateMeshOrientation(Delta Time:NiagaraFloat=$Emitter.LoopedAge, Rotation Vector=<0 1 0>)
   &SolveForcesAndVelocity(Clamp Velocity=false, Limit Acceleration=false)
   &ScaleColor(
    Scale RGB=&VectorFromFloat(Value=&FloatFromCurve(
      FloatCurve=DataInterface() {propertyValues {Curve {keys [
          {interpMode=RCIM_Cubic, tangentMode=RCTM_User, time=0, value=1}
          {interpMode=RCIM_Cubic, tangentMode=RCTM_User, time=0.5, value=0}
          {tangentMode=RCTM_User, time=1, value=1}
         ]}}}
      Scale Curve=4
     ))
    ScaleA=true
    ScaleRGB=true
   ) {enabled=false}
   &ScaleColor(
    Scale RGB=&VectorFromFloat(Value=&MakeCustomFloatFromBool(
      Bool=$User.bIsPainted
      True Float=&FloatFromCurve(FloatCurve=DataInterface() {propertyValues {Curve {keys [<at 0 1 RCIM_Cubic>,<at 0.5 0.2 RCIM_Cubic>,<at 1 1 RCIM_Cubic>]}}}, Scale Curve=4)
      False Float=&FloatFromCurve(
       FloatCurve=DataInterface() {propertyValues {Curve {keys [
           {interpMode=RCIM_Cubic, tangentMode=RCTM_User, time=0, value=1}
           {interpMode=RCIM_Cubic, tangentMode=RCTM_User, time=0.5, value=0.1}
           {tangentMode=RCTM_User, time=1, value=1}
          ]}}}
       Scale Curve=4
      )
     ))
    ScaleA=true
    ScaleRGB=true
   ) {module=ScaleColor001}
   &DynamicMaterialParameters(
    Index 1 Param 1=&MakeCustomFloatFromBool(Bool=$User.bIsPainted, True Float=0, False Float=1)
    Module.Index0Param0Write=true
    Module.Index0Param1Write=true
    Module.Index0Param2Write=true
    Module.Index0Param3Write=true
    Module.Index1Param0Write=true
    Module.Index1Param1Write=true
    Module.Index1Param2Write=true
    Module.Index1Param3Write=true
    Write Parameter Index 1=true
   )
  ]
  rendererClasses [&NiagaraMeshRendererProperties]
  renderers [{rendererClass=&NiagaraMeshRendererProperties, rendererIndex=0}]
 }
 NE_VC_DriftTrail_WheelWrap_Parent002 {
  emitterUpdateScript [
   &EmitterState(Life Cycle Mode=::NewEnumerator1, Loop Delay=0.1, Loop Duration=0.2, Recalculate Duration Each Loop=true, UseLoopDelay=false)
   &SpawnRate(SpawnRate=4, Use Spawn Probability=false) {enabled=false}
   &SpawnBurst_Instantaneous(Use Spawn Probability=true, UseLoopCountLimit=false)
   &VC_WheelWrapScale()
  ]
  particleSpawnScript [
   &VC_PaintedVariantsOverride(
    Black=<rgba 0.01 0.01 0.01 1>
    BurntSienna=<rgba 1 0.2116 0 1>
    Cobalt=<rgba 0 0.0667 2 1>
    Crimson=<rgba 0.7 0 0.019 1>
    ForestGreen=<rgba 0 0.7 0 1>
    Grey=<rgba 0.075 0.075 0.075 1>
    Lime=<rgba 0.4167 1 0 1>
    Orange=<rgba 1.5 0.075 0 1>
    Pink=<rgba 0.8 0.0958 0.5439 1>
    Purple=<rgba 0.1517 0 0.7 10>
    Saffron=<rgba 0.7 0.7 0 1>
    SkyBlue=<rgba 0 0.6654 1 1>
    TitaniumWhite=<rgba 1 1 1 1>
   )
   &InitializeParticle(
    Color=$Particles.Color
    Lifetime=0.2
    Mesh Scale=$Emitter.WheelFXScale
    Mesh Scale Mode=::NewEnumerator4
    Sprite Rotation Mode=::NewEnumerator2
    Sprite Size Mode=::NewEnumerator3
    UsePositionOffset=false
    Write Material Random=true
   )
   &PartitionParticles(Number Of Partitions=1)
   &SkeletalMeshLocation(
    Bone / Socket Index=1
    Bone / Socket Sampling Mode=::NewEnumerator5
    Interpolate Mesh Sampling=false
    MaskSpawnGroup=false
    Mesh Sampling Type=::NewEnumerator6
    Skeletal Mesh=$System.SkeletalMesh
    Use Sampled Mesh Attribute as Ribbon ID=false
   )
   &InitialMeshOrientation(Mesh Orientation Mode=::NewEnumerator2, Rotation Coordinate Space=::NewEnumerator2, Use Rotation Vector=true)
  ]
  particleUpdateScript [
   &ParticleState()
   &ScaleMeshSize(Scale Factor=&VectorFromCurve(VectorCurve=DataInterface() {propertyValues {
       XCurve {keys [<at 0 1>,<at 1 1>]}
       YCurve {keys [<at 0 1 RCIM_Cubic>, {arriveTangentWeight=0.3333, interpMode=RCIM_Cubic, tangentMode=RCTM_User, time=1, value=0.4}]}
       ZCurve {keys [<at 0 1>,<at 1 1>]}
      }})) {enabled=false, module=ScaleMeshSize001}
   &UpdateMeshOrientation(Delta Time:NiagaraFloat=$Emitter.LoopedAge, Rotation Vector=<0 1 0>)
   &SolveForcesAndVelocity(Clamp Velocity=false, Limit Acceleration=false)
   &ScaleColor(
    Scale RGB=&VectorFromFloat(Value=&FloatFromCurve(
      FloatCurve=DataInterface() {propertyValues {Curve {keys [
          {interpMode=RCIM_Cubic, tangentMode=RCTM_User, time=0, value=1}
          {interpMode=RCIM_Cubic, tangentMode=RCTM_User, time=0.5, value=0}
          {tangentMode=RCTM_User, time=1, value=1}
         ]}}}
      Scale Curve=4
     ))
    ScaleA=true
    ScaleRGB=true
   ) {enabled=false}
   &ScaleColor(
    Scale RGB=&VectorFromFloat(Value=&MakeCustomFloatFromBool(
      Bool=$User.bIsPainted
      True Float=&FloatFromCurve(FloatCurve=DataInterface() {propertyValues {Curve {keys [<at 0 1 RCIM_Cubic>,<at 0.5 0.1 RCIM_Cubic>,<at 1 1 RCIM_Cubic>]}}}, Scale Curve=4)
      False Float=&FloatFromCurve(
       FloatCurve=DataInterface() {propertyValues {Curve {keys [
           {interpMode=RCIM_Cubic, tangentMode=RCTM_User, time=0, value=1}
           {interpMode=RCIM_Cubic, tangentMode=RCTM_User, time=0.5, value=0.1}
           {tangentMode=RCTM_User, time=1, value=1}
          ]}}}
       Scale Curve=4
      )
     ))
    ScaleA=true
    ScaleRGB=true
   ) {module=ScaleColor001}
   &DynamicMaterialParameters(
    Index 1 Param 1=&MakeCustomFloatFromBool(Bool=$User.bIsPainted, True Float=0, False Float=1)
    Module.Index0Param0Write=true
    Module.Index0Param1Write=true
    Module.Index0Param2Write=true
    Module.Index0Param3Write=true
    Module.Index1Param0Write=true
    Module.Index1Param1Write=true
    Module.Index1Param2Write=true
    Module.Index1Param3Write=true
    Write Parameter Index 1=true
   )
  ]
  rendererClasses [&NiagaraMeshRendererProperties]
  renderers [{rendererClass=&NiagaraMeshRendererProperties, rendererIndex=0}]
 }
 NE_VC_DriftTrail_WheelWrap_Parent1 {
  emitterUpdateScript [
   &EmitterState(
    Inactive Response=::NewEnumerator1
    Life Cycle Mode=::NewEnumerator1
    Loop Behavior=::NewEnumerator1
    Loop Duration Mode=::NewEnumerator1
    UseLoopDelay=false
   )
   &SpawnBurst_Instantaneous(Spawn Count=4, Use Spawn Probability=false, UseLoopCountLimit=false)
   &VC_WheelWrapScale()
  ]
  particleSpawnScript [
   &VC_PaintedVariantsOverride(
    Black=<rgba 0.008 0.008 0.008 1>
    BurntSienna=<rgba 0.25 0.0625 0 1>
    Cobalt=<rgba 0 0.125 1.5 1>
    DefaultColor=&SelectLinearColorFromArray(
     Array Sampling Mode=::NewEnumerator0
     Direct Array Index=?
     Color Selection Array=DataInterface() {propertyValues {ColorData [<rgba 1 0.0712 0.8849 1>, <rgba 0.0275 0.3255 1 1>]}}
     Random Seed=0
    )
    ForestGreen=<rgba 0 0.5 0 1>
    Grey=<rgba 0.05 0.05 0.05 1>
    Lime=<rgba 0.3433 0.5885 0 1>
    Orange=<rgba 1.25 0.0833 0 1>
    Pink=<rgba 0.5052 0.101 0.3582 1>
    Saffron=<rgba 0.9 0.825 0 1>
    TitaniumWhite=<rgba 1 1 1 1>
   )
   &InitializeParticle(
    Color=$Particles.Color
    Mesh Scale=$Emitter.WheelFXScale
    Mesh Scale Mode=::NewEnumerator4
    Sprite Rotation Mode=::NewEnumerator2
    Sprite Size Mode=::NewEnumerator3
    UsePositionOffset=false
    Write Material Random=true
   )
   &PartitionParticles()
   &SkeletalMeshLocation(
    Bone / Socket Index=$Particles.PartitionParticles.Partition
    Bone / Socket Sampling Mode=::NewEnumerator5
    Interpolate Mesh Sampling=false
    MaskSpawnGroup=false
    Mesh Sampling Type=::NewEnumerator6
    Skeletal Mesh=$System.SkeletalMesh
   )
   &InitialMeshOrientation(Mesh Orientation Mode=::NewEnumerator2, Rotation Coordinate Space=::NewEnumerator2, Use Rotation Vector=true)
  ]
  particleUpdateScript [
   &ParticleState(Kill Particles When Lifetime Has Elapsed=false, Loop Particles Lifetime=true)
   &SolveForcesAndVelocity(Clamp Velocity=false, Limit Acceleration=false)
   &ScaleColor(Scale RGB=<10 10 10>, ScaleA=true, ScaleRGB=true) {module=ScaleColor001}
   &DynamicMaterialParameters(
    Index 0 Param 1=&FloatFromCurve(FloatCurve=DataInterface() {propertyValues {Curve {keys [<at 0 1>,<at 1 0>]}}})
    Index 0 Param 2=&FloatFromCurve(
     FloatCurve=DataInterface() {propertyValues {Curve {keys [{tangentMode=RCTM_User, time=0, value=0}, {tangentMode=RCTM_User, time=1, value=1}]}}}
     Scale Curve=0.1
    )
    Index 1 Param 1=&MakeCustomFloatFromBool(Bool=$User.bIsPainted, True Float=0, False Float=1)
    Index 1 Param 2=&MakeCustomFloatFromBool(Bool=$User.bIsPainted, True Float=&MakeFloatFromInt(Int=$User.Lines_Glow), False Float=100)
    Index 2 Param 1=&MakeCustomFloatFromBool(Bool=$User.bIsPainted, True Float=0, False Float=1)
    Index 2 Param 2=&MakeCustomFloatFromBool(Bool=$User.bIsPainted, True Float=&MakeFloatFromInt(Int=$User.Lines_Glow), False Float=100)
    Module.Index0Param0Write=true
    Module.Index0Param1Write=true
    Module.Index0Param2Write=true
    Module.Index0Param3Write=true
    Module.Index1Param0Write=true
    Module.Index1Param1Write=true
    Module.Index1Param2Write=true
    Module.Index1Param3Write=true
    Module.Index2Param0Write=true
    Module.Index2Param1Write=true
    Module.Index2Param2Write=true
    Module.Index2Param3Write=true
    Write Parameter Index 1=true
    Write Parameter Index 2=true
   )
  ]
  rendererClasses [&NiagaraMeshRendererProperties]
  renderers [{rendererClass=&NiagaraMeshRendererProperties, rendererIndex=0}]
 }
}
summary {
 systemName=NS_DriftTrail_AnimatedLines
 emitters [
  {
   bEnabled=false
   emitterName=NE_VC_DriftTrail_CURVE_1
   rendererClasses [&NiagaraMeshRendererProperties, &NiagaraSpriteRendererProperties]
  }
  {
   bEnabled=false
   emitterName=NE_VC_DriftTrail_CURVE_03
   rendererClasses [&NiagaraMeshRendererProperties, &NiagaraSpriteRendererProperties]
  }
  {
   bEnabled=false
   emitterName=NE_VC_DriftTrail_CURVE_02
   rendererClasses [&NiagaraMeshRendererProperties, &NiagaraSpriteRendererProperties]
  }
  {bEnabled=false, emitterName=NE_VC_DriftTrail_Ribbon_Parent, rendererClasses [&NiagaraRibbonRendererProperties]}
  {emitterName=NE_VC_DriftTrail_WheelWrap_Parent001, rendererClasses [&NiagaraMeshRendererProperties]}
  {emitterName=NE_VC_DriftTrail_WheelWrap_Parent002, rendererClasses [&NiagaraMeshRendererProperties]}
  {
   emitterName=NE_VC_DriftTrail_CURVE001
   rendererClasses [&NiagaraMeshRendererProperties, &NiagaraSpriteRendererProperties]
  }
  {
   emitterName=NE_VC_DriftTrail_CURVE_03002
   rendererClasses [&NiagaraMeshRendererProperties, &NiagaraSpriteRendererProperties]
  }
  {
   emitterName=NE_VC_DriftTrail_CURVE_02001
   rendererClasses [&NiagaraMeshRendererProperties, &NiagaraSpriteRendererProperties]
  }
  {
   emitterName=NE_VC_DriftTrail_CURVE_03003
   rendererClasses [&NiagaraMeshRendererProperties, &NiagaraSpriteRendererProperties]
  }
  {emitterName=NE_VC_DriftTrail_Ribbon_Parent001, rendererClasses [&NiagaraRibbonRendererProperties]}
  {
   bEnabled=false
   emitterName=NE_VC_DriftTrail_CURVE002
   rendererClasses [&NiagaraMeshRendererProperties, &NiagaraSpriteRendererProperties]
  }
  {
   bEnabled=false
   emitterName=NE_VC_DriftTrail_CURVE_03004
   rendererClasses [&NiagaraMeshRendererProperties, &NiagaraSpriteRendererProperties]
  }
  {
   bEnabled=false
   emitterName=NE_VC_DriftTrail_CURVE_02002
   rendererClasses [&NiagaraMeshRendererProperties, &NiagaraSpriteRendererProperties]
  }
  {
   bEnabled=false
   emitterName=NE_VC_DriftTrail_CURVE_03005
   rendererClasses [&NiagaraMeshRendererProperties, &NiagaraSpriteRendererProperties]
  }
  {bEnabled=false, emitterName=NE_VC_DriftTrail_Ribbon_Parent002, rendererClasses [&NiagaraRibbonRendererProperties]}
  {
   bEnabled=false
   emitterName=NE_VC_DriftTrail_CURVE_02003
   rendererClasses [&NiagaraMeshRendererProperties, &NiagaraSpriteRendererProperties]
  }
  {
   emitterName=NE_VC_DriftTrail_CURVE003
   rendererClasses [&NiagaraMeshRendererProperties, &NiagaraSpriteRendererProperties]
  }
  {
   emitterName=NE_VC_DriftTrail_CURVE_03001
   rendererClasses [&NiagaraMeshRendererProperties, &NiagaraSpriteRendererProperties]
  }
  {emitterName=NE_VC_DriftTrail_Ribbon_Parent003, rendererClasses [&NiagaraRibbonRendererProperties]}
  {
   emitterName=NE_VC_DriftTrail_CURVE_02004
   rendererClasses [&NiagaraMeshRendererProperties, &NiagaraSpriteRendererProperties]
  }
  {
   emitterName=NE_VC_DriftTrail_CURVE_03006
   rendererClasses [&NiagaraMeshRendererProperties, &NiagaraSpriteRendererProperties]
  }
  {
   emitterName=NE_VC_DriftTrail_CURVE004
   rendererClasses [&NiagaraMeshRendererProperties, &NiagaraSpriteRendererProperties]
  }
  {
   emitterName=NE_VC_DriftTrail_CURVE_03007
   rendererClasses [&NiagaraMeshRendererProperties, &NiagaraSpriteRendererProperties]
  }
  {
   emitterName=NE_VC_DriftTrail_CURVE_02005
   rendererClasses [&NiagaraMeshRendererProperties, &NiagaraSpriteRendererProperties]
  }
  {emitterName=NE_VC_DriftTrail_Ribbon_Parent004, rendererClasses [&NiagaraRibbonRendererProperties]}
  {
   emitterName=NE_VC_DriftTrail_CURVE_02006
   rendererClasses [&NiagaraMeshRendererProperties, &NiagaraSpriteRendererProperties]
  }
  {emitterName=NE_VC_DriftTrail_WheelWrap_Parent1, rendererClasses [&NiagaraMeshRendererProperties]}
 ]
 userVariables [
  User.bIsFrontendPreview {type=NiagaraBool}
  User.bIsPainted {type=NiagaraBool}
  User.ColorVariantID {type=NiagaraInt32}
  User.GameModeID {type=NiagaraInt32}
  User.Lines_Glow {type=NiagaraInt32, defaultValue=1}
  User.offset {type=Vector3f, description="offset from wheels(set through cosmetic template)", defaultValue {struct=Vector3f, value=<0 0 0>}}
  User.PaintedColor {type=LinearColor, description="", defaultValue {struct=LinearColor, value=<rgba 1 1 1 1>}}
  User.PaintedScalar {type=NiagaraFloat, defaultValue=1, description=""}
  User.WheelSize {type=NiagaraFloat, defaultValue=0.1, description=""}
 ]
}
