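# Niagara system for the AnimatedLines vehicle drift trail cosmetic, attaching animated line meshes, ribbons and wheel wraps to skeletal mesh sockets.
# Looping bursts spawn curve and line particles, paired ribbon emitters carry game and frontend flags, and a GPU emitter draws the wheel wraps.
# Spawn groups merge several line layers into one emitter with per-layer scale and color.
# User flags switch paint, glow and preview tuning, while per-wheel masks disable tire effects.
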NiagaraSystem /VehicleCosmetics/DriftTrail/DriftTrail_AnimatedLines/FX/NS_DriftTrail_AnimatedLines_MP
dependencies {
 usedDataInterfaces [
  &NiagaraDataInterfaceArrayColor
  &NiagaraDataInterfaceMeshRendererInfo
  &NiagaraDataInterfaceCurve
  &NiagaraDataInterfaceArrayFloat3
  &NiagaraDataInterfaceVectorCurve
  &NiagaraDataInterfaceSkeletalMesh
 ]
 usedDynamicInputs []
 usedModules [
  &self:SystemSpawnScript.SystemScriptSource.SystemScriptGraph.NiagaraNodeAssignment_0.SetVariables_#1
  &SystemState
  &FN_UnpackVehicleFlags
  /Game/Effects/Niagara/Modules/Vehicle/FN_UpdateVehicleAttribs
  /Game/Effects/Niagara/Modules/Vehicle/FN_VehicleFlags_Cheatcodes
 ]
}
emitters {
 NE_VC_DriftTrail_CURVE001 {
  emitterUpdateScript [
   &EmitterState(Life Cycle Mode=::NewEnumerator1, Loop Duration=0.05, Recalculate Duration Each Loop=true, UseLoopDelay=false)
   &SpawnBurst_Instantaneous(Spawn Count=6, Use Spawn Probability=false, UseLoopCountLimit=false)
   &SpawnRate(
    SpawnRate=&MakeCustomFloatFromBool(Bool=$User.bIsFrontendPreview, True Float=0, False Float=0)
    Use Spawn Probability=false
   )
  ]
  particleSpawnScript [
   &VC_PaintedVariantsOverride(
    Black=<rgba 0.01 0.01 0.01 1>
    BurntSienna=<rgba 0.25 0.0625 0 1>
    Cobalt=<rgba 0 0.125 1.5 1>
    DefaultColor=&SelectLinearColorFromArray(
     Array Sampling Mode=::NewEnumerator0
     Direct Array Index=?
     Color Selection Array=DataInterface() {propertyValues {ColorData [<rgba 0 1.6667 4 1>, <rgba 1.5 0 0.2 1>]}}
     Random Seed=0
    )
    ForestGreen=<rgba 0 0.7 0 1>
    Grey=<rgba 0.07 0.07 0.07 1>
    Lime=<rgba 0.5208 1.25 0 1>
    Orange=<rgba 1.25 0.0833 0 1>
    Pink=<rgba 0.8 0.0958 0.5439 1>
    Saffron=<rgba 0.7 0.7 0 1>
    SkyBlue=<rgba 0 0.6654 1 1>
    TitaniumWhite=<rgba 1 1 1 1>
   )
   &InitializeParticle(
    Color=$Particles.Color
    Lifetime Max=0.3
    Lifetime Min=0.3
    Lifetime Mode=::NewEnumerator1
    Mesh Scale Max=&RandomRangeVector(Minimum=<4 3 3>, Maximum=<4 5 5>)
    Mesh Scale Min=&RandomRangeVector(Minimum=<2 5 5>, Maximum=<2 5 7>)
    Mesh Scale Mode=::NewEnumerator2
    Sprite Rotation Mode=::NewEnumerator2
    Sprite Size Mode=::NewEnumerator3
    UsePositionOffset=false
    Write Material Random=true
   )
   &PartitionParticles()
   &SkeletalMeshLocation(
    Bone / Socket Index=&Add_Integer(A=$Particles.PartitionParticles.Partition, B=4)
    Bone / Socket Sampling Mode=::NewEnumerator5
    Interpolate Mesh Sampling=false
    MaskSpawnGroup=false
    Mesh Sampling Type=::NewEnumerator6
    Sampled Position Offset=&AddVectorToPosition(Position=<0 5 0>, Vector=$User.offset)
    Skeletal Mesh=$System.SkeletalMesh
   )
   &AddVelocity(Velocity=<1000 0 0>, Velocity Speed Scale=-1) {enabled=false}
   &InitialMeshOrientation(
    Mesh Axis To Orient=&RandomRangeVector(Minimum=<0 -1 1>, Maximum=<0 1 -1>)
    Mesh Orientation Mode=::NewEnumerator2
    Orientation Vector=&RandomRangeVector(Minimum=<0 1 1>, Maximum=<0 -1 1>, Randomness Mode=::NewEnumerator1)
    Rotation=<90 65 0>
    Rotation Coordinate Space=::NewEnumerator2
    Use Rotation Vector=false
   )
  ]
  particleUpdateScript [
   &FN_DisableTireFX(
    Mask Index Attribute:NiagaraInt32=$Particles.PartitionParticles.Partition
    Mask Index FL=2
    Mask Index FR=3
    Mask Index RL=0
    Mask Index RR=1
   )
   &MeshRotationForce(Rotation=&SelectVectorFromArray(
     Direct Array Index=$Particles.PartitionParticles.Partition
     Vector Selection Array=DataInterface() {propertyValues {FloatData [<-10 0 0>, <10 0 0>]}}
    )) {enabled=false}
   &ParticleState()
   &ScaleMeshSize(Scale Factor=&SelectVectorFromArray(
     Direct Array Index=$Particles.PartitionParticles.Partition
     Vector Selection Array=DataInterface() {propertyValues {FloatData [<1 -0.75 0.75>, <1 -0.75 -0.75>]}}
    )) {module=ScaleMeshSize001}
   &DynamicMaterialParameters(
    Index 0 Param 1=&FloatFromCurve(
     FloatCurve=DataInterface() {propertyValues {Curve {keys [
         <at 0.4 0 RCIM_Cubic>
         {arriveTangent=0.002, interpMode=RCIM_Cubic, leaveTangent=0.7645, tangentMode=RCTM_Break, time=0.5, value=0.5}
         {arriveTangent=0.644, tangentMode=RCTM_User, time=1, value=1}
        ]}}}
     Scale Curve=0.9
    )
    Index 0 Param 2=&FloatFromCurve(
     FloatCurve=DataInterface() {propertyValues {Curve {keys [
         <at 0 1 RCIM_Cubic>
         {arriveTangent=-1.3855, interpMode=RCIM_Cubic, leaveTangent=-1.3855, tangentMode=RCTM_User, time=0.5, value=0.3}
         <at 1 0 RCIM_Cubic>
        ]}}}
     Scale Curve=&RandomRangeFloat(Minimum=1, Maximum=1)
    )
    Index 1 Param 1=&MakeCustomFloatFromBool(Bool=$User.bIsPainted, True Float=2, False Float=1)
    Module.Index0Param0Write=true
    Module.Index0Param1Write=true
    Module.Index0Param2Write=true
    Module.Index0Param3Write=true
    Module.Index1Param0Write=true
    Module.Index1Param1Write=true
    Module.Index1Param2Write=true
    Module.Index1Param3Write=true
    Write Parameter Index 1=true
   ) {enabled=false}
   &DynamicMaterialParameters(
    Index 0 Param 1=&FloatFromCurve(
     FloatCurve=DataInterface() {propertyValues {Curve {keys [
         <at 0.4 0 RCIM_Cubic>
         {arriveTangent=0.002, interpMode=RCIM_Cubic, leaveTangent=0.7645, tangentMode=RCTM_Break, time=0.5, value=0.5}
         {arriveTangent=0.644, tangentMode=RCTM_User, time=1, value=1}
        ]}}}
     Scale Curve=0.9
    )
    Index 0 Param 2=&FloatFromCurve(
     FloatCurve=DataInterface() {propertyValues {Curve {keys [
         <at 0 1 RCIM_Cubic>
         {arriveTangent=-1.3855, interpMode=RCIM_Cubic, leaveTangent=-1.3855, tangentMode=RCTM_User, time=0.5, value=0.3}
         <at 1 0 RCIM_Cubic>
        ]}}}
     Scale Curve=&RandomRangeFloat(Minimum=1, Maximum=1)
    )
    Index 1 Param 1=&MakeCustomFloatFromBool(Bool=$User.bIsPainted, True Float=0, False Float=1)
    Index 1 Param 2=&MakeCustomFloatFromBool(
     Bool=$User.bIsPainted
     True Float=&MakeFloatFromInt(Int=$User.Lines_Glow)
     False Float=&MakeCustomFloatFromBool(Bool=$User.bIsFrontendPreview, True Float=50, False Float=25)
    )
    Module.Index0Param0Write=true
    Module.Index0Param1Write=true
    Module.Index0Param2Write=true
    Module.Index0Param3Write=true
    Module.Index1Param0Write=true
    Module.Index1Param1Write=true
    Module.Index1Param2Write=true
    Module.Index1Param3Write=true
    Write Parameter Index 1=true
   ) {module=DynamicMaterialParameters001}
   &ScaleColor(
    Scale RGB=&VectorFromFloat(Value=&MakeCustomFloatFromBool(
      Bool=$User.bIsPainted
      True Float=&FloatFromCurve(
       FloatCurve=DataInterface() {propertyValues {Curve {keys [{interpMode=RCIM_Cubic, leaveTangent=-1.6688, tangentMode=RCTM_Break, time=0, value=1.25}, <at 1 0.15 RCIM_Cubic>]}}}
       Scale Curve=4
      )
      False Float=&FloatFromCurve(
       FloatCurve=DataInterface() {propertyValues {Curve {keys [
           {arriveTangent=-1.9978, interpMode=RCIM_Cubic, leaveTangent=-1.9978, tangentMode=RCTM_User, time=0, value=1}
           {
            arriveTangentWeight=0.311
            interpMode=RCIM_Cubic
            leaveTangent=-2.5749
            leaveTangentWeight=0.0777
            tangentMode=RCTM_Break
            time=0.8
            value=0.4
           }
           {arriveTangentWeight=0.2828, interpMode=RCIM_Cubic, leaveTangentWeight=0.0471, tangentMode=RCTM_User, time=1, value=0.2}
          ]}}}
       Scale Curve=&RandomRangeFloat(Minimum=3, Maximum=4)
      )
     ))
    ScaleA=true
    ScaleRGB=true
   )
   &SolveForcesAndVelocity(Clamp Velocity=false, Limit Acceleration=false)
  ]
  rendererClasses [&NiagaraMeshRendererProperties, &NiagaraSpriteRendererProperties]
  renderers [
   {rendererClass=&NiagaraMeshRendererProperties, rendererIndex=0}
   {rendererClass=&NiagaraSpriteRendererProperties, rendererIndex=1}
  ]
 }
 NE_VC_DriftTrail_Ribbon_Parent {
  emitterUpdateScript [
   &EmitterState()
   &VC_RateQualityHandler(Epic Spawn Rate=90, High Spawn Rate=70, Low Spawn Rate=30, Medium Spawn Rate=45, Use Custom Rate?=true)
   &VC_SpawnRibbon_v1(Number of Ribbons=2, Spawn Time=-1, SpawnRate=$Emitter.Spawn Rate)
   /SetVariables(Emitter.ISGAME:NiagaraBool=&Boolean_NotOperation(Bool=$User.bIsFrontendPreview)) {isSetParameters=true}
  ]
  particleSpawnScript [
   &VC_PaintedVariantsOverride(
    Black=<rgba 0.015 0.015 0.015 1>
    BurntSienna=<rgba 1 0.2116 0 1>
    Cobalt=<rgba 0 0.0667 2 1>
    Crimson=<rgba 0.7 0 0.019 1>
    DefaultColor=<rgba 2 2 2 1>
    ForestGreen=<rgba 0 0.7 0 1>
    Grey=<rgba 0.1 0.1 0.1 1>
    Lime=<rgba 0.4167 1 0 1>
    Orange=<rgba 1.5 0.114 0.015 1>
    Pink=<rgba 0.8 0.0958 0.5439 1>
    Purple=<rgba 0.325 0 1.5 10>
    Saffron=<rgba 0.7 0.7 0 1>
    SkyBlue=<rgba 0 0.6654 1 1>
    TitaniumWhite=<rgba 1 1 1 1>
   )
   &VC_InitializeRibbon_v1(
    Color=$Particles.Color
    Custom Offset=true
    Enable Skeletal Position=true
    Lifetime=&MakeCustomFloatFromBool(Bool=$User.bIsFrontendPreview, True Float=0.19, False Float=0.4)
    Local Vector Offset=&AddVectorToPosition(Position=<0 0 0>, Vector=$User.offset)
    Offset Amount=1
    Ribbon Width=20
   )
   &AddVelocity(Velocity=&VC_Mask_Vectors(
     FrontEnd Value:Vector3f=<-4500 0 0>
     blsFrontEndPreview:NiagaraBool=false
     SpeedMin Value:Vector3f=<-4500 0 0>
     SpeedMax Value:Vector3f=<-4500 0 0>
     Speed Alpha:NiagaraFloat=$System.Speed_Alpha
    )) {enabled=false, module=AddVelocity001}
   &AddVelocity(
    Velocity=&MakeCustomVectorFromBool(Bool=$User.bIsFrontendPreview, True Vector=<4500 0 0>, False Vector=<2200 0 0>)
    Velocity Speed Scale=-1
   )
  ]
  particleUpdateScript [
   &FN_DisableTireFX(Mask Index Attribute:NiagaraInt32=$Particles.SpawnGroup, Mask Index FL=2, Mask Index FR=3, Mask Index RL=0, Mask Index RR=1)
   &ParticleState()
   &InheritSourceMovement(
    Applied Position Scale=&Multiply_VectorByFloat(Vector=<1 1 1>, Float=1)
    ApplyForce=false
    ApplyPosition=true
    ApplyPositionFalloff=false
   )
   &SolveForcesAndVelocity(Clamp Velocity=false, Limit Acceleration=false)
   &VC_UpdateRibbon_v1(Enable Skeletal Position=true)
   &ScaleColor(
    Scale RGB=&VectorFromFloat(Value=&MakeCustomFloatFromBool(
      Bool=$User.bIsPainted
      True Float=&FloatFromCurve(
       FloatCurve=DataInterface() {propertyValues {Curve {keys [
           {interpMode=RCIM_Cubic, leaveTangent=-3.8624, tangentMode=RCTM_Break, time=0.4, value=1}
           {arriveTangent=-0.8405, interpMode=RCIM_Cubic, leaveTangent=-6.2986, tangentMode=RCTM_Break, time=0.9, value=0.15}
           {arriveTangentWeight=0.0417, interpMode=RCIM_Cubic, tangentMode=RCTM_User, time=1, value=0}
          ]}}}
       Scale Curve=2
      )
      False Float=&FloatFromCurve(
       FloatCurve=DataInterface() {propertyValues {Curve {keys [
           {arriveTangent=-5.0663, interpMode=RCIM_Cubic, leaveTangent=-5.0663, tangentMode=RCTM_User, time=0.4, value=0.9}
           {arriveTangent=0.0002, interpMode=RCIM_Cubic, leaveTangent=0.0002, tangentMode=RCTM_User, time=0.9, value=0.4}
          ]}}}
       Scale Curve=4
      )
     ))
    ScaleA=true
    ScaleRGB=true
   )
   &DynamicMaterialParameters(
    Index 1 Param 1=&MakeCustomFloatFromBool(Bool=$User.bIsPainted, True Float=0, False Float=0.7)
    Index 1 Param 2=&MakeCustomFloatFromBool(Bool=$User.bIsPainted, True Float=3, False Float=3)
    Module.Index0Param0Write=true
    Module.Index0Param1Write=true
    Module.Index0Param2Write=true
    Module.Index0Param3Write=true
    Module.Index1Param0Write=true
    Module.Index1Param1Write=true
    Module.Index1Param2Write=true
    Module.Index1Param3Write=true
    Write Parameter Index 1=true
   ) {module=DynamicMaterialParameters001}
   &ScaleRibbonWidth(Ribbon Width Scale=&FloatFromCurve(
     FloatCurve=DataInterface() {propertyValues {Curve {keys [
         {arriveTangent=-1.1867, interpMode=RCIM_Cubic, leaveTangent=-1.1867, tangentMode=RCTM_User, time=0, value=1}
         {arriveTangent=-1.0804, interpMode=RCIM_Cubic, leaveTangent=-1.0804, tangentMode=RCTM_User, time=0.1, value=0.9}
         <at 1 0.25 RCIM_Cubic>
        ]}}}
     Scale Curve=&MakeCustomFloatFromBool(Bool=$User.bIsPainted, True Float=1.5, False Float=1.2)
    ))
  ]
  rendererClasses [&NiagaraRibbonRendererProperties]
  renderers [{rendererClass=&NiagaraRibbonRendererProperties}]
 }
 NE_VC_DriftTrail_Ribbon_Parent001 {
  emitterUpdateScript [
   &EmitterState()
   &VC_RateQualityHandler(Epic Spawn Rate=90, High Spawn Rate=70, Low Spawn Rate=30, Medium Spawn Rate=45, Use Custom Rate?=true)
   &VC_SpawnRibbon_v1(Number of Ribbons=2, Spawn Time=-1, SpawnRate=$Emitter.Spawn Rate)
   /SetVariables(Emitter.ISFRONTEND:NiagaraBool=$User.bIsFrontendPreview) {isSetParameters=true}
  ]
  particleSpawnScript [
   &VC_PaintedVariantsOverride(
    Black=<rgba 0.015 0.015 0.015 1>
    BurntSienna=<rgba 1 0.2116 0 1>
    Cobalt=<rgba 0 0.0667 2 1>
    Crimson=<rgba 0.7 0 0.019 1>
    DefaultColor=<rgba 2 2 2 1>
    ForestGreen=<rgba 0 0.7 0 1>
    Grey=<rgba 0.1 0.1 0.1 1>
    Lime=<rgba 0.4167 1 0 1>
    Orange=<rgba 1.5 0.114 0.015 1>
    Pink=<rgba 0.8 0.0958 0.5439 1>
    Purple=<rgba 0.325 0 1.5 10>
    Saffron=<rgba 0.7 0.7 0 1>
    SkyBlue=<rgba 0 0.6654 1 1>
    TitaniumWhite=<rgba 1 1 1 1>
   )
   &VC_InitializeRibbon_v1(
    Color=$Particles.Color
    Custom Offset=true
    Enable Skeletal Position=true
    Lifetime=&MakeCustomFloatFromBool(Bool=$User.bIsFrontendPreview, True Float=0.19, False Float=0.4)
    Local Vector Offset=&AddVectorToPosition(Position=<0 0 0>, Vector=$User.offset)
    Offset Amount=1
    Ribbon Width=20
   )
   &AddVelocity(Velocity=&VC_Mask_Vectors(
     FrontEnd Value:Vector3f=<-4500 0 0>
     blsFrontEndPreview:NiagaraBool=false
     SpeedMin Value:Vector3f=<-4500 0 0>
     SpeedMax Value:Vector3f=<-4500 0 0>
     Speed Alpha:NiagaraFloat=$System.Speed_Alpha
    )) {enabled=false, module=AddVelocity001}
   &AddVelocity(
    Velocity=&MakeCustomVectorFromBool(Bool=$User.bIsFrontendPreview, True Vector=<4500 0 0>, False Vector=<2200 0 0>)
    Velocity Speed Scale=-1
   )
  ]
  particleUpdateScript [
   &FN_DisableTireFX(Mask Index Attribute:NiagaraInt32=$Particles.SpawnGroup, Mask Index FL=2, Mask Index FR=3, Mask Index RL=0, Mask Index RR=1)
   &ParticleState()
   &InheritSourceMovement(
    Applied Position Scale=&Multiply_VectorByFloat(Vector=<1 1 1>, Float=1)
    ApplyForce=false
    ApplyPosition=true
    ApplyPositionFalloff=false
   )
   &SolveForcesAndVelocity(Clamp Velocity=false, Limit Acceleration=false)
   &VC_UpdateRibbon_v1(Enable Skeletal Position=true)
   &ScaleColor(
    Scale RGB=&VectorFromFloat(Value=&MakeCustomFloatFromBool(
      Bool=$User.bIsPainted
      True Float=&FloatFromCurve(
       FloatCurve=DataInterface() {propertyValues {Curve {keys [
           {interpMode=RCIM_Cubic, leaveTangent=-3.8624, tangentMode=RCTM_Break, time=0.4, value=1}
           {arriveTangent=-0.8405, interpMode=RCIM_Cubic, leaveTangent=-6.2986, tangentMode=RCTM_Break, time=0.9, value=0.15}
           {arriveTangentWeight=0.0417, interpMode=RCIM_Cubic, tangentMode=RCTM_User, time=1, value=0}
          ]}}}
       Scale Curve=2
      )
      False Float=&FloatFromCurve(
       FloatCurve=DataInterface() {propertyValues {Curve {keys [
           {arriveTangent=-5.0663, interpMode=RCIM_Cubic, leaveTangent=-5.0663, tangentMode=RCTM_User, time=0.4, value=0.9}
           {arriveTangent=0.0002, interpMode=RCIM_Cubic, leaveTangent=0.0002, tangentMode=RCTM_User, time=0.9, value=0.4}
          ]}}}
       Scale Curve=4
      )
     ))
    ScaleA=true
    ScaleRGB=true
   )
   &DynamicMaterialParameters(
    Index 1 Param 1=&MakeCustomFloatFromBool(Bool=$User.bIsPainted, True Float=0, False Float=0.7)
    Index 1 Param 2=&MakeCustomFloatFromBool(Bool=$User.bIsPainted, True Float=3, False Float=3)
    Module.Index0Param0Write=true
    Module.Index0Param1Write=true
    Module.Index0Param2Write=true
    Module.Index0Param3Write=true
    Module.Index1Param0Write=true
    Module.Index1Param1Write=true
    Module.Index1Param2Write=true
    Module.Index1Param3Write=true
    Write Parameter Index 1=true
   ) {module=DynamicMaterialParameters001}
   &ScaleRibbonWidth(Ribbon Width Scale=&FloatFromCurve(
     FloatCurve=DataInterface() {propertyValues {Curve {keys [
         {arriveTangent=-1.1867, interpMode=RCIM_Cubic, leaveTangent=-1.1867, tangentMode=RCTM_User, time=0, value=1}
         {arriveTangent=-1.0804, interpMode=RCIM_Cubic, leaveTangent=-1.0804, tangentMode=RCTM_User, time=0.1, value=0.9}
         <at 1 0.25 RCIM_Cubic>
        ]}}}
     Scale Curve=&MakeCustomFloatFromBool(Bool=$User.bIsPainted, True Float=1.5, False Float=1.2)
    ))
  ]
  rendererClasses [&NiagaraRibbonRendererProperties]
  renderers [{rendererClass=&NiagaraRibbonRendererProperties}]
 }
 NE_VC_DriftTrail_Sprite_Parent {
  emitterUpdateScript [
   &EmitterState(Life Cycle Mode=::NewEnumerator1, Loop Duration=0.12, UseLoopDelay=false)
   &SpawnBurst_Instantaneous(
    Spawn Count=&MakeCustomIntFromBool(Bool=$User.bIsFrontendPreview, True INT=4, False INT=6)
    Use Spawn Probability=false
    UseLoopCountLimit=false
   ) {module=SpawnBurst_Instantaneous003}
   &SpawnBurst_Instantaneous(
    Spawn Count=&MakeCustomIntFromBool(Bool=$User.bIsFrontendPreview, True INT=4, False INT=6)
    Spawn Group=2
    Use Spawn Probability=true
    UseLoopCountLimit=false
   ) {module=SpawnBurst_Instantaneous002}
   &SpawnBurst_Instantaneous(
    Spawn Count=&MakeCustomIntFromBool(Bool=$User.bIsFrontendPreview, True INT=4, False INT=6)
    Spawn Group=1
    Use Spawn Probability=false
    UseLoopCountLimit=false
   ) {module=SpawnBurst_Instantaneous001}
   &SpawnRate(
    SpawnGroup=4
    SpawnRate=/VehicleCosmetics/FX/Niagara/Modules/VC_Mask_Floats(
     FrontEnd Value:NiagaraFloat=0
     blsFrontEndPreview:NiagaraBool=false
     SpeedMin Value:NiagaraFloat=0
     SpeedMax Value:NiagaraFloat=0
     Speed Alpha:NiagaraFloat=$System.Speed_Alpha
    )
    Use Spawn Probability=false
   )
   &SpawnBurst_Instantaneous(Spawn Count=6, Spawn Time=1.0833, Use Spawn Probability=false, UseLoopCountLimit=false)
  ]
  particleSpawnScript [
   /SetVariables(Particles.NewVCTag:NiagaraInt32=&SpawnGroup()) {isSetParameters=true}
   &VC_PaintedVariantsOverride(
    Black=<rgba 0.01 0.01 0.01 1>
    BurntSienna=<rgba 0.25 0.0625 0 1>
    Cobalt=<rgba 0 0.125 1.5 1>
    DefaultColor=&SelectLinearColorFromArray(
     Array Sampling Mode=::NewEnumerator0
     Direct Array Index=?
     Color Selection Array=DataInterface() {propertyValues {ColorData [<rgba 0 1.6667 4 1>, <rgba 1.5 0 0.2 1>]}}
     Random Seed=0
    )
    ForestGreen=<rgba 0 0.7 0 1>
    Grey=<rgba 0.07 0.07 0.07 1>
    Lime=<rgba 0.5208 1.25 0 1>
    Orange=<rgba 1.25 0.0833 0 1>
    Pink=<rgba 0.8 0.0958 0.5439 1>
    Saffron=<rgba 0.7 0.7 0 1>
    SkyBlue=<rgba 0 0.6654 1 1>
    TitaniumWhite=<rgba 1 1 1 1>
   )
   &InitializeParticle(
    Color=$Particles.Color
    Lifetime=/VehicleCosmetics/FX/Niagara/Modules/MergeEmitters/VC_FloatBySpawnGroup(
     SpawnGroup or ExecIndex:NiagaraInt32=0
     CustomInt:NiagaraInt32=0
     MaximumSpawnGroups:NiagaraInt32=4
     Float SG 0:NiagaraFloat=0.15
     Float SG 1:NiagaraFloat=0.26
     Float SG 2:NiagaraFloat=0.3
     Float SG 3:NiagaraFloat=0.25
     Float SG 4:NiagaraFloat=0
     Float SG 5:NiagaraFloat=0
    )
    Mesh Scale Max=&RandomRangeVector(Minimum=<4 3 3>, Maximum=<4 5 5>)
    Mesh Scale Min=&RandomRangeVector(Minimum=<2 5 5>, Maximum=<2 5 7>)
    Mesh Scale Mode=::NewEnumerator2
    Sprite Rotation Mode=::NewEnumerator2
    Sprite Size Mode=::NewEnumerator3
    UsePositionOffset=false
    Write Material Random=true
   )
   &PartitionParticles()
   &SkeletalMeshLocation(
    Bone / Socket Index=&Add_Integer(A=$Particles.PartitionParticles.Partition, B=4)
    Bone / Socket Sampling Mode=::NewEnumerator5
    Interpolate Mesh Sampling=false
    MaskSpawnGroup=false
    Mesh Sampling Type=::NewEnumerator6
    Sampled Position Offset=&AddVectorToPosition(Position=<0 0 0>, Vector=$User.offset)
    Skeletal Mesh=$System.SkeletalMesh
   )
   &AddVelocity(Velocity=/VehicleCosmetics/FX/Niagara/Modules/MergeEmitters/VC_VectorBySpawnGroup(
     SpawnGroup or ExecIndex:NiagaraInt32=0
     CustomInt:NiagaraInt32=0
     MaximumSpawnGroups:NiagaraInt32=3
     Vector Group 0:Vector3f=&VC_Mask_Vectors(
      FrontEnd Value:Vector3f=<-1300 0 0>
      blsFrontEndPreview:NiagaraBool=false
      SpeedMin Value:Vector3f=<-1300 0 0>
      SpeedMax Value:Vector3f=<-1300 0 0>
      Speed Alpha:NiagaraFloat=$System.Speed_Alpha
     )
     Vector Group 1:Vector3f=&VC_Mask_Vectors(
      FrontEnd Value:Vector3f=<-800 0 0>
      blsFrontEndPreview:NiagaraBool=false
      SpeedMin Value:Vector3f=<-1200 0 0>
      SpeedMax Value:Vector3f=<-1000 0 0>
      Speed Alpha:NiagaraFloat=$System.Speed_Alpha
     )
     Vector Group 2:Vector3f=&VC_Mask_Vectors(
      FrontEnd Value:Vector3f=<-100 0 0>
      blsFrontEndPreview:NiagaraBool=false
      SpeedMin Value:Vector3f=<-100 0 0>
      SpeedMax Value:Vector3f=<-100 0 0>
      Speed Alpha:NiagaraFloat=$System.Speed_Alpha
     )
     Vector Group 3:Vector3f=&MakeCustomVectorFromBool(Bool=$User.bIsFrontendPreview, True Vector=<-1300 0 1>, False Vector=<-1300 0 0>)
     Vector Group 4:Vector3f=<0 0 0>
     Vector Group 5:Vector3f=<0 0 0>
    )) {enabled=false}
   &InitialMeshOrientation(
    Mesh Axis To Orient=&RandomRangeVector(Minimum=<0 1 1>, Maximum=<0 -1 -1>)
    Mesh Orientation Mode=::NewEnumerator2
    Orientation Vector=&RandomRangeVector(Minimum=<0 1 1>, Maximum=<0 -1 1>, Randomness Mode=::NewEnumerator1)
    Rotation=<90 65 0>
    Rotation Coordinate Space=::NewEnumerator2
    Use Rotation Vector=false
   )
   &AddVelocity(
    Velocity=&VC_VectorBySpawnGroup(
     SpawnGroup or ExecIndex:NiagaraInt32=0
     CustomInt:NiagaraInt32=0
     MaximumSpawnGroups:NiagaraInt32=3
     Vector Group 0:Vector3f=<1500 0 0>
     Vector Group 1:Vector3f=<1150 0 0>
     Vector Group 2:Vector3f=<200 0 0>
     Vector Group 3:Vector3f=<0 0 0>
     Vector Group 4:Vector3f=<0 0 0>
     Vector Group 5:Vector3f=<0 0 0>
    )
    Velocity Speed Scale=&MakeCustomFloatFromBool(Bool=$User.bIsFrontendPreview, True Float=-1, False Float=-0.8)
   ) {module=AddVelocity001}
  ]
  particleUpdateScript [
   &FN_DisableTireFX(
    Mask Index Attribute:NiagaraInt32=$Particles.PartitionParticles.Partition
    Mask Index FL=2
    Mask Index FR=3
    Mask Index RL=0
    Mask Index RR=1
   )
   &ParticleState()
   &ScaleMeshSize(Scale Factor=&SelectVectorFromArray(
     Direct Array Index=$Particles.PartitionParticles.Partition
     Vector Selection Array=DataInterface() {propertyValues {FloatData [<1 -0.75 0.75>, <1 -0.75 -0.75>]}}
    )) {enabled=false}
   &DynamicMaterialParameters(
    Index 0 Param 1=&VC_FloatBySpawnGroup(
     SpawnGroup or ExecIndex:NiagaraInt32=0
     CustomInt:NiagaraInt32=0
     MaximumSpawnGroups:NiagaraInt32=3
     Float SG 0:NiagaraFloat=&FloatFromCurve(
      FloatCurve=DataInterface() {propertyValues {Curve {keys [
          <at 0.2 0 RCIM_Cubic>
          {arriveTangent=0.002, interpMode=RCIM_Cubic, leaveTangent=0.7645, tangentMode=RCTM_Break, time=0.6, value=0.1}
          {arriveTangent=2.25, interpMode=RCIM_Cubic, leaveTangent=2.25, time=0.8, value=0.3}
          {arriveTangent=0.644, tangentMode=RCTM_User, time=1, value=1}
         ]}}}
      Scale Curve=&RandomRangeFloat(Minimum=1, Maximum=1)
     )
     Float SG 1:NiagaraFloat=&FloatFromCurve(FloatCurve=DataInterface() {propertyValues {Curve {keys [
          <at 0.1 0 RCIM_Cubic>
          {arriveTangent=0.002, interpMode=RCIM_Cubic, leaveTangent=0.7645, tangentMode=RCTM_Break, time=0.6, value=0.2}
          {arriveTangent=0.644, tangentMode=RCTM_User, time=1, value=1}
         ]}}})
     Float SG 2:NiagaraFloat=&FloatFromCurve(FloatCurve=DataInterface() {propertyValues {Curve {keys [<at 0.2 0 RCIM_Cubic>, {arriveTangent=0.1221, tangentMode=RCTM_User, time=1, value=1}]}}})
     Float SG 3:NiagaraFloat=&FloatFromCurve(
      FloatCurve=DataInterface() {propertyValues {Curve {keys [
          <at 0.2 0 RCIM_Cubic>
          {arriveTangent=0.002, interpMode=RCIM_Cubic, leaveTangent=0.7645, tangentMode=RCTM_Break, time=0.6, value=0.1}
          {arriveTangent=2.25, interpMode=RCIM_Cubic, leaveTangent=2.25, time=0.8, value=0.3}
          {arriveTangent=0.644, tangentMode=RCTM_User, time=1, value=1}
         ]}}}
      Scale Curve=&RandomRangeFloat(Minimum=1, Maximum=1)
     )
     Float SG 4:NiagaraFloat=0
     Float SG 5:NiagaraFloat=0
    )
    Index 0 Param 2=&VC_FloatBySpawnGroup(
     SpawnGroup or ExecIndex:NiagaraInt32=0
     CustomInt:NiagaraInt32=0
     MaximumSpawnGroups:NiagaraInt32=3
     Float SG 0:NiagaraFloat=&FloatFromCurve(FloatCurve=DataInterface() {propertyValues {Curve {keys [<at 0 1 RCIM_Cubic>,<at 0.5 0 RCIM_Cubic>]}}}, Scale Curve=0.5)
     Float SG 1:NiagaraFloat=&FloatFromCurve(
      FloatCurve=DataInterface() {propertyValues {Curve {keys [
          <at 0 1 RCIM_Cubic>
          {arriveTangent=-1.4256, interpMode=RCIM_Cubic, leaveTangent=-1.4256, tangentMode=RCTM_User, time=0.5, value=0.4}
          <at 1 0 RCIM_Cubic>
         ]}}}
      Scale Curve=0.5
     )
     Float SG 2:NiagaraFloat=&FloatFromCurve(
      FloatCurve=DataInterface() {propertyValues {Curve {keys [
          <at 0 1 RCIM_Cubic>
          {arriveTangent=-1.3855, interpMode=RCIM_Cubic, leaveTangent=-1.3855, tangentMode=RCTM_User, time=0.5, value=0.45}
          <at 1 0 RCIM_Cubic>
         ]}}}
      Scale Curve=0.8
     )
     Float SG 3:NiagaraFloat=&FloatFromCurve(FloatCurve=DataInterface() {propertyValues {Curve {keys [<at 0 1 RCIM_Cubic>,<at 0.5 0 RCIM_Cubic>]}}})
     Float SG 4:NiagaraFloat=0
     Float SG 5:NiagaraFloat=0
    )
    Index 1 Param 1=&MakeCustomFloatFromBool(Bool=$User.bIsPainted, True Float=0, False Float=1)
    Index 1 Param 2=&MakeCustomFloatFromBool(
     Bool=$User.bIsPainted
     True Float=&MakeFloatFromInt(Int=$User.Lines_Glow)
     False Float=&MakeCustomFloatFromBool(Bool=$User.bIsFrontendPreview, True Float=50, False Float=25)
    )
    Module.Index0Param0Write=true
    Module.Index0Param1Write=true
    Module.Index0Param2Write=true
    Module.Index0Param3Write=true
    Module.Index1Param0Write=true
    Module.Index1Param1Write=true
    Module.Index1Param2Write=true
    Module.Index1Param3Write=true
    Write Parameter Index 1=true
   )
   &InheritSourceMovement(
    Applied Position Scale=&VC_VectorBySpawnGroup(
     SpawnGroup or ExecIndex:NiagaraInt32=0
     CustomInt:NiagaraInt32=0
     MaximumSpawnGroups:NiagaraInt32=3
     Vector Group 0:Vector3f=&Multiply_VectorByFloat(
      Vector=<1 1 1>
      Float=&FloatFromCurve(FloatCurve=DataInterface() {propertyValues {Curve {keys [
           {arriveTangent=-0.4618, interpMode=RCIM_Cubic, leaveTangent=-0.4618, tangentMode=RCTM_User, time=0, value=1}
           {arriveTangent=-0.0795, interpMode=RCIM_Cubic, leaveTangent=-0.0795, tangentMode=RCTM_User, time=1.1, value=0.75}
          ]}}})
     )
     Vector Group 1:Vector3f=&Multiply_VectorByFloat(
      Vector=<1 1 1>
      Float=&FloatFromCurve(FloatCurve=DataInterface() {propertyValues {Curve {keys [
           {arriveTangent=-0.4618, interpMode=RCIM_Cubic, leaveTangent=-0.4618, tangentMode=RCTM_User, time=0, value=1}
           {arriveTangent=-0.0795, interpMode=RCIM_Cubic, leaveTangent=-0.0795, tangentMode=RCTM_User, time=1.1, value=0.75}
          ]}}})
     )
     Vector Group 2:Vector3f=&Multiply_VectorByFloat(
      Vector=<1 1 1>
      Float=&FloatFromCurve(FloatCurve=DataInterface() {propertyValues {Curve {keys [
           {arriveTangent=-0.4618, interpMode=RCIM_Cubic, leaveTangent=-0.4618, tangentMode=RCTM_User, time=0, value=1}
           {arriveTangent=-0.0795, interpMode=RCIM_Cubic, leaveTangent=-0.0795, tangentMode=RCTM_User, time=1.1, value=0.75}
          ]}}})
     )
     Vector Group 3:Vector3f=<0 0 0>
     Vector Group 4:Vector3f=<0 0 0>
     Vector Group 5:Vector3f=<0 0 0>
    )
    ApplyForce=false
    ApplyPosition=true
    ApplyPositionFalloff=false
   ) {enabled=false}
   &InheritSourceMovement(
    Applied Position Scale=&Multiply_VectorByFloat(Vector=<1 1 1>, Float=1)
    ApplyForce=false
    ApplyPosition=true
    ApplyPositionFalloff=false
   ) {module=InheritSourceMovement001}
   &SolveForcesAndVelocity(Clamp Velocity=false, Limit Acceleration=false)
   /SetVariables(
    Particles.Lines01ScaleColor:LinearColor=&VC_DynamicScaleColor(
     Color to Scale:LinearColor=$Particles.Color
     RGB Scale:Vector3f=&VectorFromFloat(Value=&MakeCustomFloatFromBool(
       Bool=$User.bIsPainted
       True Float=&RandomRangeFloat(
        Minimum=&FloatFromCurve(
         FloatCurve=DataInterface() {propertyValues {Curve {keys [{interpMode=RCIM_Cubic, leaveTangent=-2.9826, tangentMode=RCTM_Break, time=0, value=1}, <at 0.95 0.15 RCIM_Cubic>]}}}
         Scale Curve=0
        )
        Maximum=&FloatFromCurve(FloatCurve=DataInterface() {propertyValues {Curve {keys [<at 0 1>,<at 1 0.5>]}}}, Scale Curve=3)
       )
       False Float=&FloatFromCurve(
        FloatCurve=DataInterface() {propertyValues {Curve {keys [{arriveTangent=-5.3196, interpMode=RCIM_Cubic, leaveTangent=-5.3196, tangentMode=RCTM_User, time=0.3, value=2}, <at 1 1 RCIM_Cubic>]}}}
        Scale Curve=&RandomRangeFloat(Minimum=1, Maximum=3)
       )
      ))
     Alpha Scale:NiagaraFloat=0
    )
    Particles.Lines01ScaleMesh:Vector3f=&Multiply_Vector(
     A=<4 4 4>
     B=&VectorFromCurve(VectorCurve=DataInterface() {propertyValues {
        XCurve {keys [
          <at 0.1 1 RCIM_Cubic>
          {arriveTangent=-1.2479, arriveTangentWeight=0.3356, interpMode=RCIM_Cubic, leaveTangent=-1.2479, tangentMode=RCTM_User, time=1, value=0}
         ]}
        YCurve {keys [
          <at 0.1 1 RCIM_Cubic>
          {arriveTangent=-1.2479, arriveTangentWeight=0.3356, interpMode=RCIM_Cubic, leaveTangent=-1.2479, tangentMode=RCTM_User, time=1, value=0}
         ]}
        ZCurve {keys [
          <at 0.1 1 RCIM_Cubic>
          {arriveTangent=-1.2479, arriveTangentWeight=0.3356, interpMode=RCIM_Cubic, leaveTangent=-1.2479, tangentMode=RCTM_User, time=1, value=0}
         ]}
       }})
    )
    Particles.Lines02ScaleColor:LinearColor=&VC_DynamicScaleColor(
     Color to Scale:LinearColor=$Particles.Color
     RGB Scale:Vector3f=&VectorFromFloat(Value=&MakeCustomFloatFromBool(
       Bool=$User.bIsPainted
       True Float=&FloatFromCurve(
        FloatCurve=DataInterface() {propertyValues {Curve {keys [{interpMode=RCIM_Cubic, leaveTangent=-2.5187, tangentMode=RCTM_Break, time=0, value=1}, <at 1 0.15 RCIM_Cubic>]}}}
        Scale Curve=3
       )
       False Float=&FloatFromCurve(
        FloatCurve=DataInterface() {propertyValues {Curve {keys [{arriveTangent=-0.3719, interpMode=RCIM_Cubic, leaveTangent=-0.3719, tangentMode=RCTM_User, time=0.1, value=1}, <at 1 0.25 RCIM_Cubic>]}}}
        Scale Curve=&RandomRangeFloat(Minimum=2, Maximum=1)
       )
      ))
     Alpha Scale:NiagaraFloat=0
    )
    Particles.Lines02ScaleMesh:Vector3f=<6 3.5 3.5>
    Particles.Lines03ScaleColor:LinearColor=&VC_DynamicScaleColor(
     Color to Scale:LinearColor=$Particles.Color
     RGB Scale:Vector3f=&VectorFromFloat(Value=&MakeCustomFloatFromBool(
       Bool=$User.bIsPainted
       True Float=&FloatFromCurve(
        FloatCurve=DataInterface() {propertyValues {Curve {keys [{interpMode=RCIM_Cubic, leaveTangent=-2.4022, tangentMode=RCTM_Break, time=0, value=1}, <at 1 0.25 RCIM_Cubic>]}}}
        Scale Curve=3
       )
       False Float=&FloatFromCurve(
        FloatCurve=DataInterface() {propertyValues {Curve {keys [{arriveTangent=-5.2742, interpMode=RCIM_Cubic, leaveTangent=-5.2742, tangentMode=RCTM_User, time=0.3, value=2}, <at 1 1 RCIM_Cubic>]}}}
        Scale Curve=&RandomRangeFloat(Minimum=1, Maximum=1)
       )
      ))
     Alpha Scale:NiagaraFloat=0
    )
    Particles.Lines03ScaleMesh:Vector3f=&Multiply_Vector(
     A=<4 5 5>
     B=&VectorFromCurve(VectorCurve=DataInterface() {propertyValues {
        XCurve {keys [<at 0 1 RCIM_Cubic>, {arriveTangent=-0.4, interpMode=RCIM_Cubic, leaveTangent=-0.4, time=0.8, value=0.65}, <at 1 0.6 RCIM_Cubic>]}
        YCurve {keys [
          <at 0 0.7 RCIM_Cubic>
          {arriveTangentWeight=0.178, interpMode=RCIM_Cubic, leaveTangentWeight=0.1068, tangentMode=RCTM_User, time=0.3, value=0.7}
          {arriveTangentWeight=0.1414, interpMode=RCIM_Cubic, leaveTangentWeight=0.1886, tangentMode=RCTM_User, time=0.5, value=0.8}
          {arriveTangent=-0.1027, interpMode=RCIM_Cubic, leaveTangent=-0.1027, tangentMode=RCTM_User, time=1, value=0}
         ]}
        ZCurve {keys [
          <at 0 0.7 RCIM_Cubic>
          {arriveTangentWeight=0.178, interpMode=RCIM_Cubic, leaveTangentWeight=0.1068, tangentMode=RCTM_User, time=0.3, value=0.7}
          {arriveTangentWeight=0.1414, interpMode=RCIM_Cubic, leaveTangentWeight=0.1886, tangentMode=RCTM_User, time=0.5, value=0.8}
          {arriveTangent=-0.1027, interpMode=RCIM_Cubic, leaveTangent=-0.1027, tangentMode=RCTM_User, time=1, value=0}
         ]}
       }})
    )
   ) {isSetParameters=true}
  ]
  rendererClasses [&NiagaraMeshRendererProperties, &NiagaraSpriteRendererProperties]
  renderers [
   {rendererClass=&NiagaraMeshRendererProperties, rendererIndex=0}
   {rendererClass=&NiagaraMeshRendererProperties, rendererIndex=1}
   {rendererClass=&NiagaraMeshRendererProperties, rendererIndex=2}
   {rendererClass=&NiagaraSpriteRendererProperties, rendererIndex=3}
  ]
 }
 NE_VC_DriftTrail_WheelWrap_Parent {
  simTarget=GPUComputeSim
  emitterUpdateScript [
   &EmitterState(Life Cycle Mode=::NewEnumerator1, Loop Delay=0.1, Loop Duration=0.2, Recalculate Duration Each Loop=true, UseLoopDelay=false)
   &SpawnBurst_Instantaneous(Spawn Count=8, Spawn Group=1, Use Spawn Probability=true, UseLoopCountLimit=false) {module=SpawnBurst_Instantaneous001}
   &SpawnBurst_Instantaneous(Spawn Count=8, Use Spawn Probability=true, UseLoopCountLimit=false)
  ]
  particleSpawnScript [
   /SetVariables(Particles.NewVCTag:NiagaraInt32=&SpawnGroup()) {isSetParameters=true}
   &VC_PaintedVariantsOverride(
    Black=<rgba 0.01 0.01 0.01 1>
    BurntSienna=<rgba 1 0.2116 0 1>
    Cobalt=<rgba 0 0.0667 2 1>
    Crimson=<rgba 0.7 0 0.019 1>
    DefaultColor=&SelectLinearColorFromArray(
     Array Sampling Mode=::NewEnumerator0
     Direct Array Index=?
     Color Selection Array=DataInterface() {propertyValues {ColorData [<rgba 1 0.0712 0.8849 1>, <rgba 0.0275 0.3255 1 1>]}}
     Random Seed=0
    )
    ForestGreen=<rgba 0 0.7 0 1>
    Grey=<rgba 0.075 0.075 0.075 1>
    Lime=<rgba 0.4167 1 0 1>
    Orange=<rgba 1.5 0.075 0 1>
    Pink=<rgba 0.8 0.0958 0.5439 1>
    Purple=<rgba 0.2333 0 0.7 10>
    Saffron=<rgba 0.7 0.7 0 1>
    SkyBlue=<rgba 0 0.6654 1 1>
    TitaniumWhite=<rgba 1 1 1 1>
   )
   &InitializeParticle(
    Color=$Particles.Color
    Lifetime=&VC_FloatBySpawnGroup(
     SpawnGroup or ExecIndex:NiagaraInt32=0
     CustomInt:NiagaraInt32=0
     MaximumSpawnGroups:NiagaraInt32=2
     Float SG 0:NiagaraFloat=0.2
     Float SG 1:NiagaraFloat=1
     Float SG 2:NiagaraFloat=0
     Float SG 3:NiagaraFloat=0
     Float SG 4:NiagaraFloat=0
     Float SG 5:NiagaraFloat=0
    )
    Mesh Scale Mode=::NewEnumerator4
    Sprite Rotation Mode=::NewEnumerator2
    Sprite Size Mode=::NewEnumerator3
    UsePositionOffset=false
    Write Material Random=true
   )
   &PartitionParticles()
   &SkeletalMeshLocation(
    Bone / Socket Index=&Add_Integer(A=$Particles.PartitionParticles.Partition, B=4)
    Bone / Socket Sampling Mode=::NewEnumerator5
    Interpolate Mesh Sampling=false
    MaskSpawnGroup=false
    Mesh Sampling Type=::NewEnumerator6
    Orientation Sampling=::NewEnumerator1
    Sampled Position Offset=<0 2 0>
    Skeletal Mesh=$System.SkeletalMesh
    Use Sampled Mesh Attribute as Ribbon ID=false
   )
   &InitialMeshOrientation(Mesh Orientation Mode=::NewEnumerator2, Rotation Coordinate Space=::NewEnumerator2, Use Rotation Vector=true)
  ]
  particleUpdateScript [
   &FN_DisableTireFX(
    Mask Index Attribute:NiagaraInt32=$Particles.PartitionParticles.Partition
    Mask Index FL=2
    Mask Index FR=3
    Mask Index RL=0
    Mask Index RR=1
   )
   &VC_WheelWrapScale(Wheel Index Attribute:NiagaraInt32=&Add_Integer(A=$Particles.PartitionParticles.Partition, B=2))
   &ScaleColor(
    Scale RGB=&VectorFromFloat(Value=&FloatFromCurve(
      FloatCurve=DataInterface() {propertyValues {Curve {keys [
          {interpMode=RCIM_Cubic, tangentMode=RCTM_User, time=0, value=1}
          {interpMode=RCIM_Cubic, tangentMode=RCTM_User, time=0.5, value=0}
          {tangentMode=RCTM_User, time=1, value=1}
         ]}}}
      Scale Curve=4
     ))
    ScaleA=true
    ScaleRGB=true
   ) {enabled=false}
   &ScaleColor(
    Scale RGB=&VectorFromFloat(Value=&MakeCustomFloatFromBool(
      Bool=$User.bIsPainted
      True Float=&FloatFromCurve(FloatCurve=DataInterface() {propertyValues {Curve {keys [<at 0 1 RCIM_Cubic>,<at 0.5 0.2 RCIM_Cubic>,<at 1 1 RCIM_Cubic>]}}}, Scale Curve=4)
      False Float=&FloatFromCurve(
       FloatCurve=DataInterface() {propertyValues {Curve {keys [
           {interpMode=RCIM_Cubic, tangentMode=RCTM_User, time=0, value=1}
           {interpMode=RCIM_Cubic, tangentMode=RCTM_User, time=0.5, value=0.1}
           {tangentMode=RCTM_User, time=1, value=1}
          ]}}}
       Scale Curve=8
      )
     ))
    ScaleA=true
    ScaleRGB=true
   ) {enabled=false, module=ScaleColor001}
   &DynamicMaterialParameters(
    Index 1 Param 1=&MakeCustomFloatFromBool(Bool=$User.bIsPainted, True Float=0, False Float=1)
    Index 2 Param 1=&MakeCustomFloatFromBool(Bool=$User.bIsPainted, True Float=0, False Float=1)
    Index 2 Param 2=&MakeCustomFloatFromBool(Bool=$User.bIsPainted, True Float=&MakeFloatFromInt(Int=$User.Lines_Glow), False Float=100)
    Index 3 Param 2=10
    Module.Index0Param0Write=true
    Module.Index0Param1Write=true
    Module.Index0Param2Write=true
    Module.Index0Param3Write=true
    Module.Index1Param0Write=true
    Module.Index1Param1Write=true
    Module.Index1Param2Write=true
    Module.Index1Param3Write=true
    Module.Index2Param0Write=true
    Module.Index2Param1Write=true
    Module.Index2Param2Write=true
    Module.Index2Param3Write=true
    Module.Index3Param0Write=true
    Module.Index3Param1Write=true
    Module.Index3Param2Write=true
    Module.Index3Param3Write=true
    Write Parameter Index 1=true
    Write Parameter Index 2=true
    Write Parameter Index 3=true
   )
   /SetVariables(
    Particles.RingScaleColor:LinearColor=&VC_DynamicScaleColor(
     Color to Scale:LinearColor=$Particles.Color
     RGB Scale:Vector3f=&VectorFromFloat(Value=&MakeCustomFloatFromBool(
       Bool=$User.bIsPainted
       True Float=&FloatFromCurve(FloatCurve=DataInterface() {propertyValues {Curve {keys [<at 0 1 RCIM_Cubic>,<at 0.5 0.2 RCIM_Cubic>,<at 1 1 RCIM_Cubic>]}}}, Scale Curve=5)
       False Float=&FloatFromCurve(
        FloatCurve=DataInterface() {propertyValues {Curve {keys [
            {interpMode=RCIM_Cubic, tangentMode=RCTM_User, time=0, value=1}
            {interpMode=RCIM_Cubic, tangentMode=RCTM_User, time=0.5, value=0.5}
            {tangentMode=RCTM_User, time=1, value=1}
           ]}}}
        Scale Curve=5
       )
      ))
     Alpha Scale:NiagaraFloat=0
    )
    Particles.SparksScaleColor:LinearColor=&VC_DynamicScaleColor(
     Color to Scale:LinearColor=$Particles.Color
     RGB Scale:Vector3f=&VectorFromFloat(Value=&MakeCustomFloatFromBool(Bool=$User.bIsPainted, True Float=$User.WHSPARKSGLOW, False Float=5))
     Alpha Scale:NiagaraFloat=1
    )
   ) {isSetParameters=true}
   &ParticleState()
   &UpdateMeshOrientation(Rotation Coordinate Space=::NewEnumerator2, Rotation Vector=<0 1 0>) {enabled=false}
   &UpdateMeshOrientation(Rotation Rate=10, Rotation Vector=<0 1 0>) {module=UpdateMeshOrientation001}
   &SkeletalMeshLocation(
    Bone / Socket Index=&Add_Integer(A=$Particles.PartitionParticles.Partition, B=4)
    Bone / Socket Sampling Mode=::NewEnumerator5
    MaskSpawnGroup=false
    Mesh Sampling Type=::NewEnumerator6
    Skeletal Mesh=$System.SkeletalMesh
   ) {module=SkeletalMeshLocation001}
   &SolveForcesAndVelocity(Clamp Velocity=false, Limit Acceleration=false)
  ]
  rendererClasses [&NiagaraMeshRendererProperties]
  renderers [
   {rendererClass=&NiagaraMeshRendererProperties, rendererIndex=0}
   {rendererClass=&NiagaraMeshRendererProperties, rendererIndex=1}
  ]
 }
}
summary {
 systemName=NS_DriftTrail_AnimatedLines_MP
 emitters [
  {emitterName=NE_VC_DriftTrail_WheelWrap_Parent, simTarget=GPUComputeSim, rendererClasses [&NiagaraMeshRendererProperties]}
  {
   emitterName=NE_VC_DriftTrail_Sprite_Parent
   rendererClasses [&NiagaraMeshRendererProperties, &NiagaraSpriteRendererProperties]
  }
  {emitterName=NE_VC_DriftTrail_Ribbon_Parent001, rendererClasses [&NiagaraRibbonRendererProperties]}
  {
   emitterName=NE_VC_DriftTrail_CURVE001
   rendererClasses [&NiagaraMeshRendererProperties, &NiagaraSpriteRendererProperties]
  }
  {emitterName=NE_VC_DriftTrail_Ribbon_Parent, rendererClasses [&NiagaraRibbonRendererProperties]}
 ]
 userVariables [
  User.bIsFrontendPreview {type=NiagaraBool}
  User.bIsPainted {type=NiagaraBool}
  User.ColorVariantID {type=NiagaraInt32}
  User.GameModeID {type=NiagaraInt32}
  User.Lines_Glow {type=NiagaraInt32, defaultValue=1}
  User.offset {type=Vector3f, description="offset from wheels(set through cosmetic template)", defaultValue {struct=Vector3f, value=<0 0 0>}}
  User.PaintedColor {type=LinearColor, description="", defaultValue {struct=LinearColor, value=<rgba 1 1 1 1>}}
  User.PaintedScalar {type=NiagaraFloat, defaultValue=1, description=""}
  User.VehicleFlags {type=NiagaraInt32, defaultValue=806289408}
  User.Wheel_Radius_Array {
   type=NiagaraDataInterfaceArrayFloat
   defaultValue {
    struct=NiagaraExt_VariableValue_DataInterface
    value {
     dataInterface=&self:NiagaraDataInterfaceArrayFloat_0
     dataInterfaceClass=&NiagaraDataInterfaceArrayFloat
    }
   }
  }
  User.Wheel_Width_Array {
   type=NiagaraDataInterfaceArrayFloat
   defaultValue {
    struct=NiagaraExt_VariableValue_DataInterface
    value {
     dataInterface=&self:NiagaraDataInterfaceArrayFloat_1
     dataInterfaceClass=&NiagaraDataInterfaceArrayFloat
    }
   }
  }
  User.WheelSize {type=NiagaraFloat, defaultValue=0.1, description=""}
  User.WHSPARKSGLOW {type=NiagaraFloat, defaultValue=20, description=""}
 ]
}
