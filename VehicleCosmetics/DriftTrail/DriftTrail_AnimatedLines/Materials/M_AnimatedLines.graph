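# Unlit, two-sided masked material for animated drift-trail lines on Niagara mesh particles.
# Particle dynamic parameters BotEdge and TopEdge crop a curved UV band, shaped by either a radial-gradient hard cutout or a panned-texture, inverse-Fresnel glow, chosen by UseGlow.
# DEFAULT_COLOR selects particle-tinted orange-to-blue or gradient-darkened particle color, scaled by EDGE_GLOW into emissive adjustments.
# Mask tiling and panning reuse their X values on both axes, leaving TilingY and SpeedY unused.
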Material /VehicleCosmetics/DriftTrail/DriftTrail_AnimatedLines/Materials/M_AnimatedLines
graph {
 properties {
  bAutomaticallySetUsageInEditor=false
  bUsedWithNiagaraMeshParticles=true
  blendMode=BLEND_Masked
  shadingModel=MSM_Unlit
  twoSided=true
  usedShadingModels=Unlit
 }
 nodes [
  n1 = Power(Base=ComponentMask(None=TextureCoordinate()) {g=true}, Exp=ScalarParameter() {defaultValue=0.8, parameterName=CurvePower})
  n2 = DynamicParameter() {paramNames [BotEdge,TopEdge,Param3,Param4]}
  n3 = Subtract(A=OneMinus(None=@n1), B=@n2.BotEdge)
  n4 = Multiply(A=Saturate(None=@n3), B=Saturate(None=Subtract(A=@n1, B=@n2.TopEdge)))
  n5 = Saturate(None=@n4)
  n6 = TextureCoordinate()
  n7 = VectorParameter() {group=Mask, parameterName=MaskUVControls, channelNames {a=SpeedY, b=SpeedX, g=TilingY, r=TilingX}, defaultValue=<rgba 1 1 0 0>}
  n8 = AppendVector(A=@n7.TilingX, B=@n7.TilingX)
  n9 = TextureSampleParameter2D(UVs=Panner(Coordinate=Multiply(A=@n6, B=@n8), Speed=AppendVector(A=@n7.SpeedX, B=@n7.SpeedX))) {
   group=Mask
   parameterName=MaskTexture
   samplerType=SAMPLERTYPE_Grayscale
   texture=/DealershipSharedLibrary/FX/Textures/Masks/T_GradientCircle
  }
  n10 = ScalarParameter() {defaultValue=0.5, parameterName=FresnelExponent}
  n11 = Multiply(
   A=Multiply(A=Multiply(A=@n5, B=@n9.R), B=ScalarParameter() {defaultValue=10, parameterName=GlowMultiplier}) {constB=3}
   B=OneMinus(None=Fresnel(ExponentIn=@n10))
  )
  n12 = Multiply(A=TextureCoordinate(), B=ScalarParameter() {defaultValue=2, parameterName=U_Mult})
  n13 = Constant() {r=1}
  n14 = Subtract(A=OneMinus(None=@n2.TopEdge), B=@n2.BotEdge)
  n15 = Saturate(None=Multiply(A=@n3, B=Divide(A=@n13, B=@n14)))
  n16 = MaterialFunctionCall(
   UVs=AppendVector(A=ComponentMask(None=@n12) {r=true}, B=@n15)
   CenterPosition=Constant2Vector() {g=0.5, r=0.5}
   Radius=ScalarParameter() {defaultValue=0.5, parameterName=Radius}
   Density=ScalarParameter() {defaultValue=2.3, parameterName=Density}
  ) {materialFunction=&RadialGradientExponential}
  If(
   A=ScalarParameter() {parameterName=UseGlow?}
   "A > B"=@n11
   "A == B"=@n11
   "A < B"=Ceil(None=Multiply(A=@n4, B=@n16.RadialGradientExponential))
  ) {constB=1}
  n17 = DynamicParameter() {parameterIndex=1, paramNames [DEFAULT_COLOR, EDGE_GLOW, Param3, Param4]}
  n18 = ParticleColor()
  n19 = Multiply(
   A=@n18.RGB
   B=LinearInterpolate(
    A=VectorParameter() {parameterName=OrangeColor, defaultValue=<rgba 1 0.1667 0 1>}.RGB
    B=VectorParameter() {parameterName=BlueColor, defaultValue=<rgba 0 0.1 2 1>}.RGB
    Alpha=@n15
   )
  )
  n20 = LinearInterpolate(A=@n18.RGB, B=Multiply(A=@n18.RGB) {constB=0.25}, Alpha=@n16.RadialGradientExponential)
  MaterialFunctionCall(Emissive=Multiply(A=If(A=@n17.DEFAULT_COLOR, "A > B"=@n19, "A == B"=@n20, "A < B"=@n20) {constB=0.1}, B=@n17.EDGE_GLOW) {constB=100}) {materialFunction=&MF_FX_EmissiveAdjustments}
 ]
}
