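# Material function in the ElectroBeam drift trail folder that computes a soft depth-blend factor with a maskable depth bias, exposed to the library.
# The bias mask, inverted by default through a static switch, scales the bias distance, which is added to pixel depth.
# Scene depth minus that biased depth is divided by the blend distance and saturated, giving a fade near intersecting geometry that the mask shifts per pixel.
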MaterialFunction /VehicleCosmetics/DriftTrail/DriftTrail_ElectroBeam/Materials/DepthBiasBlend_MF
graph {
 properties {bExposeToLibrary=true, previewBlendMode=BLEND_Translucent}
 nodes [
  n1 = FunctionInput() {inputName=Depth Bias Mask, inputType=FunctionInput_Scalar}
  n2 = StaticSwitch(
   True=OneMinus(None=@n1)
   False=@n1
   Value=FunctionInput(Preview=StaticBool() {value=true}) {bUsePreviewValueAsDefault=true, inputName=Invert Mask, inputType=FunctionInput_StaticBool, sortPriority=5}
  )
  n3 = FunctionInput() {inputName=Depth Bias Distance, inputType=FunctionInput_Scalar, sortPriority=2}
  n4 = Subtract(A=SceneDepth() {constInput=<1 1>}.R, B=Add(A=PixelDepth().R, B=Multiply(A=@n2, B=@n3)))
  n5 = FunctionInput() {inputName=DepthBlend Distance, inputType=FunctionInput_Scalar, sortPriority=1}
  FunctionOutput(None=Saturate(None=Divide(A=@n4, B=@n5)))
  PixelDepth()
 ]
}
