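# Masked unlit Niagara ribbon material for the animated-lines vehicle drift trail.
# Dynamic parameters trim the curved U axis between edges, and an exponential radial gradient over the remapped UVs is rounded up into a hard shape mask.
# That gradient also drives particle-tinted color, with a dynamic parameter choosing a built-in two-color ramp.
# Switches multiply the mask by a view-angle fade, using screen-derivative face normals, and an optional depth fade.
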Material /VehicleCosmetics/DriftTrail/DriftTrail_AnimatedLines/Materials/M_AnimatedLines_Ribbon
graph {
 properties {bAutomaticallySetUsageInEditor=false, bUsedWithNiagaraRibbons=true, blendMode=BLEND_Masked, shadingModel=MSM_Unlit, usedShadingModels=Unlit}
 nodes [
  n1 = Power(Base=ComponentMask(None=TextureCoordinate()) {r=true}, Exp=ScalarParameter() {defaultValue=0.8, parameterName=CurvePower})
  n2 = DynamicParameter() {paramNames [BotEdge,TopEdge,Param3,Param4]}
  n3 = Subtract(A=OneMinus(None=@n1), B=@n2.BotEdge)
  MaterialExpressionConstantClamp_13 = Saturate(None=@n3)
  MaterialExpressionConstantClamp_14 = Saturate(None=Subtract(A=@n1, B=@n2.TopEdge))
  n4 = Constant() {r=1}
  n5 = Subtract(A=OneMinus(None=@n2.TopEdge), B=@n2.BotEdge)
  MaterialExpressionConstantClamp_15 = Saturate(None=Multiply(A=@n3, B=Divide(A=@n4, B=@n5)))
  n6 = ComponentMask(None=Multiply(A=TextureCoordinate(), B=ScalarParameter() {defaultValue=2, parameterName=U_Mult})) {g=true}
  n7 = Divide(A=Subtract(B=Constant() {r=1}), B=Constant() {r=2})
  n8 = DynamicParameter() {parameterIndex=1, defaultValue=<rgba 1 0.5 1 1>, paramNames [DEFAULTCOLOR, DEFAULT COLOR CIRCULAR LUT DENSITY, Param3, Param4]}
  n9 = MaterialFunctionCall(
   UVs=AppendVector(A=@MaterialExpressionConstantClamp_15, B=Subtract(A=@n6, B=@n7))
   CenterPosition=AppendVector(A=ScalarParameter() {defaultValue=0.5, parameterName=Center_X}, B=ScalarParameter() {defaultValue=0.5, parameterName=Center_Y})
   Radius=Constant() {r=0.525}
   Density=@n8.DEFAULT COLOR CIRCULAR LUT DENSITY
  ) {materialFunction=&RadialGradientExponential}
  n10 = Ceil(None=Multiply(A=Multiply(A=@MaterialExpressionConstantClamp_13, B=@MaterialExpressionConstantClamp_14), B=@n9.RadialGradientExponential))
  MaterialExpressionConstantClamp_12 = Saturate(None=@n10)
  MaterialExpressionVertexColor_2 = ParticleColor()
  n11 = StaticSwitchParameter(True=@n9.RadialGradientExponential, False=@MaterialExpressionConstantClamp_15) {defaultValue=true, parameterName=CircularLUT}
  n12 = LinearInterpolate(A=Constant3Vector() {constant=<rgba 0 0.1 2 1>}, B=Constant3Vector() {constant=<rgba 2.5 0 0.4167 1>}, Alpha=@n11)
  n13 = LinearInterpolate(A=Multiply(A=@MaterialExpressionVertexColor_2.RGB), B=Multiply(A=@MaterialExpressionVertexColor_2.RGB) {constB=0.5}, Alpha=@n11)
  MaterialFunctionCall(Emissive=If(A=@n8.DEFAULTCOLOR, "A > B"=Multiply(A=@MaterialExpressionVertexColor_2.RGB, B=@n12), "A == B"=@n13, "A < B"=@n13) {constB=0.1}) {materialFunction=&MF_FX_EmissiveAdjustments}
  n14 = Constant() {r=1}
  n15 = ScalarParameter() {parameterName=DepthBlendDistance}
  MaterialFunctionCall!(
   Depth Bias Mask=@n14
   DepthBlend Distance=@n15
   Depth Bias Distance=ScalarParameter() {parameterName=DepthBiasDistance}
   Invert Mask=StaticBoolParameter() {parameterName=DepthBiasInvertMask}
  ) {materialFunction=/VehicleCosmetics/DriftTrail/DriftTrail_ElectroBeam/Materials/DepthBiasBlend_MF}
  MaterialExpressionWorldPosition_6 = WorldPosition()
  n16 = Custom!(x=@MaterialExpressionWorldPosition_6.XYZ) {
   code="return ddx(x);\n"
   inputs [{
     inputName=x
     input {
      expression=&self:MaterialExpressionWorldPosition_6
      mask=1
      maskB=1
      maskG=1
      maskR=1
     }
    }]
  }
  n17 = Custom!(y=@MaterialExpressionWorldPosition_6.XYZ) {
   code="return ddy(y);\n"
   inputs [{
     inputName=y
     input {
      expression=&self:MaterialExpressionWorldPosition_6
      mask=1
      maskB=1
      maskG=1
      maskR=1
     }
    }]
  }
  n18 = StaticSwitch(True=@MaterialExpressionWorldPosition_6.XYZ, False=CrossProduct(A=@n16, B=@n17), Value=StaticBool())
  MaterialExpressionCameraWorldPosition_3 = CameraPositionWS()
  n19 = Subtract(A=@MaterialExpressionCameraWorldPosition_3, B=WorldPosition().XYZ)
  n20 = Power(
   Base=DotProduct(A=Normalize(VectorInput=@n18), B=Normalize(VectorInput=@n19))
   Exp=ScalarParameter() {defaultValue=0.5, parameterName=FadeAngleAmount}
  )
  n21 = Constant() {r=1}
  n22 = DepthFade(Opacity=@n14, FadeDistance=@n15)
  n23 = Constant() {r=1}
  Multiply(
   A=@n10
   B=Multiply(
    A=StaticSwitchParameter(True=@n20, False=@n21) {defaultValue=true, parameterName=UseFadeAngle}
    B=StaticSwitchParameter(True=@n22, False=@n23) {parameterName=UseDepthBiasBlend}
   )
  )
  Transform(None=VectorParameter() {parameterName=ribbonwpo, defaultValue=<rgba 0 0 0 1>}.RGB) {transformSourceType=TRANSFORMSOURCE_Local}
 ]
}
