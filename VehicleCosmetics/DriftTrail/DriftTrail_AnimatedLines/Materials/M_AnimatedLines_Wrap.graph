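# Additive unlit material for Niagara mesh particles in the animated-lines drift trail vehicle cosmetic, drawing fast-scrolling glowing streaks.
# A panned tyre texture modulates particle color, which pulses to black when the DEFAULT_COLOR dynamic parameter passes a threshold, with brightness scaled by EDGE_GLOW.
# A rounded sine mask on vertical UV shapes the band before MF_FX_EmissiveAdjustments.
# Separate branches combine a time-flickered texture with that mask or a stepped edge mask.
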Material /VehicleCosmetics/DriftTrail/DriftTrail_AnimatedLines/Materials/M_AnimatedLines_Wrap
graph {
 properties {
  bAutomaticallySetUsageInEditor=false
  bUsedWithNiagaraMeshParticles=true
  blendMode=BLEND_Additive
  shadingModel=MSM_Unlit
  usedShadingModels=Unlit
 }
 nodes [
  n1 = DynamicParameter() {parameterIndex=3, paramNames [DEFAULT_COLOR, EDGE_GLOW, Param3, Param4]}
  n2 = ParticleColor()
  n3 = Time()
  n4 = Multiply(
   A=@n2.RGB
   B=LinearInterpolate(A=Constant3Vector() {constant=<rgba 1 1 1 1>}, B=Constant3Vector() {constant=<rgba 0 0 0 1>}, Alpha=Sine(None=@n3) {period=0.5})
  )
  n5 = TextureCoordinate()
  n6 = Desaturation(None=TextureSampleParameter2D(UVs=Panner(Coordinate=@n5) {speedX=4}) {parameterName=Texture, texture=/VehicleCosmetics/DriftTrail/DriftTrail_Sakura/Texture/Tyre_Texture}.R)
  n7 = Multiply(A=Multiply(A=If(A=@n1.DEFAULT_COLOR, "A > B"=@n4, "A == B"=@n2.RGB, "A < B"=@n2.RGB) {constB=0.1}, B=@n1.EDGE_GLOW) {constB=100}, B=@n6)
  n8 = ComponentMask(None=TextureCoordinate()) {g=true}
  n9 = OneMinus(None=@n8)
  n10 = Sine(None=@n9) {period=2}
  n11 = Round(None=Sine(None=Power(Base=@n10) {constExponent=3}))
  MaterialFunctionCall(Emissive=Multiply(A=Multiply(A=@n7) {constB=0.5}, B=@n11)) {materialFunction=&MF_FX_EmissiveAdjustments}
  n12 = LinearInterpolate(A=@n6, B=Constant() {r=1}, Alpha=Sine(None=Time()) {period=0.1}) {constAlpha=0}
  Multiply(A=@n12, B=@n11)
  n13 = Step(Y=@n8) {constX=0.3}
  n14 = Step(Y=@n9) {constX=0.3}
  Multiply(A=@n12, B=Round(None=Add(A=@n13, B=@n14)))
 ]
}
